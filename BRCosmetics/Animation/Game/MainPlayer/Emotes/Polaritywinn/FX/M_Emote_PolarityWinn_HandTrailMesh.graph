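# Translucent two-sided particle material for the hand trail mesh effect in the PolarityWinn emote.
# Two rotated, counter-panning T_Lines samples form line masks, alongside a panned wispy noise energy mask.
# Two color parameters tint the emissive through additive and three-color lerp blends, finished by MF_FX_EmissiveAdjustments.
# Dynamic parameters BackwardHandMask and DownwardMask offset UV gradients into control masks, with particle color and alpha exposed as named reroutes.
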Material /BRCosmetics/Animation/Game/MainPlayer/Emotes/Polaritywinn/FX/M_Emote_PolarityWinn_HandTrailMesh
graph {
 properties {
  bAutomaticallySetUsageInEditor=false
  bUsedWithMeshParticles=true
  bUsedWithNiagaraMeshParticles=true
  bUsedWithNiagaraSprites=true
  blendMode=BLEND_Translucent
  previewMesh=&/SM_Emote_PolarityWinn_RibbonA
  twoSided=true
  usedShadingModels=Default Lit
 }
 nodes [
  n1 = VectorParameter() {parameterName=Color1, defaultValue=<rgba 1.052 0 3 1>}
  n2 = VectorParameter() {parameterName=Color2, defaultValue=<rgba 0 0.7799 1 1>}
  n3 = Add(A=Multiply(A=NamedRerouteUsage(), B=@n1.RGB), B=Multiply(A=NamedRerouteUsage(), B=@n2.RGB))
  n4 = NamedRerouteUsage()
  n5 = MaterialFunctionCall(A=Constant(), B=@n1.RGB, C=@n2.RGB, Alpha=NamedRerouteUsage()) {materialFunction=&Lerp_3Color}
  MaterialFunctionCall(Emissive=Add(A=Multiply(A=@n3, B=@n4), B=Multiply(A=@n5.Result) {constB=0.01})) {materialFunction=&MF_FX_EmissiveAdjustments}
  n6 = TextureCoordinate()
  n7 = ComponentMask(None=@n6) {g=true}
  n8 = Multiply(A=@n7, B=OneMinus(None=@n7))
  n9 = NamedRerouteUsage()
  n10 = Add(A=Saturate(None=Subtract(A=@n9) {constB=0.95}), B=NamedRerouteUsage())
  Multiply(A=Multiply(A=Multiply(A=@n8, B=@n10), B=NamedRerouteUsage()), B=ComponentMask(None=@n6) {r=true})
  n11 = TextureCoordinate()
  n12 = Constant() {r=0.25}
  n13 = TextureSample(UVs=Panner(Coordinate=MaterialFunctionCall(UVs=@n11, "Rotation Angle (0-1)"=@n12) {materialFunction=&CustomRotator}.Rotated Values) {speedX=-0.1}) {samplerType=SAMPLERTYPE_LinearColor, texture=&/T_Lines}
  n14 = TextureCoordinate() {uTiling=0.75, vTiling=0.75}
  n15 = TextureSample(UVs=Panner(Coordinate=MaterialFunctionCall(UVs=@n14, "Rotation Angle (0-1)"=@n12) {materialFunction=&CustomRotator}.Rotated Values) {speedX=0.3}) {samplerType=SAMPLERTYPE_LinearColor, texture=&T_Lines}
  NamedRerouteDeclaration(None=Saturate(None=Add(A=@n13.G, B=@n15.G))) {name=G Lines, nodeColor=<rgba 0 1 0.1094 1>}
  NamedRerouteDeclaration(None=Saturate(None=Add(A=@n15.R, B=@n13.R))) {name=R Lines, nodeColor=<rgba 1 0 0.2109 1>}
  n16 = Panner(Coordinate=TextureCoordinate() {uTiling=5}) {speedX=0.4}
  NamedRerouteDeclaration(None=SmoothStep(Value=TextureSample(UVs=@n16) {texture=/BRCosmetics/Athena/Effects/Particles/Skydiving/NS_Contrail_Chainmail/T_WispyFireNoise03}.R) {constMax=0.1}) {name=Color Energy, nodeColor=<rgba 0 0.5312 1 1>}
  n17 = Add(A=@n13.G, B=@n13.G)
  n18 = Add(A=@n15.R, B=@n15.R)
  NamedRerouteDeclaration(None=Saturate(None=Multiply(A=@n17, B=@n18))) {name=A Lines, nodeColor=<rgba 0.8516 1 0 1>}
  n19 = ParticleColor()
  NamedRerouteDeclaration(None=@n19.RGB) {name=Particle Color, nodeColor=<rgba 0.8281 0 1 1>}
  NamedRerouteDeclaration(None=@n19.A) {name=Particle Alpha, nodeColor=<rgba 0 1 0.5078 1>}
  n20 = TextureCoordinate()
  n21 = Multiply(A=OneMinus(None=ComponentMask(None=@n20) {r=true})) {constB=5}
  n22 = DynamicParameter() {defaultValue=<rgba 1 2 1 1>, paramNames [BackwardHandMask, DownwardMask, Param3, Param4]}
  n23 = Multiply(A=ComponentMask(None=TextureCoordinate()) {r=true}) {constB=5}
  n24 = Multiply(A=Saturate(None=Add(A=@n21, B=@n22.BackwardHandMask) {constB=-0.6}), B=Saturate(None=Add(A=@n23, B=@n22.DownwardMask) {constB=-3}))
  NamedRerouteDeclaration(None=Multiply(A=Saturate(None=@n24), B=NamedRerouteUsage())) {name=Main Control Masks, nodeColor=<rgba 1 0.1875 0 1>}
 ]
}
