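# Niagara one-shot pulse effect for the PolarityWinn emote, a short cyan, blue and magenta burst.
# Active emitters layer expanding, fading mesh rings, a glow flare with its light attributes disabled, and distance-culled radial sparks and squares slowed by drag.
# Time-of-day scalars adjust brightness, front-end checks adjust flare size and camera offset, and System.AudioReactivity feeds ring material parameters.
# Four other emitters, mostly duplicates of these, are disabled.
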NiagaraSystem /BRCosmetics/Animation/Game/MainPlayer/Emotes/Polaritywinn/FX/NS_Emote_PolarityWinn_Pulse
dependencies {
 usedDataInterfaces [
  &NiagaraDataInterfaceArrayColor
  &NiagaraDataInterfaceCurve
  &NiagaraDataInterfaceColorCurve
  &NiagaraDataInterfaceMeshRendererInfo
  &NiagaraDataInterfaceVector2DCurve
 ]
 usedDynamicInputs [/Niagara/DynamicInputs/Lerp/SmoothLerpOverTime_Float, &RemapRange, &Lerp_Float]
 usedModules [
  &self:SystemSpawnScript.SystemScriptSource.SystemScriptGraph.NiagaraNodeAssignment_3.SetVariables_#1
  &self:SystemSpawnScript.SystemScriptSource.SystemScriptGraph.NiagaraNodeAssignment_4.SetVariables_#2
  &SystemState
  &self:SystemSpawnScript.SystemScriptSource.SystemScriptGraph.NiagaraNodeAssignment_0.SetVariables_#3
  &self:SystemSpawnScript.SystemScriptSource.SystemScriptGraph.NiagaraNodeAssignment_1.SetVariables_#4
  &self:SystemSpawnScript.SystemScriptSource.SystemScriptGraph.NiagaraNodeAssignment_2.SetVariables_#5
 ]
}
emitters {
 BurstRing {
  emitterUpdateScript [
   &EmitterState(
    Inactive Response=::NewEnumerator1
    Life Cycle Mode=::NewEnumerator1
    Loop Behavior=::NewEnumerator1
    Loop Duration Mode=::NewEnumerator1
    UseLoopDelay=false
   )
   &SpawnBurst_Instantaneous(Spawn Count=5, Use Spawn Probability=false, UseLoopCountLimit=false)
   &SpawnRate(SpawnRate=5, Use Spawn Probability=false) {enabled=false}
  ]
  particleSpawnScript [
   &InitializeParticle(
    Lifetime Max=0.45
    Lifetime Min=0.1
    Lifetime Mode=::NewEnumerator1
    Mesh Scale Mode=::NewEnumerator1
    Mesh Uniform Scale Max=0.2
    Mesh Uniform Scale Min=0.07
    Sprite Size Mode=::NewEnumerator3
    UsePositionOffset=false
    Write Material Random=true
   )
   &Color(Color=&SelectLinearColorFromArray(
     Array Sampling Mode=::NewEnumerator0
     Direct Array Index=?
     Direct Array Mode=::NewEnumerator1
     Color Selection Array=DataInterface() {propertyValues {ColorData [<rgba 0 0.3319 1 1>, <rgba 0.9226 0.0503 1 1>, <rgba 0 0 0 0>]}}
     Random Seed=0
    )) {enabled=false}
  ]
  particleUpdateScript [
   &ParticleState()
   /SetVariables(Particles.MeshOrientation=<1 0 1 0>) {isSetParameters=true, module=SetVariables_#6}
   &DynamicMaterialParameters(
    Index 0 Param 1=&MakeFloatFromVector(Vector=$Particles.Scale, Channel=::NewEnumerator0)
    Index 0 Param 2=2
    Index 0 Param 3=&FloatFromCurve(FloatCurve=DataInterface() {propertyValues {Curve {keys [<at 0 0>,<at 1 0>]}}}, CurveIndex=$System.AudioReactivity)
    Index 0 Param 4=15
    Index 1 Param 1=$Particles.Timeline.Time
    Index 1 Param 2=1
    Index 1 Param 3=0.5
    Index 1 Param 4=$Particles.Timeline.Time
    Index 2 Param 1=2
    Index 2 Param 2=5
    Index 2 Param 3=&FloatFromCurve(FloatCurve=DataInterface() {propertyValues {Curve {keys [<at 0 0.5>,<at 1 1>]}}}, CurveIndex=$System.AudioReactivity)
    Module.Index0Param0Write=true
    Module.Index0Param1Write=true
    Module.Index0Param2Write=true
    Module.Index0Param3Write=true
    Module.Index1Param0Write=true
    Module.Index1Param1Write=true
    Module.Index1Param2Write=true
    Module.Index1Param3Write=true
    Module.Index2Param0Write=true
    Module.Index2Param1Write=true
    Module.Index2Param2Write=true
    Module.Index2Param3Write=true
    Write Parameter Index 1=true
    Write Parameter Index 2=true
   )
   &ScaleMeshSize(Scale Factor=&VectorFromFloat(Value=&FloatFromCurve(FloatCurve=DataInterface() {propertyValues {Curve {keys [
          {interpMode=RCIM_Cubic, leaveTangent=0.6196, tangentMode=RCTM_Break, time=0, value=0.5}
          {arriveTangent=0.7561, interpMode=RCIM_Cubic, tangentMode=RCTM_Break, time=1, value=1.7}
         ]}}})))
   &ScaleColor(
    Scale Alpha=&FloatFromCurve(FloatCurve=DataInterface() {propertyValues {Curve {keys [
         {interpMode=RCIM_Cubic, tangentMode=RCTM_User, time=0, value=0}
         {arriveTangent=-2.0542, interpMode=RCIM_Cubic, leaveTangent=-2.0542, tangentMode=RCTM_User, time=0.2, value=1}
         {arriveTangent=-0.4397, tangentMode=RCTM_User, time=0.5, value=0}
        ]}}})
    Scale RGB=&VectorFromFloat(Value=&FloatFromCurve(FloatCurve=DataInterface() {propertyValues {Curve {
         defaultValue=1
         keys [{interpMode=RCIM_Cubic, tangentMode=RCTM_User, time=0, value=1}, {interpMode=RCIM_Cubic, tangentMode=RCTM_User, time=0.5, value=0}]
        }}}))
    ScaleA=true
    ScaleRGB=true
   )
   &ScaleColor(
    Scale Alpha=0.7
    Scale RGB=&VectorFromFloat(Value=&FNTimeOfDayScalar(Night Scale=1, Morning Scale=1, Day Scale=5, Evening Scale=1))
    ScaleA=false
    ScaleRGB=true
   ) {module=ScaleColor003}
   &Color(
    Color=&ColorFromCurve(ColorCurve=DataInterface() {propertyValues {
       AlphaCurve {keys [<at 0 1>,<at 1 0>]}
       BlueCurve {keys [<at 0.0238 1>,<at 0.1259 1 RCIM_Cubic>,<at 0.4 1>,<at 0.5655 0.4072>,<at 1 1>]}
       GreenCurve {keys [
         <at 0.0238 0.0503>
         {arriveTangent=0.7488, interpMode=RCIM_Cubic, leaveTangent=0.7488, time=0.1259, value=0.9157}
         <at 0.4 0.3319>
         <at 0.5655 0.107>
         <at 1 0.3357>
        ]}
       RedCurve {keys [
         <at 0.0238 0.9226>
         {arriveTangent=-2.4526, interpMode=RCIM_Cubic, leaveTangent=-2.4526, time=0.1259, value=0}
         <at 0.4 0>
         <at 0.5655 0.227>
         <at 1 0>
        ]}
      }})
    Scale Alpha=&FloatFromCurve(FloatCurve=DataInterface() {propertyValues {Curve {keys [
         {interpMode=RCIM_Cubic, tangentMode=RCTM_User, time=0, value=0}
         {arriveTangent=-2.0542, interpMode=RCIM_Cubic, leaveTangent=-2.0542, tangentMode=RCTM_User, time=0.2, value=1}
         {arriveTangent=-0.4397, tangentMode=RCTM_User, time=0.8, value=0}
        ]}}})
   ) {module=Color002}
   &KillParticles(Kill Particles=&SetBoolByFloatComparison(A=$Particles.NormalizedAge, B=0.7))
  ]
  rendererClasses [&NiagaraMeshRendererProperties]
  renderers [
   {rendererClass=&NiagaraMeshRendererProperties, rendererIndex=0}
   {rendererClass=&NiagaraMeshRendererProperties, rendererIndex=1}
  ]
 }
 BurstRing001 {
  bEnabled=false
  emitterUpdateScript [
   &EmitterState(
    Inactive Response=::NewEnumerator1
    Life Cycle Mode=::NewEnumerator1
    Loop Behavior=::NewEnumerator1
    Loop Duration Mode=::NewEnumerator1
    UseLoopDelay=false
   )
   &SpawnBurst_Instantaneous(Spawn Count=5, Use Spawn Probability=false, UseLoopCountLimit=false)
   &SpawnRate(SpawnRate=5, Use Spawn Probability=false) {enabled=false}
  ]
  particleSpawnScript [
   &InitializeParticle(
    Lifetime Max=0.45
    Lifetime Min=0.1
    Lifetime Mode=::NewEnumerator1
    Mesh Scale Mode=::NewEnumerator1
    Mesh Uniform Scale Max=0.2
    Mesh Uniform Scale Min=0.07
    Sprite Size Mode=::NewEnumerator3
    UsePositionOffset=false
    Write Material Random=true
   )
   &Color(Color=&SelectLinearColorFromArray(
     Array Sampling Mode=::NewEnumerator0
     Direct Array Index=?
     Direct Array Mode=::NewEnumerator1
     Color Selection Array=DataInterface() {propertyValues {ColorData [<rgba 0 0.3319 1 1>, <rgba 0.9226 0.0503 1 1>, <rgba 0 0 0 0>]}}
     Random Seed=0
    )) {enabled=false}
  ]
  particleUpdateScript [
   &ParticleState()
   /SetVariables(Particles.MeshOrientation=<1 0 1 0>) {isSetParameters=true, module=SetVariables_#6}
   /SetVariables(Particles.MeshOrientation=<0 1 1 0>) {enabled=false, isSetParameters=true}
   &Timeline(Initial Start Time=256.258, Play Rate=0.7, Timeline Beginning=256.258, Timeline End=100000) {enabled=false}
   &DynamicMaterialParameters(
    Index 0 Param 1=&MakeFloatFromVector(Vector=$Particles.Scale, Channel=::NewEnumerator0)
    Index 0 Param 2=2
    Index 0 Param 3=&FloatFromCurve(FloatCurve=DataInterface() {propertyValues {Curve {keys [<at 0 0>,<at 1 0>]}}}, CurveIndex=$System.AudioReactivity)
    Index 0 Param 4=15
    Index 1 Param 1=$Particles.Timeline.Time
    Index 1 Param 2=1
    Index 1 Param 3=0.5
    Index 1 Param 4=$Particles.Timeline.Time
    Index 2 Param 1=2
    Index 2 Param 2=5
    Index 2 Param 3=&FloatFromCurve(FloatCurve=DataInterface() {propertyValues {Curve {keys [<at 0 0.5>,<at 1 1>]}}}, CurveIndex=$System.AudioReactivity)
    Module.Index0Param0Write=true
    Module.Index0Param1Write=true
    Module.Index0Param2Write=true
    Module.Index0Param3Write=true
    Module.Index1Param0Write=true
    Module.Index1Param1Write=true
    Module.Index1Param2Write=true
    Module.Index1Param3Write=true
    Module.Index2Param0Write=true
    Module.Index2Param1Write=true
    Module.Index2Param2Write=true
    Module.Index2Param3Write=true
    Write Parameter Index 1=true
    Write Parameter Index 2=true
   )
   &ScaleMeshSize(Scale Factor=&VectorFromFloat(Value=&FloatFromCurve(FloatCurve=DataInterface() {propertyValues {Curve {keys [
          {interpMode=RCIM_Cubic, leaveTangent=0.6196, tangentMode=RCTM_Break, time=0, value=0.5}
          {arriveTangent=0.7561, interpMode=RCIM_Cubic, tangentMode=RCTM_Break, time=1, value=1.7}
         ]}}})))
   &ScaleColor(
    Scale Alpha=&FloatFromCurve(FloatCurve=DataInterface() {propertyValues {Curve {keys [
         {interpMode=RCIM_Cubic, tangentMode=RCTM_User, time=0, value=0}
         {arriveTangent=-2.0542, interpMode=RCIM_Cubic, leaveTangent=-2.0542, tangentMode=RCTM_User, time=0.2, value=1}
         {arriveTangent=-0.4397, tangentMode=RCTM_User, time=0.5, value=0}
        ]}}})
    Scale RGB=&VectorFromFloat(Value=&FloatFromCurve(FloatCurve=DataInterface() {propertyValues {Curve {
         defaultValue=1
         keys [{interpMode=RCIM_Cubic, tangentMode=RCTM_User, time=0, value=1}, {interpMode=RCIM_Cubic, tangentMode=RCTM_User, time=0.5, value=0}]
        }}}))
    ScaleA=true
    ScaleRGB=true
   )
   &ScaleColor(
    Scale Alpha=0.7
    Scale RGB=&VectorFromFloat(Value=&FNTimeOfDayScalar(Night Scale=1, Morning Scale=1, Day Scale=5, Evening Scale=1))
    ScaleA=false
    ScaleRGB=true
   ) {module=ScaleColor003}
   &Color(
    Color=&ColorFromCurve(ColorCurve=DataInterface() {propertyValues {
       AlphaCurve {keys [<at 0 1>,<at 1 0>]}
       BlueCurve {keys [<at 0.0238 1>,<at 0.1259 1 RCIM_Cubic>,<at 0.4 1>,<at 0.5655 0.4072>,<at 1 1>]}
       GreenCurve {keys [
         <at 0.0238 0.0503>
         {arriveTangent=0.7488, interpMode=RCIM_Cubic, leaveTangent=0.7488, time=0.1259, value=0.9157}
         <at 0.4 0.3319>
         <at 0.5655 0.107>
         <at 1 0.3357>
        ]}
       RedCurve {keys [
         <at 0.0238 0.9226>
         {arriveTangent=-2.4526, interpMode=RCIM_Cubic, leaveTangent=-2.4526, time=0.1259, value=0}
         <at 0.4 0>
         <at 0.5655 0.227>
         <at 1 0>
        ]}
      }})
    Scale Alpha=&FloatFromCurve(FloatCurve=DataInterface() {propertyValues {Curve {keys [
         {interpMode=RCIM_Cubic, tangentMode=RCTM_User, time=0, value=0}
         {arriveTangent=-2.0542, interpMode=RCIM_Cubic, leaveTangent=-2.0542, tangentMode=RCTM_User, time=0.2, value=1}
         {arriveTangent=-0.4397, tangentMode=RCTM_User, time=0.8, value=0}
        ]}}})
   ) {module=Color002}
   &ScaleColor(
    Scale Alpha=0.7
    Scale RGB=&VectorFromFloat(Value=&FNTimeOfDayScalar(Night Scale=5, Morning Scale=1, Evening Scale=1))
    ScaleA=true
    ScaleRGB=true
   ) {enabled=false, module=ScaleColor001}
   &ScaleColor(
    Scale RGB=&VectorFromFloat(Value=&FloatFromCurve(
      FloatCurve=DataInterface() {propertyValues {Curve {keys [{interpMode=RCIM_Cubic, leaveTangent=1.4829, tangentMode=RCTM_Break, time=0, value=1}, <at 1 2.3>]}}}
      CurveIndex=$User.Hitted
     ))
    ScaleA=true
    ScaleRGB=true
   ) {enabled=false, module=ScaleColor002}
   &Color(Color=&ColorFromCurve(ColorCurve=DataInterface() {propertyValues {
       AlphaCurve {keys [<at 0 1 RCIM_Cubic>,<at 1 0 RCIM_Cubic>]}
       BlueCurve {keys [<at 0 1 RCIM_Cubic>,<at 0.2697 0.8863>,<at 0.6621 0.8058>,<at 1 0.7366 RCIM_Cubic>]}
       GreenCurve {keys [<at 0 0.3319 RCIM_Cubic>,<at 0.2697 0.0493>,<at 0.6621 0.7096>,<at 1 1 RCIM_Cubic>]}
       RedCurve {keys [<at 0 0 RCIM_Cubic>,<at 0.2697 1>,<at 0.6621 0.0014>,<at 1 0 RCIM_Cubic>]}
      }})) {enabled=false, module=Color001}
   &KillParticles(Kill Particles=&SetBoolByFloatComparison(A=$Particles.NormalizedAge, B=0.7))
   /SetVariables(Particles.MeshOrientation=&FindQuatBetweenTwoVectors(From Vector=<1 0 1>, To Vector=<0 0 1>)) {enabled=false, isSetParameters=true}
  ]
  rendererClasses [&NiagaraMeshRendererProperties]
  renderers [
   {rendererClass=&NiagaraMeshRendererProperties, rendererIndex=0}
   {rendererClass=&NiagaraMeshRendererProperties, rendererIndex=1}
  ]
 }
 Flare {
  bEnabled=false
  emitterUpdateScript [
   &EmitterState(Life Cycle Mode=::NewEnumerator1, Loop Behavior=::NewEnumerator1, Loop Delay=0.1, Loop Duration=1, UseLoopDelay=false)
   &SpawnBurst_Instantaneous(Spawn Count=1, Use Spawn Probability=false)
  ]
  particleSpawnScript [
   &InitializeParticle(
    Color=<rgba 5 5 5 0.5>
    Lifetime=0.3
    Sprite Rotation Mode=::NewEnumerator2
    Sprite Size Mode=::NewEnumerator3
    Uniform Sprite Size=18
    UsePositionOffset=false
    Write Material Random=true
   )
   &DynamicMaterialParameters(
    Index 0 Param 1=&RandomRangeFloat(Minimum=0.4, Maximum=0.6)
    Index 0 Param 2=1
    Index 0 Param 3=1
    Index 0 Param 4=1
    Module.Param0WriteEnabled=true
    Module.Param1WriteEnabled=true
    Module.Param2WriteEnabled=true
    Module.Param3WriteEnabled=true
   )
   /SetVariables(Particles.Dark_Glow_CameraOffset=&MakeCustomFloatFromBool(Bool=$User.bIsFrontEndPreview, True Float=1, False Float=50)) {isSetParameters=true, module=SetVariables_#7}
   &CameraOffset(Camera Offset Amount=&MakeCustomFloatFromBool(Bool=$User.bIsFrontEndPreview, True Float=1, False Float=30))
  ]
  particleUpdateScript [
   /SetVariables(Particles.Dark_Glow=&MakeLinearColorFromVectorAndFloat(
     "Vector (RGB)"=&MakeVectorFromLinearColorRGB(Color=&MultiplyLinearColor(
       Linear Color=&ColorFromCurve(ColorCurve=DataInterface() {propertyValues {
          AlphaCurve {keys []}
          BlueCurve {keys [<at 0.1777 0.45 RCIM_Cubic>,<at 0.5555 0.6354>,<at 0.7486 0.95>,<at 1 0.8594 RCIM_Cubic>]}
          GreenCurve {keys [<at 0.1777 0.2475 RCIM_Cubic>,<at 0.5555 0.0318>,<at 0.7486 0.095>,<at 1 0.4082 RCIM_Cubic>]}
          RedCurve {keys [<at 0.1777 0.045 RCIM_Cubic>,<at 0.5555 0.233>,<at 0.7486 0.38>,<at 1 0.0859 RCIM_Cubic>]}
         }})
       Scale Factor=&MakeVector4_FromFloat(Float=0.5)
      ))
     "Float (Alpha)"=&Multiply_Float(
      A=&Multiply_Float(A=1, B=1)
      B=&FloatFromCurve(FloatCurve=DataInterface() {propertyValues {Curve {keys [<at 0 0>,<at 0.2 1>,<at 0.7 0.4>,<at 1 0>]}}})
     )
    )) {isSetParameters=true, module=SetVariables_#8}
   /SetVariables(Particles.Dark_Glow_Size=&Multiply_Vector2(
     A=&Vector2DFromFloat(Value=5)
     B=&Vector2DFromFloat(Value=&FloatFromCurve(
       FloatCurve=DataInterface() {propertyValues {Curve {keys [<at 0 1.5 RCIM_Cubic>,<at 1 0.5>]}}}
       Scale Curve=&FN_FrontEndFloatModifier(Gameplay=1.5, Front End=0.6)
      ))
    )) {isSetParameters=true, module=SetVariables_#9}
   &ParticleState()
   &ScaleColor(
    Scale Alpha=&Multiply_Float(
     A=1
     B=&FloatFromCurve(
      FloatCurve=DataInterface() {propertyValues {Curve {keys [<at 0 0>,<at 0.2 1>,<at 0.6 0.6>,<at 0.7 -0.05>]}}}
      Scale Curve=&FNTimeOfDayScalar(Night Scale=1, Morning Scale=1, Evening Scale=1)
     )
    )
    Scale RGB=&Multiply_VectorByFloat(
     Vector=&MakeVectorFromLinearColorRGB(Color=&ColorFromCurve(ColorCurve=DataInterface() {propertyValues {
         AlphaCurve {keys []}
         BlueCurve {keys [<at 0.0097 0.7 RCIM_Cubic>,<at 1 1 RCIM_Cubic>]}
         GreenCurve {keys [<at 0.0097 0 RCIM_Cubic>,<at 1 0.6667 RCIM_Cubic>]}
         RedCurve {keys [<at 0.0097 0.4531 RCIM_Cubic>,<at 1 0 RCIM_Cubic>]}
        }}))
     Float=5
    )
    ScaleA=true
    ScaleRGB=true
   )
   &ScaleSpriteSize(Scale Factor=&Vector2DFromFloat(Value=&FloatFromCurve(
      FloatCurve=DataInterface() {propertyValues {Curve {keys [<at 0 1.5 RCIM_Cubic>,<at 0.8 0>]}}}
      Scale Curve=&FN_FrontEndFloatModifier(Gameplay=1.1, Front End=0.6)
     )))
   &FN_Light_Attributes(
    Color=<rgba 1.2 1.2 8 1>
    Create Color Attribute=true
    Create Exponent Attribute=true
    Create Position Attribute=true
    Create Radius Attribute=true
    Position=<20 0 0>
    Radius=100
    Scale Color=&Multiply_Float(
     A=3000
     B=&FloatFromCurve(FloatCurve=DataInterface() {propertyValues {Curve {keys [
          {arriveTangent=6.278, interpMode=RCIM_Cubic, leaveTangent=6.278, tangentMode=RCTM_User, time=0, value=0}
          {arriveTangent=-2.0311, interpMode=RCIM_Cubic, leaveTangent=-2.0311, tangentMode=RCTM_User, time=0.5, value=1}
          {arriveTangent=-2.1506, interpMode=RCIM_Cubic, tangentMode=RCTM_Break, time=1, value=0}
         ]}}})
    )
   ) {enabled=false}
  ]
  rendererClasses [&NiagaraSpriteRendererProperties, &NiagaraLightRendererProperties]
  renderers [
   {rendererClass=&NiagaraSpriteRendererProperties}
   {rendererClass=&NiagaraSpriteRendererProperties, rendererIndex=1}
   {rendererClass=&NiagaraLightRendererProperties, rendererIndex=2}
  ]
 }
 Flare001 {
  emitterUpdateScript [
   &EmitterState(Life Cycle Mode=::NewEnumerator1, Loop Behavior=::NewEnumerator1, Loop Delay=0.1, Loop Duration=1, UseLoopDelay=false)
   &SpawnBurst_Instantaneous(Spawn Count=1, Use Spawn Probability=false)
  ]
  particleSpawnScript [
   &InitializeParticle(
    Color=<rgba 5 5 5 0.5>
    Lifetime=0.3
    Sprite Rotation Mode=::NewEnumerator2
    Sprite Size Mode=::NewEnumerator3
    Uniform Sprite Size=18
    UsePositionOffset=false
    Write Material Random=true
   )
   &DynamicMaterialParameters(
    Index 0 Param 1=&RandomRangeFloat(Minimum=0.4, Maximum=0.6)
    Index 0 Param 2=1
    Index 0 Param 3=1
    Index 0 Param 4=1
    Module.Param0WriteEnabled=true
    Module.Param1WriteEnabled=true
    Module.Param2WriteEnabled=true
    Module.Param3WriteEnabled=true
   )
   /SetVariables(Particles.Dark_Glow_CameraOffset=&MakeCustomFloatFromBool(Bool=$User.bIsFrontEndPreview, True Float=1, False Float=50)) {isSetParameters=true, module=SetVariables_#7}
   &CameraOffset(Camera Offset Amount=&MakeCustomFloatFromBool(Bool=$User.bIsFrontEndPreview, True Float=1, False Float=30))
  ]
  particleUpdateScript [
   /SetVariables(Particles.Dark_Glow=&MakeLinearColorFromVectorAndFloat(
     "Vector (RGB)"=&MakeVectorFromLinearColorRGB(Color=&MultiplyLinearColor(
       Linear Color=&ColorFromCurve(ColorCurve=DataInterface() {propertyValues {
          AlphaCurve {keys []}
          BlueCurve {keys [<at 0.1777 0.45 RCIM_Cubic>,<at 0.5555 0.6354>,<at 0.7486 0.95>,<at 1 0.8594 RCIM_Cubic>]}
          GreenCurve {keys [<at 0.1777 0.2475 RCIM_Cubic>,<at 0.5555 0.0318>,<at 0.7486 0.095>,<at 1 0.4082 RCIM_Cubic>]}
          RedCurve {keys [<at 0.1777 0.045 RCIM_Cubic>,<at 0.5555 0.233>,<at 0.7486 0.38>,<at 1 0.0859 RCIM_Cubic>]}
         }})
       Scale Factor=&MakeVector4_FromFloat(Float=0.5)
      ))
     "Float (Alpha)"=&Multiply_Float(
      A=&Multiply_Float(A=1, B=1)
      B=&FloatFromCurve(FloatCurve=DataInterface() {propertyValues {Curve {keys [<at 0 0>,<at 0.2 1>,<at 0.7 0.4>,<at 1 0>]}}})
     )
    )) {isSetParameters=true, module=SetVariables_#8}
   /SetVariables(Particles.Dark_Glow_Size=&Multiply_Vector2(
     A=&Vector2DFromFloat(Value=5)
     B=&Vector2DFromFloat(Value=&FloatFromCurve(
       FloatCurve=DataInterface() {propertyValues {Curve {keys [<at 0 1.5 RCIM_Cubic>,<at 1 0.5>]}}}
       Scale Curve=&FN_FrontEndFloatModifier(Gameplay=1.5, Front End=0.6)
      ))
    )) {isSetParameters=true, module=SetVariables_#9}
   &ParticleState()
   &ScaleColor(
    Scale Alpha=&Multiply_Float(
     A=1
     B=&FloatFromCurve(
      FloatCurve=DataInterface() {propertyValues {Curve {keys [<at 0 0>,<at 0.2 1>,<at 0.6 0.6>,<at 0.7 -0.05>]}}}
      Scale Curve=&FNTimeOfDayScalar(Night Scale=1, Morning Scale=1, Evening Scale=1)
     )
    )
    Scale RGB=&Multiply_VectorByFloat(
     Vector=&MakeVectorFromLinearColorRGB(Color=&ColorFromCurve(ColorCurve=DataInterface() {propertyValues {
         AlphaCurve {keys []}
         BlueCurve {keys [<at 0.0097 0.7 RCIM_Cubic>,<at 1 1 RCIM_Cubic>]}
         GreenCurve {keys [<at 0.0097 0 RCIM_Cubic>,<at 1 0.6667 RCIM_Cubic>]}
         RedCurve {keys [<at 0.0097 0.4531 RCIM_Cubic>,<at 1 0 RCIM_Cubic>]}
        }}))
     Float=5
    )
    ScaleA=true
    ScaleRGB=true
   )
   &ScaleSpriteSize(Scale Factor=&Vector2DFromFloat(Value=&FloatFromCurve(
      FloatCurve=DataInterface() {propertyValues {Curve {keys [<at 0 1.5 RCIM_Cubic>,<at 0.8 0>]}}}
      Scale Curve=&FN_FrontEndFloatModifier(Gameplay=1.1, Front End=0.6)
     )))
   &FN_Light_Attributes(
    Color=<rgba 1.2 1.2 8 1>
    Create Color Attribute=true
    Create Exponent Attribute=true
    Create Position Attribute=true
    Create Radius Attribute=true
    Position=<20 0 0>
    Radius=100
    Scale Color=&Multiply_Float(
     A=3000
     B=&FloatFromCurve(FloatCurve=DataInterface() {propertyValues {Curve {keys [
          {arriveTangent=6.278, interpMode=RCIM_Cubic, leaveTangent=6.278, tangentMode=RCTM_User, time=0, value=0}
          {arriveTangent=-2.0311, interpMode=RCIM_Cubic, leaveTangent=-2.0311, tangentMode=RCTM_User, time=0.5, value=1}
          {arriveTangent=-2.1506, interpMode=RCIM_Cubic, tangentMode=RCTM_Break, time=1, value=0}
         ]}}})
    )
   ) {enabled=false}
  ]
  rendererClasses [&NiagaraSpriteRendererProperties, &NiagaraLightRendererProperties]
  renderers [
   {rendererClass=&NiagaraSpriteRendererProperties}
   {rendererClass=&NiagaraSpriteRendererProperties, rendererIndex=1}
   {rendererClass=&NiagaraLightRendererProperties, rendererIndex=2}
  ]
 }
 Sparks {
  bEnabled=false
  emitterUpdateScript [
   &EmitterState(
    Enable Distance Culling=true
    Max Distance Response=::KillImmediately
    Module.bApplySpawnCountScaleByDistance=true
    Module.bUseMaxDistance=true
    Module.bUseMinDistance=false
    Scalability Mode=::NewEnumerator1
    Scale Spawn Count=true
    Spawn Count Scale By Distance=DataInterface() {propertyValues {Curve {keys [<at 0.3676 0.9657>,<at 0.9962 0.4982>]}}}
   )
   &SpawnBurst_Instantaneous(Spawn Count=&UniformRangedInt(Minimum=12, Maximum=15), Use Spawn Probability=false)
   &SpawnRate(
    Spawn Probability=&FloatFromCurve(
     FloatCurve=DataInterface() {propertyValues {Curve {keys [{arriveTangent=-12.2872, interpMode=RCIM_Cubic, leaveTangent=-12.2872, tangentMode=RCTM_User, time=0, value=1}, <at 0.1 0>]}}}
     CurveIndex=$System.Age
    )
    SpawnRate=444
    Use Spawn Probability=true
   )
  ]
  particleSpawnScript [
   &InitializeParticle(
    Color=<rgba 2 2 2 1>
    Lifetime Max=0.45
    Lifetime Min=0.166
    Lifetime Mode=::NewEnumerator1
    Sprite Size Mode=::NewEnumerator1
    Uniform Sprite Size Max=2
    Uniform Sprite Size Min=1
    UsePositionOffset=false
    Write Material Random=true
   )
   &SphereLocation(
    MaskSpawnGroup=false
    Override Seed=false
    Sphere Radius=5
    Surface Only=false
    UseNonUniformScale=false
    UseOffset=false
    UseSphereOrientation=false
   ) {module=SphereLocation001}
   &AddVelocityFromPoint(Module.Use Falloff Distance=false, Velocity Strength=&UniformRangedFloat(Minimum=250, Maximum=500))
  ]
  particleUpdateScript [
   &ParticleState()
   &Drag(Drag=&UniformRangedFloat(Minimum=6, Maximum=12), Use Linear Drag=true, Use Rotational Drag=false)
   &SolveForcesAndVelocity(Clamp Velocity=false, Limit Acceleration=false)
   &ScaleSpriteSize(Scale Factor=&Vector2DFromCurve(
     Vector2Curve=DataInterface() {propertyValues {
       XCurve {keys [
         {arriveTangent=5.9728, interpMode=RCIM_Cubic, leaveTangent=5.9728, tangentMode=RCTM_User, time=0, value=0.1}
         {arriveTangent=-0.1, interpMode=RCIM_Cubic, leaveTangent=-0.1, tangentMode=RCTM_User, time=0.18, value=0.8}
         {interpMode=RCIM_Cubic, tangentMode=RCTM_User, time=1, value=0}
        ]}
       YCurve {keys [
         {arriveTangent=6.2715, interpMode=RCIM_Cubic, leaveTangent=6.2715, tangentMode=RCTM_User, time=0, value=0.1}
         {arriveTangent=-0.1001, interpMode=RCIM_Cubic, leaveTangent=-0.1001, tangentMode=RCTM_User, time=0.18, value=0.8}
         {interpMode=RCIM_Cubic, tangentMode=RCTM_User, time=0.9988, value=0}
        ]}
      }}
     ScaleCurve=<1 1>
    ))
   &ScaleSpriteSizeBySpeed(Max Scale Factor=<0.5 8>, Min Scale Factor=<0.8 1>)
   &ScaleColor(
    Scale Alpha=&FNTimeOfDayScalar(Night Scale=1, Morning Scale=1, Evening Scale=1)
    Scale RGB=&MakeVectorFromLinearColorRGB(Color=&HueShiftLinearColor(
      Hue=&RandomRangeFloat(Minimum=-0.1, Maximum=0)
      Saturation=1
      Value=1
      Linear Color=&ColorFromCurve(ColorCurve=DataInterface() {propertyValues {
         AlphaCurve {keys [<at 0 1>,<at 1 0>]}
         BlueCurve {keys [<at 0.17 33>,<at 0.2 3>,<at 0.33 2>,<at 0.85 1.5>]}
         GreenCurve {keys [<at 0.17 16.8532>,<at 0.2 1.083>,<at 0.33 0.0511>,<at 0.85 0.4167>]}
         RedCurve {keys [<at 0.17 0>,<at 0.2 0>,<at 0.33 0.6389>,<at 0.85 0>]}
        }})
     ))
    ScaleA=true
    ScaleRGB=true
   )
   &ScaleSpriteSize(
    Scale Sprite Size Mode=::NewEnumerator0
    Uniform Scale Factor=&MakeCustomFloatFromBool(Bool=$User.bIsFrontEndPreview, True Float=$User.FEScale, False Float=1)
   ) {enabled=false, module=ScaleSpriteSize001}
  ]
  rendererClasses [&NiagaraSpriteRendererProperties]
  renderers [{rendererClass=&NiagaraSpriteRendererProperties}]
 }
 Sparks001 {
  emitterUpdateScript [
   &EmitterState(
    Enable Distance Culling=true
    Max Distance Response=::KillImmediately
    Module.bApplySpawnCountScaleByDistance=true
    Module.bUseMaxDistance=true
    Module.bUseMinDistance=false
    Scalability Mode=::NewEnumerator1
    Scale Spawn Count=true
    Spawn Count Scale By Distance=DataInterface() {propertyValues {Curve {keys [<at 0.3676 0.9657>,<at 0.9962 0.4982>]}}}
   )
   &SpawnBurst_Instantaneous(Spawn Count=&UniformRangedInt(Minimum=12, Maximum=15), Use Spawn Probability=false)
   &SpawnRate(
    Spawn Probability=&FloatFromCurve(
     FloatCurve=DataInterface() {propertyValues {Curve {keys [{arriveTangent=-12.2872, interpMode=RCIM_Cubic, leaveTangent=-12.2872, tangentMode=RCTM_User, time=0, value=1}, <at 0.1 0>]}}}
     CurveIndex=$System.Age
    )
    SpawnRate=444
    Use Spawn Probability=true
   )
  ]
  particleSpawnScript [
   &InitializeParticle(
    Color=<rgba 2 2 2 1>
    Lifetime Max=0.45
    Lifetime Min=0.166
    Lifetime Mode=::NewEnumerator1
    Sprite Size Mode=::NewEnumerator1
    Uniform Sprite Size Max=2
    Uniform Sprite Size Min=1
    UsePositionOffset=false
    Write Material Random=true
   )
   &SphereLocation(
    MaskSpawnGroup=false
    Override Seed=false
    Sphere Radius=5
    Surface Only=false
    UseNonUniformScale=false
    UseOffset=false
    UseSphereOrientation=false
   ) {module=SphereLocation001}
   &AddVelocityFromPoint(Module.Use Falloff Distance=false, Velocity Strength=&UniformRangedFloat(Minimum=250, Maximum=500))
  ]
  particleUpdateScript [
   &ParticleState()
   &Drag(Drag=&UniformRangedFloat(Minimum=6, Maximum=12), Use Linear Drag=true, Use Rotational Drag=false)
   &SolveForcesAndVelocity(Clamp Velocity=false, Limit Acceleration=false)
   &ScaleSpriteSize(Scale Factor=&Vector2DFromCurve(
     Vector2Curve=DataInterface() {propertyValues {
       XCurve {keys [
         {arriveTangent=5.9728, interpMode=RCIM_Cubic, leaveTangent=5.9728, tangentMode=RCTM_User, time=0, value=0.1}
         {arriveTangent=-0.1, interpMode=RCIM_Cubic, leaveTangent=-0.1, tangentMode=RCTM_User, time=0.18, value=0.8}
         {interpMode=RCIM_Cubic, tangentMode=RCTM_User, time=1, value=0}
        ]}
       YCurve {keys [
         {arriveTangent=6.2715, interpMode=RCIM_Cubic, leaveTangent=6.2715, tangentMode=RCTM_User, time=0, value=0.1}
         {arriveTangent=-0.1001, interpMode=RCIM_Cubic, leaveTangent=-0.1001, tangentMode=RCTM_User, time=0.18, value=0.8}
         {interpMode=RCIM_Cubic, tangentMode=RCTM_User, time=0.9988, value=0}
        ]}
      }}
     ScaleCurve=<1 1>
    ))
   &ScaleSpriteSizeBySpeed(Max Scale Factor=<0.5 8>, Min Scale Factor=<0.8 1>)
   &ScaleColor(
    Scale Alpha=&FNTimeOfDayScalar(Night Scale=1, Morning Scale=1, Evening Scale=1)
    Scale RGB=&MakeVectorFromLinearColorRGB(Color=&HueShiftLinearColor(
      Hue=&RandomRangeFloat(Minimum=-0.1, Maximum=0)
      Saturation=1
      Value=1
      Linear Color=&ColorFromCurve(ColorCurve=DataInterface() {propertyValues {
         AlphaCurve {keys [<at 0 1>,<at 1 0>]}
         BlueCurve {keys [<at 0.17 33>,<at 0.2 3>,<at 0.33 2>,<at 0.85 1.5>]}
         GreenCurve {keys [<at 0.17 16.8532>,<at 0.2 1.083>,<at 0.33 0.0511>,<at 0.85 0.4167>]}
         RedCurve {keys [<at 0.17 0>,<at 0.2 0>,<at 0.33 0.6389>,<at 0.85 0>]}
        }})
     ))
    ScaleA=true
    ScaleRGB=true
   )
   &ScaleSpriteSize(
    Scale Sprite Size Mode=::NewEnumerator0
    Uniform Scale Factor=&MakeCustomFloatFromBool(Bool=$User.bIsFrontEndPreview, True Float=$User.FEScale, False Float=1)
   ) {enabled=false, module=ScaleSpriteSize001}
  ]
  rendererClasses [&NiagaraSpriteRendererProperties]
  renderers [{rendererClass=&NiagaraSpriteRendererProperties}]
 }
 Squares {
  emitterUpdateScript [
   &EmitterState(
    Enable Distance Culling=true
    Max Distance Response=::KillImmediately
    Module.bApplySpawnCountScaleByDistance=true
    Module.bUseMaxDistance=true
    Module.bUseMinDistance=false
    Scalability Mode=::NewEnumerator1
    Scale Spawn Count=true
    Spawn Count Scale By Distance=DataInterface() {propertyValues {Curve {keys [<at 0.3676 0.9657>,<at 0.9962 0.4982>]}}}
   )
   &SpawnBurst_Instantaneous(Spawn Count=&UniformRangedInt(Minimum=5, Maximum=10), Use Spawn Probability=false)
  ]
  particleSpawnScript [
   &FN_SetSpawnGroup()
   &InitializeParticle(
    Color=&SelectLinearColorFromArray(
     Array Sampling Mode=::NewEnumerator0
     Direct Array Index=?
     Color Selection Array=DataInterface() {propertyValues {ColorData [<rgba 1.0237 0.195 1.3 5>, <rgba 1.1813 0.225 1.5 5>, <rgba 0 0.2667 0.4 5>, <rgba 0 0.6823 1.5 5>, <rgba 1.05 0 1.5 5>]}}
     Random Seed=0
    )
    Lifetime Max=0.45
    Lifetime Min=0.166
    Lifetime Mode=::NewEnumerator1
    Sprite Size Mode=::NewEnumerator1
    Uniform Sprite Size Min=3
    UsePositionOffset=false
    Write Material Random=true
   )
   &SphereLocation(
    MaskSpawnGroup=false
    Override Seed=false
    Sphere Radius=5
    Surface Only=false
    UseNonUniformScale=false
    UseOffset=false
    UseSphereOrientation=false
   ) {module=SphereLocation001}
   &AddVelocityFromPoint(Module.Use Falloff Distance=false, Velocity Strength=&UniformRangedFloat(Minimum=250, Maximum=500))
  ]
  particleUpdateScript [
   &ParticleState()
   &Drag(Drag=&UniformRangedFloat(Minimum=6, Maximum=12), Use Linear Drag=true, Use Rotational Drag=false)
   &SolveForcesAndVelocity(Clamp Velocity=false, Limit Acceleration=false)
   &ScaleSpriteSize(Scale Factor=&Vector2DFromCurve(
     Vector2Curve=DataInterface() {propertyValues {
       XCurve {keys [
         {arriveTangent=9.231, interpMode=RCIM_Cubic, leaveTangent=9.231, tangentMode=RCTM_User, time=0, value=0.1}
         {arriveTangent=-0.1, interpMode=RCIM_Cubic, leaveTangent=-0.1, tangentMode=RCTM_User, time=0.18, value=0.8}
         {arriveTangent=-2.4373, interpMode=RCIM_Cubic, leaveTangent=-2.4373, tangentMode=RCTM_User, time=1, value=0}
        ]}
       YCurve {keys [
         {arriveTangent=9.7317, interpMode=RCIM_Cubic, leaveTangent=9.7317, tangentMode=RCTM_User, time=0, value=0.1}
         {arriveTangent=-0.1001, interpMode=RCIM_Cubic, leaveTangent=-0.1001, tangentMode=RCTM_User, time=0.18, value=0.8}
         {arriveTangent=-2.4373, interpMode=RCIM_Cubic, leaveTangent=-2.4373, tangentMode=RCTM_User, time=0.9988, value=0}
        ]}
      }}
     ScaleCurve=<1 1>
    ))
   &ScaleColor(
    Scale Alpha=0.7
    Scale RGB=&VectorFromFloat(Value=&FNTimeOfDayScalar(Night Scale=1, Morning Scale=1, Day Scale=5, Evening Scale=1))
    ScaleA=true
    ScaleRGB=true
   ) {module=ScaleColor003}
   &DynamicMaterialParameters(Module.Index0Float0Write=true, Module.Index0Float1Write=true, Module.Index0Float2Write=true, Module.Index0Float3Write=true)
  ]
  rendererClasses [&NiagaraSpriteRendererProperties]
  renderers [{rendererClass=&NiagaraSpriteRendererProperties}]
 }
 sparkles001 {
  bEnabled=false
  emitterUpdateScript [
   &EmitterState(
    Enable Distance Culling=true
    Max Distance Response=::KillImmediately
    Module.bApplySpawnCountScaleByDistance=true
    Module.bUseMaxDistance=true
    Module.bUseMinDistance=false
    Scalability Mode=::NewEnumerator1
    Scale Spawn Count=true
    Spawn Count Scale By Distance=DataInterface() {propertyValues {Curve {keys [<at 0.3676 0.9657>,<at 0.9962 0.4982>]}}}
   )
   &SpawnBurst_Instantaneous(Spawn Count=&UniformRangedInt(Minimum=5, Maximum=10), Use Spawn Probability=false)
  ]
  particleSpawnScript [
   &FN_SetSpawnGroup()
   &InitializeParticle(
    Color=&SelectLinearColorFromArray(
     Array Sampling Mode=::NewEnumerator0
     Direct Array Index=?
     Color Selection Array=DataInterface() {propertyValues {ColorData [<rgba 1.0237 0.195 1.3 1>, <rgba 1.1813 0.225 1.5 0.8>, <rgba 0 0.2667 0.4 1>, <rgba 0 0.6823 1.5 1>, <rgba 1.05 0 1.5 1>]}}
     Random Seed=0
    )
    Lifetime Max=0.45
    Lifetime Min=0.166
    Lifetime Mode=::NewEnumerator1
    Sprite Size Mode=::NewEnumerator1
    Uniform Sprite Size Max=15
    Uniform Sprite Size Min=7
    UsePositionOffset=false
    Write Material Random=true
   )
   &SphereLocation(
    MaskSpawnGroup=false
    Override Seed=false
    Sphere Radius=5
    Surface Only=false
    UseNonUniformScale=false
    UseOffset=false
    UseSphereOrientation=false
   ) {module=SphereLocation001}
   &AddVelocityFromPoint(Module.Use Falloff Distance=false, Velocity Strength=&UniformRangedFloat(Minimum=250, Maximum=500))
  ]
  particleUpdateScript [
   &ParticleState()
   &Drag(Drag=&UniformRangedFloat(Minimum=6, Maximum=12), Use Linear Drag=true, Use Rotational Drag=false)
   &SolveForcesAndVelocity(Clamp Velocity=false, Limit Acceleration=false)
   &ScaleSpriteSize(Scale Factor=&Vector2DFromCurve(
     Vector2Curve=DataInterface() {propertyValues {
       XCurve {keys [
         {arriveTangent=9.231, interpMode=RCIM_Cubic, leaveTangent=9.231, tangentMode=RCTM_User, time=0, value=0.1}
         {arriveTangent=-0.1, interpMode=RCIM_Cubic, leaveTangent=-0.1, tangentMode=RCTM_User, time=0.18, value=0.8}
         {arriveTangent=-2.4373, interpMode=RCIM_Cubic, leaveTangent=-2.4373, tangentMode=RCTM_User, time=1, value=0}
        ]}
       YCurve {keys [
         {arriveTangent=9.7317, interpMode=RCIM_Cubic, leaveTangent=9.7317, tangentMode=RCTM_User, time=0, value=0.1}
         {arriveTangent=-0.1001, interpMode=RCIM_Cubic, leaveTangent=-0.1001, tangentMode=RCTM_User, time=0.18, value=0.8}
         {arriveTangent=-2.4373, interpMode=RCIM_Cubic, leaveTangent=-2.4373, tangentMode=RCTM_User, time=0.9988, value=0}
        ]}
      }}
     ScaleCurve=<1 1>
    ))
   &ScaleColor(
    Scale Alpha=0.7
    Scale RGB=&VectorFromFloat(Value=&FNTimeOfDayScalar(Night Scale=1, Morning Scale=1, Day Scale=5, Evening Scale=1))
    ScaleA=true
    ScaleRGB=true
   ) {module=ScaleColor003}
   &ScaleColor(
    Scale RGB=&MakeVectorFromLinearColorRGB(Color=&HueShiftLinearColor(
      Hue=&RandomRangeFloat(Minimum=-0.1, Maximum=0)
      Saturation=1
      Value=1
      Linear Color=&ColorFromCurve(ColorCurve=DataInterface() {propertyValues {
         AlphaCurve {keys [<at 0 1>,<at 1 0>]}
         BlueCurve {keys [<at 0.17 15>,<at 0.2011 2>,<at 0.33 1.35>,<at 0.85 1.1>]}
         GreenCurve {keys [<at 0.17 7.6606>,<at 0.2011 0.722>,<at 0.33 0.2547>,<at 0.85 0.3056>]}
         RedCurve {keys [<at 0.17 0>,<at 0.2011 0>,<at 0.33 0>,<at 0.85 0>]}
        }})
     ))
    ScaleA=true
    ScaleRGB=false
   ) {enabled=false}
   /SetVariables(
    Particles.Color2=&ColorFromCurve(ColorCurve=DataInterface() {propertyValues {
       AlphaCurve {keys [<at 0 1 RCIM_Cubic>,<at 1 0 RCIM_Cubic>]}
       BlueCurve {keys [<at 0 1 RCIM_Cubic>,<at 1 1 RCIM_Cubic>]}
       GreenCurve {keys [<at 0 0.0423 RCIM_Cubic>,<at 1 1 RCIM_Cubic>]}
       RedCurve {keys [<at 0 0 RCIM_Cubic>,<at 1 1 RCIM_Cubic>]}
      }})
    Particles.SpriteSize2=&Multiply_Vector2D_ByFloat(Float=1, Vector2D=$Particles.SpriteSize)
   ) {enabled=false, isSetParameters=true}
   &DynamicMaterialParameters(Module.Index0Float0Write=true, Module.Index0Float1Write=true, Module.Index0Float2Write=true, Module.Index0Float3Write=true)
  ]
  rendererClasses [&NiagaraSpriteRendererProperties]
  renderers [
   {rendererClass=&NiagaraSpriteRendererProperties}
   {rendererClass=&NiagaraSpriteRendererProperties, rendererIndex=1}
  ]
 }
}
summary {
 systemName=NS_Emote_PolarityWinn_Pulse
 emitters [
  {bEnabled=false, emitterName=BurstRing001, rendererClasses [&NiagaraMeshRendererProperties]}
  {bEnabled=false, emitterName=Sparks, rendererClasses [&NiagaraSpriteRendererProperties]}
  {
   bEnabled=false
   emitterName=Flare
   rendererClasses [&NiagaraSpriteRendererProperties, &NiagaraLightRendererProperties]
  }
  {bEnabled=false, emitterName=sparkles001, rendererClasses [&NiagaraSpriteRendererProperties]}
  {emitterName=BurstRing, rendererClasses [&NiagaraMeshRendererProperties]}
  {emitterName=Sparks001, rendererClasses [&NiagaraSpriteRendererProperties]}
  {emitterName=Flare001, rendererClasses [&NiagaraSpriteRendererProperties, &NiagaraLightRendererProperties]}
  {emitterName=Squares, rendererClasses [&NiagaraSpriteRendererProperties]}
 ]
 userVariables [
  User.bIsADS {type=NiagaraBool, defaultValue=true}
  User.bIsFrontEnd {type=NiagaraBool}
  User.bIsFrontEndPreview {type=NiagaraBool}
  User.Color {type=LinearColor, description="", defaultValue {struct=LinearColor, value=<rgba 16.9886 0 20 0.5>}}
  User.DebugReactivity {type=NiagaraBool, defaultValue=true}
  User.Duration {type=NiagaraFloat, defaultValue=5, description=""}
  User.FEOffset {type=NiagaraFloat, defaultValue=10, description=""}
  User.FEScale {type=NiagaraFloat, defaultValue=0.6, description=""}
  User.FrontEndLooping {type=NiagaraFloat, defaultValue=0, description=""}
  User.Hit {type=NiagaraBool, defaultValue=true}
  User.Hitted {type=NiagaraFloat, defaultValue=0, description=""}
 ]
}
