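# Niagara effect for the ground soundwave of the PolarityWinn emote, built as a ripple lead-in followed by an impact burst.
# Mesh ripples pushed downward and flat ground sprite ripples spawn briefly at a steady rate, cycling partition-indexed magenta, blue and cyan colors.
# Around 0.9 seconds in, bursts fire an energy pulse mesh, light, rising smoke, vortex-pulled square sparks and swirling wind-streak meshes.
# Curves drive expansion, fades and dynamic material parameters, while time-of-day scalars adjust ripple, pulse and spark intensity.
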NiagaraSystem /BRCosmetics/Animation/Game/MainPlayer/Emotes/Polaritywinn/FX/NS_Emote_PolarityWinn_Soundwave_ground
dependencies {
 usedDataInterfaces [
  &NiagaraDataInterfaceArrayColor
  &NiagaraDataInterfaceCurve
  &NiagaraDataInterfaceColorCurve
  &NiagaraDataInterfaceVector2DCurve
  &NiagaraDataInterfaceArrayFloat
 ]
 usedDynamicInputs [
  &MakeCustomFloatFromBool
  /Niagara/DynamicInputs/Lerp/SmoothLerpOverTime_Float
  &FN_FrontEndFloatModifier
  &RemapRange
 ]
 usedModules [
  &self:SystemSpawnScript.SystemScriptSource.SystemScriptGraph.NiagaraNodeAssignment_1.SetVariables_#1
  &self:SystemSpawnScript.SystemScriptSource.SystemScriptGraph.NiagaraNodeAssignment_3.SetVariables_#2
  &SystemState
  &self:SystemSpawnScript.SystemScriptSource.SystemScriptGraph.NiagaraNodeAssignment_0.SetVariables_#3
  &self:SystemSpawnScript.SystemScriptSource.SystemScriptGraph.NiagaraNodeAssignment_2.SetVariables_#4
  &self:SystemSpawnScript.SystemScriptSource.SystemScriptGraph.NiagaraNodeAssignment_4.SetVariables_#5
 ]
}
emitters {
 Energy_Pulse {
  emitterUpdateScript [
   &EmitterState(Life Cycle Mode=::NewEnumerator1, Loop Behavior=::NewEnumerator1, Loop Duration=1, UseLoopDelay=false)
   &SpawnBurst_Instantaneous(Spawn Time=0.93, Use Spawn Probability=false, UseLoopCountLimit=false)
  ]
  particleSpawnScript [
   &InitializeParticle(
    Color=<rgba 0.0249 0.113 1 1>
    Lifetime=0.45
    Mesh Scale Mode=::NewEnumerator4
    Position Offset=<0 0 25>
    UsePositionOffset=true
    Write Material Random=true
   )
   &InitialMeshOrientation(Mesh Orientation Mode=::NewEnumerator2, Rotation=<1.5 0 0>, Use Rotation Vector=true)
  ]
  particleUpdateScript [
   &ParticleState()
   &ScaleMeshSize(Scale Factor=&Multiply_VectorByFloat(
     Vector=<1 1 1>
     Float=&FloatFromCurve(FloatCurve=DataInterface() {propertyValues {Curve {keys [<at 0 1>,<at 0.5 2>]}}})
    ))
   &ScaleColor(
    Scale Alpha=&FloatFromCurve(
     FloatCurve=DataInterface() {propertyValues {Curve {keys [
         {arriveTangent=26.1682, interpMode=RCIM_Cubic, leaveTangent=26.1682, tangentMode=RCTM_User, time=0, value=1}
         {arriveTangent=-3.7386, interpMode=RCIM_Cubic, leaveTangent=-3.7386, tangentMode=RCTM_User, time=0.1, value=1}
         {tangentMode=RCTM_User, time=1, value=0}
        ]}}}
     Scale Curve=&FNTimeOfDayScalar(Night Scale=2, Morning Scale=3, Day Scale=4, Evening Scale=3)
    )
    ScaleA=true
    ScaleRGB=true
   )
   &FN_Mesh_Attributes(
    Color=&MakeLinearColorFromVectorAndFloat(
     "Vector (RGB)"=<1.2 1.2 1.2>
     "Float (Alpha)"=&FloatFromCurve(
      FloatCurve=DataInterface() {propertyValues {Curve {keys [
          <at 0 0 RCIM_Cubic>
          {arriveTangent=0.1095, interpMode=RCIM_Cubic, leaveTangent=0.1095, tangentMode=RCTM_User, time=0.2, value=1}
          {arriveTangent=-1.5546, interpMode=RCIM_Cubic, leaveTangent=-1.5546, tangentMode=RCTM_User, time=0.8, value=-0.05}
         ]}}}
      Scale Curve=0.5
     )
    )
    Create Color Attribute=true
    Create Orientation Attribute=true
    Create Scale Attribute=true
    Orientation:Quat4f=&OrientMeshToVector(Look At Direction=<1 0 0>, Reference Axis Vector=<1 0 0>)
    Scale=&Multiply_Vector(A=$Particles.Scale, B=<0.5 0.5 0.5>)
    Write to Module Namespace=true
   )
   /SetVariables(Particles.WindFlowDynmat:Vector4f=&MakeVector4(X=1, Y=1, Z=0, W=0)) {isSetParameters=true}
  ]
  rendererClasses [&NiagaraMeshRendererProperties]
  renderers [
   {rendererClass=&NiagaraMeshRendererProperties, rendererIndex=0}
   {rendererClass=&NiagaraMeshRendererProperties, rendererIndex=1}
  ]
 }
 Ground_ripples001 {
  emitterUpdateScript [
   &EmitterState(Life Cycle Mode=::NewEnumerator1, Loop Behavior=::NewEnumerator1, Loop Duration=0.65, UseLoopDelay=false)
   &SpawnRate(SpawnRate=6, Use Spawn Probability=false)
  ]
  particleSpawnScript [
   &PartitionParticles(Particles Per Partition=5, Partition Mode:ENiagara_PartitionMode_v2=::NewEnumerator1)
   &InitializeParticle(
    Color=&SelectLinearColorFromArray(
     Array Sampling Mode=::NewEnumerator1
     Direct Array Index=$Particles.PartitionParticles.Index In Partition
     Color Selection Array=DataInterface() {propertyValues {ColorData [<rgba 1.0237 0.195 1.3 0.3>, <rgba 1.1813 0.225 1.5 0.8>, <rgba 0 0.2667 0.4 0.5623>, <rgba 0 0.6823 1.5 0.3>, <rgba 1.05 0 1.5 0.3>]}}
     Random Seed=?
    )
    Lifetime=0.55
    Mesh Scale Mode=::NewEnumerator3
    Mesh Uniform Scale=0.5
    Position Offset=<0 0 2>
    Sprite Size=&Vector2DFromFloat(Value=300)
    Sprite Size Mode=::NewEnumerator4
    UsePositionOffset=true
    Write Material Random=true
   )
   &InitialMeshOrientation(
    Mesh Orientation Mode=::NewEnumerator1
    Rotation=&RandomRangeVector(Minimum=<0 0 0>, Maximum=<0 0 1>)
    Use Rotation Vector=true
   )
   /SetVariables(Particles.SpriteFacing=<0.02 0 1>) {isSetParameters=true}
  ]
  particleUpdateScript [
   &ParticleState()
   &ScaleSpriteSize(Scale Factor=&Vector2DFromFloat(Value=&FloatFromCurve(FloatCurve=DataInterface() {propertyValues {Curve {keys [
          {arriveTangent=0.6835, interpMode=RCIM_Cubic, leaveTangent=0.6835, tangentMode=RCTM_User, time=0, value=0.5706}
          {arriveTangentWeight=0.4822, interpMode=RCIM_Cubic, tangentMode=RCTM_User, time=1.0011, value=1}
         ]}}})))
   &UpdateMeshOrientation(Rotation Vector=<0 0 1>)
   &ScaleMeshSize(Scale Factor=&VectorFromFloat(Value=&FloatFromCurve(FloatCurve=DataInterface() {propertyValues {Curve {keys [
          {arriveTangent=0.0505, interpMode=RCIM_Cubic, leaveTangent=0.0505, tangentMode=RCTM_User, time=0, value=0.3}
          {arriveTangent=1.3027, interpMode=RCIM_Cubic, leaveTangent=1.3027, tangentMode=RCTM_User, time=1, value=3}
         ]}}})))
   &ScaleColor(
    Scale Alpha=&FloatFromCurve(
     FloatCurve=DataInterface() {propertyValues {Curve {keys [{interpMode=RCIM_Cubic, leaveTangent=-3.3902, tangentMode=RCTM_Break, time=0, value=1}, {tangentMode=RCTM_User, time=1, value=0}]}}}
     Scale Curve=&FNTimeOfDayScalar(Night Scale=1, Morning Scale=0.5, Day Scale=0.5, Evening Scale=0.5)
    )
    Scale RGB=&VectorFromFloat(Value=&FloatFromCurve(
      FloatCurve=DataInterface() {propertyValues {Curve {
         defaultValue=1
         keys [{interpMode=RCIM_Cubic, tangentMode=RCTM_User, time=0, value=1}, {interpMode=RCIM_Cubic, tangentMode=RCTM_User, time=1, value=0}]
        }}}
      Scale Curve=&FNTimeOfDayScalar(Night Scale=1, Morning Scale=1, Day Scale=2, Evening Scale=1)
     ))
    ScaleA=true
    ScaleRGB=true
   ) {module=ScaleColor003}
   &Drag(Drag=7, Use Linear Drag=true, Use Rotational Drag=true)
   &SolveForcesAndVelocity(Clamp Velocity=false, Limit Acceleration=false)
   &DynamicMaterialParameters(
    Index 0 Param 1=&FloatFromCurve(FloatCurve=DataInterface() {propertyValues {Curve {keys [<at 0 1.5364>,<at 1 2.501>]}}})
    Module.Param0WriteEnabled=true
    Module.Param1WriteEnabled=false
    Module.Param2WriteEnabled=false
    Module.Param3WriteEnabled=false
   ) {module=DynamicMaterialParameters001}
  ]
  rendererClasses [&NiagaraSpriteRendererProperties]
  renderers [{rendererClass=&NiagaraSpriteRendererProperties}]
 }
 Light {
  emitterUpdateScript [
   &EmitterState()
   &SpawnBurst_Instantaneous(Spawn Time=0.93, Use Spawn Probability=false, UseLoopCountLimit=false)
  ]
  particleSpawnScript [
   &InitializeParticle(
    Color=<rgba 0.6771 0.6458 1 1>
    Lifetime=0.5
    Mass Mode=::NewEnumerator1
    Mesh Scale Mode=::NewEnumerator3
    Mesh Uniform Scale=0.2
    Position=&SimulationPosition()
    Position Mode=::NewEnumerator1
    Position Offset=$System.SystemOffset
    Sprite Rotation Mode=::NewEnumerator2
    Sprite Size=&Vector2DFromFloat(Value=&Multiply_Float(A=750, B=1))
    Sprite Size Mode=::NewEnumerator4
    UsePositionOffset=true
    Write Material Random=true
   ) {module=InitializeParticle001}
   &CameraOffset(Camera Offset Amount=75)
  ]
  particleUpdateScript [
   &ParticleState()
   &SolveForcesAndVelocity(Clamp Velocity=false, Limit Acceleration=false)
   &ScaleSpriteSize(Scale Factor=&Multiply_Vector2D_ByFloat(
     Float=&FloatFromCurve(
      FloatCurve=DataInterface() {propertyValues {Curve {keys [
          {arriveTangent=0.7585, interpMode=RCIM_Cubic, leaveTangent=0.7585, tangentMode=RCTM_User, time=0, value=0.4}
          {interpMode=RCIM_Cubic, tangentMode=RCTM_User, time=1, value=1}
         ]}}}
      Scale Curve=0.1
     )
     Vector2D=<1 1>
    ))
   &SpriteRotationRate(Rotation Rate=&UniformRangedFloat(Minimum=-50, Maximum=50))
   /SetVariables(Particles.CentralLightColor:LinearColor=&MakeLinearColorFromVectorAndFloat(
     "Vector (RGB)"=&MakeVectorFromLinearColorRGB(Color=&ColorFromCurve(ColorCurve=DataInterface() {propertyValues {
         AlphaCurve {keys [<at 0 1 RCIM_Cubic>,<at 1 0 RCIM_Cubic>]}
         BlueCurve {keys [<at 0 2>, {interpMode=RCIM_Cubic, leaveTangent=1.3364, time=1, value=1}]}
         GreenCurve {keys [<at 0 0.5272>, {interpMode=RCIM_Cubic, leaveTangent=-5.7627, time=1, value=0.3351}]}
         RedCurve {keys [<at 0 0.0903>, {interpMode=RCIM_Cubic, leaveTangent=-13.364, time=1, value=0.8022}]}
        }}))
     "Float (Alpha)"=&FloatFromCurve(
      FloatCurve=DataInterface() {propertyValues {Curve {keys [
          <at 0 0 RCIM_Cubic>
          {arriveTangent=0.3854, interpMode=RCIM_Cubic, leaveTangent=0.3854, tangentMode=RCTM_User, time=0.1, value=1}
          <at 1 0 RCIM_Cubic>
         ]}}}
      Scale Curve=0.5
     )
    )) {isSetParameters=true}
   &ScaleColor(
    Scale Alpha=&FloatFromCurve(
     FloatCurve=DataInterface() {propertyValues {Curve {keys [
         {arriveTangent=6.4019, interpMode=RCIM_Cubic, leaveTangent=6.4019, tangentMode=RCTM_User, time=0, value=1}
         {interpMode=RCIM_Cubic, leaveTangent=0.002, time=1, value=0}
        ]}}}
     Scale Curve=0.02
    )
    Scale RGB=&MakeVectorFromLinearColorRGB(Color=$Particles.CentralFlareColor)
    ScaleA=true
    ScaleRGB=true
   )
   &FN_Light_Attributes(
    Color=$Particles.CentralLightColor
    Create Color Attribute=true
    Create Exponent Attribute=true
    Create Radius Attribute=true
    Exponent=1
    Radius=&Multiply_Float(
     A=&FloatFromCurve(
      FloatCurve=DataInterface() {propertyValues {Curve {keys [
          {arriveTangent=2.3282, interpMode=RCIM_Cubic, leaveTangent=2.3282, tangentMode=RCTM_User, time=0, value=0}
          {interpMode=RCIM_Cubic, tangentMode=RCTM_User, time=1, value=1}
         ]}}}
      Scale Curve=250
     )
     B=0.5
    )
    Scale Color=&FloatFromCurve(
     FloatCurve=DataInterface() {propertyValues {Curve {keys [
         {interpMode=RCIM_Cubic, tangentMode=RCTM_User, time=0, value=1}
         {arriveTangent=-3.4524, interpMode=RCIM_Cubic, leaveTangent=-3.4524, tangentMode=RCTM_User, time=1, value=0}
        ]}}}
     Scale Curve=7
    )
   )
   &DynamicMaterialParameters(
    Index 1 Param 1=0.3
    Index 1 Param 4=&Lerp_Float(A=0, B=1, Alpha=0.5)
    Module.Index1Float0Write=true
    Module.Index1Float1Write=true
    Module.Index1Float2Write=true
    Module.Index1Float3Write=true
    Module.Index3Float0Write=true
    Module.Index3Float1Write=true
    Module.Index3Float2Write=true
    Module.Index3Float3Write=true
    Write Parameter Index 0=::NewEnumerator0
    Write Parameter Index 1=::NewEnumerator1
    Write Parameter Index 3=::NewEnumerator1
   )
   &FN_Sprite_Attributes(
    Create Sprite Rotation Attribute=true
    Create SpriteSize Attribute=true
    SpriteSize=&Multiply_Vector2D(A=$Particles.CentralFlareSize, B=<2 0.15>)
    Write to Module Namespace=true
   )
  ]
  rendererClasses [&NiagaraLightRendererProperties]
  renderers [{rendererClass=&NiagaraLightRendererProperties, rendererIndex=0}]
 }
 Multiple_ripples001 {
  emitterUpdateScript [
   &EmitterState(Life Cycle Mode=::NewEnumerator1, Loop Behavior=::NewEnumerator1, Loop Duration=0.65, UseLoopDelay=false)
   &SpawnRate(SpawnRate=10, Use Spawn Probability=false)
  ]
  particleSpawnScript [
   &PartitionParticles(Particles Per Partition=5, Partition Mode:ENiagara_PartitionMode_v2=::NewEnumerator1)
   &InitializeParticle(
    Color=&SelectLinearColorFromArray(
     Array Sampling Mode=::NewEnumerator1
     Direct Array Index=$Particles.PartitionParticles.Index In Partition
     Direct Array Mode=::NewEnumerator1
     Color Selection Array=DataInterface() {propertyValues {ColorData [<rgba 1.0237 0.195 1.3 0.3>, <rgba 0.227 0.107 0.4072 0.5>, <rgba 0 0.6823 1.5 0.7>, <rgba 0 1.5 1.2526 0.9>]}}
     Random Seed=?
    )
    Lifetime=0.7
    Mesh Scale Mode=::NewEnumerator3
    Mesh Uniform Scale=0.5
    Position Offset=<0 0 30>
    Sprite Size Mode=::NewEnumerator3
    Uniform Sprite Size=100
    UsePositionOffset=true
    Write Material Random=true
   )
   &AddVelocity(Velocity=<0 0 -250>)
   &InitialMeshOrientation(
    Mesh Orientation Mode=::NewEnumerator1
    Rotation=&RandomRangeVector(Minimum=<0 0 0>, Maximum=<0 0 1>)
    Use Rotation Vector=true
   )
  ]
  particleUpdateScript [
   &ParticleState()
   &UpdateMeshOrientation(Rotation Vector=<0 0 1>)
   &ScaleMeshSize(Scale Factor=&VectorFromFloat(Value=&FloatFromCurve(FloatCurve=DataInterface() {propertyValues {Curve {keys [
          {arriveTangent=0.0505, interpMode=RCIM_Cubic, leaveTangent=0.0505, tangentMode=RCTM_User, time=0, value=0.4}
          {arriveTangent=1.3027, interpMode=RCIM_Cubic, leaveTangent=1.3027, tangentMode=RCTM_User, time=1, value=3}
         ]}}})))
   &ScaleColor(
    Scale Alpha=&FloatFromCurve(
     FloatCurve=DataInterface() {propertyValues {Curve {keys [{interpMode=RCIM_Cubic, leaveTangent=-3.3902, tangentMode=RCTM_Break, time=0, value=1}, {tangentMode=RCTM_User, time=1, value=0}]}}}
     Scale Curve=&FNTimeOfDayScalar(Night Scale=1, Morning Scale=1, Evening Scale=1)
    )
    Scale RGB=&VectorFromFloat(Value=&FloatFromCurve(
      FloatCurve=DataInterface() {propertyValues {Curve {
         defaultValue=1
         keys [{interpMode=RCIM_Cubic, tangentMode=RCTM_User, time=0, value=1}, {interpMode=RCIM_Cubic, tangentMode=RCTM_User, time=0.8, value=0}]
        }}}
      Scale Curve=&MakeFloatFromVector(
       Vector=&VectorFromFloat(Value=&FNTimeOfDayScalar(Night Scale=2, Morning Scale=1, Day Scale=2, Evening Scale=2))
       Channel=::NewEnumerator0
      )
     ))
    ScaleA=true
    ScaleRGB=true
   ) {module=ScaleColor003}
   &ScaleColor(
    Linear Color Curve=DataInterface() {propertyValues {
      AlphaCurve {keys [<at 0 1>,<at 1 0>]}
      BlueCurve {keys [
        <at 0.0238 1.6785>
        {arriveTangent=-1.8038, interpMode=RCIM_Cubic, leaveTangent=-1.8038, time=0.1259, value=1}
        <at 0.4 1>
        <at 0.5655 1.6785>
        {arriveTangent=-1.5877, interpMode=RCIM_Cubic, leaveTangent=-1.5877, time=0.75, value=2}
        <at 1 0.9887>
       ]}
      GreenCurve {keys [
        <at 0.0238 0>
        {arriveTangent=0.8824, interpMode=RCIM_Cubic, leaveTangent=0.8824, time=0.1259, value=0.9157}
        <at 0.4 0.3319>
        <at 0.5655 0>
        {arriveTangent=1.4984, interpMode=RCIM_Cubic, leaveTangent=1.4984, time=0.75, value=0.7287}
        <at 1 0.651>
       ]}
      RedCurve {keys [
        <at 0.0238 0.8445>
        {arriveTangent=-2.245, interpMode=RCIM_Cubic, leaveTangent=-2.245, time=0.1259, value=0}
        <at 0.4 0>
        <at 0.5655 0.8445>
        {arriveTangent=0.3579, interpMode=RCIM_Cubic, leaveTangent=0.3579, time=0.75, value=1.7489}
        <at 1 1>
       ]}
     }}
    Scale Mode=::NewEnumerator2
   ) {module=ScaleColor001}
   &ScaleColor(
    Scale Alpha=&FloatFromCurve(FloatCurve=DataInterface() {propertyValues {Curve {keys [
         {arriveTangent=9.6668, interpMode=RCIM_Cubic, leaveTangent=9.6668, tangentMode=RCTM_User, time=0, value=0}
         <at 0.25 1 RCIM_Cubic>
         {arriveTangent=-4.1315, interpMode=RCIM_Cubic, leaveTangent=-4.1315, tangentMode=RCTM_User, time=1, value=0}
        ]}}})
    ScaleA=true
    ScaleRGB=true
   ) {enabled=false}
   &ScaleColor(
    Scale Alpha=0.7
    Scale RGB=&VectorFromFloat(Value=&FNTimeOfDayScalar(Night Scale=1, Morning Scale=1, Evening Scale=1))
    ScaleA=false
    ScaleRGB=true
   ) {enabled=false, module=ScaleColor002}
   &Drag(Drag=7, Use Linear Drag=true, Use Rotational Drag=true)
   &SolveForcesAndVelocity(Clamp Velocity=false, Limit Acceleration=false)
   &DynamicMaterialParameters(
    Index 0 Param 1=&FloatFromCurve(FloatCurve=DataInterface() {propertyValues {Curve {keys [<at 0.1 5 RCIM_Cubic>,<at 0.95 0 RCIM_Cubic>]}}})
    Index 0 Param 2=&FloatFromCurve(FloatCurve=DataInterface() {propertyValues {Curve {keys [<at 0 5 RCIM_Cubic>,<at 0.4 36 RCIM_Cubic>]}}})
    Index 0 Param 3=1
    Index 0 Param 4=&FloatFromCurve(FloatCurve=DataInterface() {propertyValues {Curve {keys [<at 0 1 RCIM_Cubic>,<at 1 0 RCIM_Cubic>]}}})
    Module.Param0WriteEnabled=true
    Module.Param1WriteEnabled=true
    Module.Param2WriteEnabled=true
    Module.Param3WriteEnabled=true
   )
   &KillParticles(Kill Particles=&SetBoolByFloatComparison(A=$Particles.NormalizedAge, B=0.7)) {enabled=false}
  ]
  rendererClasses [&NiagaraMeshRendererProperties]
  renderers [
   {rendererClass=&NiagaraMeshRendererProperties, rendererIndex=0}
   {rendererClass=&NiagaraMeshRendererProperties, rendererIndex=1}
  ]
 }
 Smoke {
  emitterUpdateScript [
   &EmitterState()
   &SpawnBurst_Instantaneous(
    Spawn Count=&RandomRangeInt(Minimum=15, Maximum=20)
    Spawn Time=0.94
    Use Spawn Probability=false
    UseLoopCountLimit=false
   )
  ]
  particleSpawnScript [
   &InitializeParticle(
    Color=<rgba 0.6376 0.5841 0.5333 0.5>
    Lifetime Max=1
    Lifetime Min=0.7
    Lifetime Mode=::NewEnumerator1
    Mesh Scale Mode=::NewEnumerator3
    Mesh Uniform Scale=150
    Position Offset=<0 0 -40>
    Sprite Rotation Mode=::NewEnumerator2
    Sprite Size Mode=::NewEnumerator1
    Uniform Sprite Size Max=12
    Uniform Sprite Size Min=10
    UsePositionOffset=true
    Write Material Random=true
   ) {module=InitializeParticle002}
   &ShapeLocation(
    MaskSpawnGroup=false
    Sphere Distribution=::NewEnumerator1
    Sphere Radius=&RandomRangeFloat(Minimum=50, Maximum=60)
    U Position=&ReturnNormalizedExecIndex()
    UseOwnerScale=true
    V Position=&RandomRangeFloat(Minimum=0.5, Maximum=0.55)
   )
   &AddVelocity(
    Velocity=&Multiply_VectorByFloat(Vector=$Particles.ShapeLocation.ShapeVector, Float=0.02)
    Velocity Speed Scale=&RandomRangeFloat(Minimum=250, Maximum=350)
   )
  ]
  particleUpdateScript [
   &ParticleState()
   &GravityForce(Gravity=<0 0 200>)
   &Drag(Drag=7, Use Linear Drag=true, Use Rotational Drag=false)
   &SolveForcesAndVelocity(Clamp Velocity=false, Limit Acceleration=false)
   &ScaleColor(
    Scale Alpha=&FloatFromCurve(FloatCurve=DataInterface() {propertyValues {Curve {keys [
         {arriveTangent=7.3326, interpMode=RCIM_Cubic, leaveTangent=7.3326, tangentMode=RCTM_User, time=0, value=0}
         {arriveTangent=0.0071, interpMode=RCIM_Cubic, leaveTangent=0.0071, tangentMode=RCTM_User, time=0.4, value=1}
         <at 1 0 RCIM_Cubic>
        ]}}})
    Scale RGB=&MakeVectorFromLinearColorRGB(Color=&ColorFromCurve(ColorCurve=DataInterface() {propertyValues {
        AlphaCurve {keys [<at 0 1 RCIM_Cubic>,<at 1 0 RCIM_Cubic>]}
        BlueCurve {keys [<at 0 1.2 RCIM_Cubic>,<at 0.4269 1 RCIM_Cubic>]}
        GreenCurve {keys [<at 0 1.2 RCIM_Cubic>,<at 0.4269 1 RCIM_Cubic>]}
        RedCurve {keys [<at 0 1.2 RCIM_Cubic>,<at 0.4269 1 RCIM_Cubic>]}
       }}))
    ScaleA=true
    ScaleRGB=true
   ) {module=ScaleColor001}
   &DynamicMaterialParameters(
    Index 0 Param 1=&Add_Float(
     A=&FloatFromCurve(FloatCurve=DataInterface() {propertyValues {Curve {keys [{arriveTangent=2.1549, time=0, value=0.8}, <at 1 1>]}}})
     B=&Multiply_Float(A=$Particles.Initial.MaterialRandom, B=0.5)
    )
    Index 1 Param 1=1
    Index 1 Param 2=1
    Index 1 Param 3=1
    Index 1 Param 4=1
    Index 2 Param 1=&FloatFromCurve(FloatCurve=DataInterface() {propertyValues {Curve {keys [
         {arriveTangent=1.8288, interpMode=RCIM_Cubic, leaveTangent=1.8288, tangentMode=RCTM_User, time=0, value=0.3}
         {arriveTangent=-0.0004, interpMode=RCIM_Cubic, leaveTangent=-0.0004, tangentMode=RCTM_User, time=1, value=1}
        ]}}})
    Index 2 Param 2=&RandomRangeFloat(Minimum=0.5, Maximum=1)
    Index 2 Param 3=4
    Index 2 Param 4=&RandomRangeFloat(Minimum=0.8, Maximum=1.2)
    Module.Param0WriteEnabled=true
    Module.Param0WriteEnabled1=true
    Module.Param0WriteEnabled2=true
    Module.Param0WriteEnabled3=true
    Module.Param1WriteEnabled=true
    Module.Param1WriteEnabled1=true
    Module.Param1WriteEnabled2=true
    Module.Param1WriteEnabled3=true
    Module.Param2WriteEnabled=true
    Module.Param2WriteEnabled1=true
    Module.Param2WriteEnabled2=true
    Module.Param2WriteEnabled3=true
    Module.Param3WriteEnabled=true
    Module.Param3WriteEnabled1=true
    Module.Param3WriteEnabled2=true
    Module.Param3WriteEnabled3=true
    Write Parameter Index 1=true
    Write Parameter Index 2=true
    Write Parameter Index 3=true
   ) {module=DynamicMaterialParameters002}
   &ScaleSpriteSize(Scale Factor=&Vector2DFromFloat(Value=&FloatFromCurve(FloatCurve=DataInterface() {propertyValues {Curve {keys [
          {arriveTangent=87.4549, interpMode=RCIM_Cubic, leaveTangent=87.4549, tangentMode=RCTM_User, time=0, value=1}
          {arriveTangent=4.8897, interpMode=RCIM_Cubic, leaveTangent=4.8897, tangentMode=RCTM_User, time=0.1, value=8}
          {interpMode=RCIM_Cubic, leaveTangent=-6.7553, time=1.0011, value=10}
         ]}}})))
   &CameraOffset(Camera Offset Amount=50)
   &SpriteRotationRate(Rotation Rate=&RandomRangeFloat(Minimum=-90, Maximum=90))
   &SpriteRotationRate(Rotation Rate=&Multiply_Float(A=90, B=&RandomRangeFloat(Minimum=-1, Maximum=1))) {module=SpriteRotationRate001}
  ]
  rendererClasses [&NiagaraSpriteRendererProperties]
  renderers [
   {rendererClass=&NiagaraSpriteRendererProperties}
   {rendererClass=&NiagaraSpriteRendererProperties, rendererIndex=1}
  ]
 }
 Square_Sprites {
  emitterUpdateScript [
   &EmitterState()
   &SpawnBurst_Instantaneous(Spawn Count=15, Spawn Time=0.9, Use Spawn Probability=false, UseLoopCountLimit=false)
   &SpawnBurst_Instantaneous(Spawn Count=15, Spawn Group=1, Spawn Time=0.9, Use Spawn Probability=false, UseLoopCountLimit=false) {module=SpawnBurst_Instantaneous001}
  ]
  particleSpawnScript [
   &InitializeParticle(
    Color=&SelectLinearColorFromArray(
     Array Sampling Mode=::NewEnumerator0
     Direct Array Index=?
     Color Selection Array=DataInterface() {propertyValues {ColorData [<rgba 1.0237 0.195 1.3 3>, <rgba 1.1813 0.225 1.5 3>, <rgba 0 0.2667 0.4 3>, <rgba 0 0.6823 1.5 3>, <rgba 1.05 0 1.5 3>]}}
     Random Seed=0
    )
    Lifetime=&RandomRangeFloat(Minimum=0.5, Maximum=0.7)
    Sprite Size=&Multiply_Vector2D_ByFloat(Float=&RandomRangeFloat(Minimum=2, Maximum=9), Vector2D=<1 3>)
    Sprite Size Mode=::NewEnumerator4
    UsePositionOffset=false
    Write Material Random=true
   )
   /SetVariables(Particles.VisibilityTag=&SpawnGroup()) {isSetParameters=true}
   &ShapeLocation(
    MaskSpawnGroup=false
    Offset=<0 0 -25>
    Offset Mode=::NewEnumerator0
    Sphere Distribution=::NewEnumerator1
    Sphere Radius=1
    U Position=&Lerp_Float(A=0, B=1, Alpha=&ReturnNormalizedExecIndex())
    UseOwnerScale=true
    V Position=&RandomRangeFloat(Minimum=0.5, Maximum=0.6)
   )
   &AddVelocity(
    Velocity=$Particles.ShapeLocation.ShapeNormal
    Velocity Speed Scale=&RandomRangeFloat(Minimum=450, Maximum=1000)
   )
   &AddVelocity(Velocity=<0 0 1>, Velocity Speed Scale=&RandomRangeFloat(Minimum=0, Maximum=550)) {module=AddVelocity001}
  ]
  particleUpdateScript [
   &ParticleState()
   &InheritSourceMovement(
    Applied Position Scale=&VectorFromFloat(Value=&Multiply_Float(
      A=1
      B=&FloatFromCurve(FloatCurve=DataInterface() {propertyValues {Curve {keys [
           {arriveTangent=-1.9407, interpMode=RCIM_Cubic, leaveTangent=-1.9407, tangentMode=RCTM_User, time=0, value=1}
           {interpMode=RCIM_Cubic, tangentMode=RCTM_User, time=1, value=0.5}
          ]}}})
     ))
    ApplyForce=false
    ApplyPosition=true
    ApplyPositionFalloff=false
   )
   &Drag(Drag=&RandomRangeFloat(Minimum=3, Maximum=5), Use Linear Drag=true, Use Rotational Drag=true)
   &ScaleColor(
    Scale Alpha=&FloatFromCurve(FloatCurve=DataInterface() {propertyValues {Curve {keys [
         {arriveTangent=6.4019, interpMode=RCIM_Cubic, leaveTangent=6.4019, tangentMode=RCTM_User, time=0, value=0}
         <at 0.1 0.7 RCIM_Cubic>
         <at 1 0 RCIM_Cubic>
        ]}}})
    Scale RGB=&VectorFromFloat(Value=&FNTimeOfDayScalar(Night Scale=2, Morning Scale=2, Day Scale=3, Evening Scale=2))
    ScaleA=true
    ScaleRGB=true
   ) {module=ScaleColor003}
   &DynamicMaterialParameters(
    Index 0 Param 1=2
    Index 0 Param 3=1
    Index 3 Param 1=&Lerp_Float(
     A=0.2
     B=1
     Alpha=&FloatFromCurve(FloatCurve=DataInterface() {propertyValues {Curve {keys [<at 0 0>,<at 1 1>]}}})
    )
    Module.Index0Float0Write=true
    Module.Index0Float1Write=true
    Module.Index0Float2Write=true
    Module.Index0Float3Write=true
    Module.Index3Float0Write=true
    Module.Index3Float1Write=true
    Module.Index3Float2Write=true
    Module.Index3Float3Write=true
    Write Parameter Index 3=::NewEnumerator1
   )
   /SetVariables(Particles.SparkGlowColor:LinearColor=&MultiplyLinearColor(
     Linear Color=<rgba 0.1 0.1 0.1 0.05>
     Scale Factor=&MakeVector4_FromLinearColor(LinearColor=$Particles.Color)
    )) {isSetParameters=true}
   /SetVariables(Particles.SparkGlowSize:Vector2f=&Multiply_Vector2D(A=$Particles.SpriteSize, B=<5 2>)) {isSetParameters=true}
   /SetVariables(Particles.SparkGlowDynMat:Vector4f=<2 20 90 1>) {isSetParameters=true}
   &FN_Sprite_Attributes(
    Color=&MakeLinearColorFromVectorAndFloat(
     "Vector (RGB)"=&MakeVectorFromLinearColorRGB(Color=&Multiply_LinearColorByFloat(
       Linear Color=&ColorFromCurve(
        ColorCurve=DataInterface() {propertyValues {
          AlphaCurve {keys [<at 0 1 RCIM_Cubic>]}
          BlueCurve {keys [<at 0.3173 1 RCIM_Cubic>,<at 0.3548 1>,<at 0.6928 1>,<at 0.7245 0.6864>]}
          GreenCurve {keys [<at 0.3173 0.1714 RCIM_Cubic>,<at 0.3548 0.3565>,<at 0.6928 0.2891>,<at 0.7245 0.1519>]}
          RedCurve {keys [<at 0.3173 0.3432 RCIM_Cubic>,<at 0.3548 0>,<at 0.6928 0>,<at 0.7245 1>]}
         }}
        CurveIndex=&RandomRangeFloat(Minimum=0, Maximum=1)
       )
       Float=&Lerp_Float(
        A=1
        B=20
        Alpha=&FloatFromCurve(FloatCurve=DataInterface() {propertyValues {Curve {keys [
             {arriveTangent=-0.4493, interpMode=RCIM_Cubic, leaveTangent=-0.4493, tangentMode=RCTM_User, time=0, value=1}
             {interpMode=RCIM_Cubic, tangentMode=RCTM_User, time=0.7, value=0}
            ]}}})
       )
      ))
     "Float (Alpha)"=&FloatFromCurve(FloatCurve=DataInterface() {propertyValues {Curve {keys [
          {interpMode=RCIM_Cubic, tangentMode=RCTM_User, time=0, value=1}
          {arriveTangent=-2.0246, interpMode=RCIM_Cubic, leaveTangent=-2.0246, tangentMode=RCTM_User, time=1, value=0}
         ]}}})
    )
    Create Color Attribute=true
    Create SpriteSize Attribute=true
    SpriteSize=&Multiply_Vector2D_ByFloat(
     Float=&Multiply_Float(
      A=&RandomRangeFloat(Minimum=0.5, Maximum=1.5)
      B=&FloatFromCurve(FloatCurve=DataInterface() {propertyValues {Curve {keys [
           {interpMode=RCIM_Cubic, tangentMode=RCTM_User, time=0.2, value=1}
           {arriveTangent=-2.4619, interpMode=RCIM_Cubic, leaveTangent=-2.4619, tangentMode=RCTM_User, time=1, value=0}
          ]}}})
     )
     Vector2D=<5 5>
    )
    Write to Module Namespace=true
   )
   &VortexForce(
    Origin Pull Amount=750
    UseOriginOffset=false
    Vortex Force Amount=&RandomRangeFloat(Minimum=150, Maximum=1000)
   )
   &SolveForcesAndVelocity(Clamp Velocity=false, Limit Acceleration=false)
  ]
  rendererClasses [&NiagaraSpriteRendererProperties]
  renderers [{rendererClass=&NiagaraSpriteRendererProperties}]
 }
 wind_streaks {
  emitterUpdateScript [
   &EmitterState()
   &SpawnBurst_Instantaneous(Spawn Count=15, Spawn Group=2, Spawn Time=0.93, Use Spawn Probability=false, UseLoopCountLimit=false) {module=SpawnBurst_Instantaneous002}
  ]
  particleSpawnScript [
   /SetVariables(Particles.VisibilityTag=&SpawnGroup()) {isSetParameters=true}
   /SetVariables(Particles.RandomFloat=&UniformRangedFloat(Minimum=0, Maximum=3)) {isSetParameters=true}
   &InitializeParticle(
    Color=&RandomRangeLinearColor(Minimum=<rgba 0.5022 0.25 1 1>, Maximum=<rgba 0.1495 0.2101 0.776 1>)
    Lifetime=&MergeHelper_FloatSwitch(
     0=&RandomRangeFloat(Minimum=0.3, Maximum=0.45)
     1=&RandomRangeFloat(Minimum=0.3, Maximum=0.4)
     2=&RandomRangeFloat(Minimum=0.8, Maximum=1)
     3=?
     Index=&SpawnGroup()
     Num=::NewEnumerator1
    )
    Mesh Scale=&Multiply_VectorByFloat(Vector=&RandomRangeVector(Minimum=<1 1 1>, Maximum=<1.5 1.5 1>), Float=2)
    Mesh Scale Mode=::NewEnumerator4
    Position Offset=<0 0 10>
    Sprite Size=&RandomRangeVector2D(Minimum=<15 25>, Maximum=<20 50>)
    Sprite Size Mode=::NewEnumerator4
    UsePositionOffset=true
    Write Material Random=true
   ) {module=InitializeParticle001}
   &ShapeLocation(
    MaskSpawnGroup=false
    Offset=<0 0 -25>
    Offset Mode=::NewEnumerator0
    Sphere Distribution=::NewEnumerator1
    Sphere Radius=3.5
    U Position=&Lerp_Float(A=0, B=1, Alpha=&ReturnNormalizedExecIndex(Normalized Index Scale=2))
    UseOwnerScale=true
    V Position=&RandomRangeFloat(Minimum=0.5, Maximum=0.6)
   )
   &AddVelocity(
    Module.Use Falloff Distance=false
    Origin Offset=<0 0 -30>
    Override Seed=false
    Velocity Mode=::NewEnumerator1
    Velocity Speed=&RandomRangeFloat(Minimum=50, Maximum=100)
   )
   &InitialMeshOrientation(
    Mesh Orientation Mode=::NewEnumerator2
    Orientation Coordinate Space=::World
    Rotation=&Add_Vector(A=&RandomRangeVector(Minimum=<0.04 0.04 0>, Maximum=<-0.04 -0.04 1>), B=<0 0 0>)
    Use Rotation Vector=true
   )
   &SubUVAnimation(End Frame=3, SubUV Animation Mode:ENiagara_SubUVLookupMode=::NewEnumerator2)
   /SetVariables() {isSetParameters=true}
  ]
  particleUpdateScript [
   &ParticleState() {module=UpdateAge}
   &SolveForcesAndVelocity(Clamp Velocity=false, Limit Acceleration=false)
   &ScaleSpriteSize(Scale Factor=&Vector2DFromCurve(
     Vector2Curve=DataInterface() {propertyValues {XCurve {keys [<at 0.0036 1.6897>,<at 0.9967 0.0582>]}, YCurve {keys [<at 0 1>,<at 1.002 3.7227>]}}}
     ScaleCurve=<1 1.25>
    ))
   &ScaleSpriteSize(Scale Sprite Size Mode=::NewEnumerator0, Uniform Scale Factor=2) {module=ScaleSpriteSize001}
   &ScaleColor(
    Scale Alpha=&FloatFromCurve(
     FloatCurve=DataInterface() {propertyValues {Curve {keys [{interpMode=RCIM_Cubic, tangentMode=RCTM_Break, time=0, value=1}, {interpMode=RCIM_Cubic, tangentMode=RCTM_Break, time=1, value=0}]}}}
     Scale Curve=&RandomRangeFloat(Minimum=0.3, Maximum=0.5)
    )
    Scale RGB=&VectorFromFloat(Value=1)
    ScaleA=true
    ScaleRGB=true
   )
   &DynamicMaterialParameters(
    Index 0 Param 2=1
    Index 0 Param 3=1
    Index 0 Param 4=&FloatFromCurve(FloatCurve=DataInterface() {propertyValues {Curve {keys [<at 0.3 1>,<at 1 0.25>]}}})
    Index 3 Param 1=&FloatFromCurve(FloatCurve=DataInterface() {propertyValues {Curve {keys [
         {arriveTangent=1.526, interpMode=RCIM_Cubic, leaveTangent=1.526, tangentMode=RCTM_User, time=0, value=0}
         {interpMode=RCIM_Cubic, tangentMode=RCTM_User, time=1, value=1}
        ]}}})
    Index 3 Param 3=1
    Index 3 Param 4=1
    Module.Param0WriteEnabled=true
    Module.Param0WriteEnabled3=true
    Module.Param1WriteEnabled=true
    Module.Param1WriteEnabled3=true
    Module.Param2WriteEnabled=true
    Module.Param2WriteEnabled3=true
    Module.Param3WriteEnabled=true
    Module.Param3WriteEnabled3=true
    Write Parameter Index 3=true
   )
   &FN_Sprite_Attributes(
    Color=&MultiplyLinearColor(
     Linear Color=<rgba 0.2505 0.0437 1 1>
     Scale Factor=&MakeFloat4FromVectorAndFloat(
      "Vector (RGB)"=&MakeVectorFromLinearColorRGB(Color=&ColorFromCurve(ColorCurve=DataInterface() {propertyValues {
          AlphaCurve {keys [<at 0 1 RCIM_Cubic>]}
          BlueCurve {keys [<at 0.0426 10 RCIM_Cubic>, {arriveTangent=-12.0954, interpMode=RCIM_Cubic, leaveTangent=-12.0954, time=0.2988, value=5}, <at 0.8197 0.6>]}
          GreenCurve {keys [<at 0.0426 10 RCIM_Cubic>, {arriveTangent=-12.0954, interpMode=RCIM_Cubic, leaveTangent=-12.0954, time=0.2988, value=5}, <at 0.8197 0.6>]}
          RedCurve {keys [<at 0.0426 10 RCIM_Cubic>, {arriveTangent=-12.0954, interpMode=RCIM_Cubic, leaveTangent=-12.0954, time=0.2988, value=5}, <at 0.8197 0.6>]}
         }}))
      "Float (Alpha)"=&FloatFromCurve(FloatCurve=DataInterface() {propertyValues {Curve {keys [
           {arriveTangent=-2.2233, interpMode=RCIM_Cubic, leaveTangent=-2.2233, tangentMode=RCTM_User, time=0.1, value=1}
           {arriveTangent=-0.0075, interpMode=RCIM_Cubic, leaveTangent=-0.0075, tangentMode=RCTM_User, time=1, value=0}
          ]}}})
     )
    )
    Create Color Attribute=true
    Create SpriteSize Attribute=true
    SpriteSize=&Multiply_Vector2D(
     A=&Multiply_Vector2D_ByFloat(
      Float=&FloatFromCurve(FloatCurve=DataInterface() {propertyValues {Curve {keys [
           {arriveTangent=2.0201, interpMode=RCIM_Cubic, leaveTangent=2.0201, tangentMode=RCTM_User, time=0, value=0}
           {arriveTangent=-0.026, interpMode=RCIM_Cubic, leaveTangent=-0.026, tangentMode=RCTM_User, time=0.6, value=1}
          ]}}})
      Vector2D=&RandomRangeVector2D(Minimum=<100 250>, Maximum=<150 350>)
     )
     B=&Vector2DFromCurve(
      Vector2Curve=DataInterface() {propertyValues {
        XCurve {keys [
          {arriveTangent=-1.6611, interpMode=RCIM_Cubic, leaveTangent=-1.6611, tangentMode=RCTM_User, time=0.2, value=1.2}
          {arriveTangent=-0.1324, interpMode=RCIM_Cubic, leaveTangent=-0.1324, tangentMode=RCTM_User, time=1, value=0.7}
         ]}
        YCurve {keys [{interpMode=RCIM_Cubic, tangentMode=RCTM_User, time=0.4, value=1}]}
       }}
      ScaleCurve=<1 1>
     )
    )
    Write to Module Namespace=true
   )
   &ScaleMeshSize(Scale Factor=&Multiply_VectorByFloat(
     Vector=&MakeVector(
      X=1
      Y=1
      Z=&SelectFloatFromArray(
       Array Sampling Mode=::NewEnumerator0
       Direct Array Index=?
       Float Selection Array=DataInterface() {propertyValues {FloatData [1,-1]}}
       Random Seed=0
      )
     )
     Float=&Lerp_Float(
      A=0.2
      B=0.7
      Alpha=&FloatFromCurve(FloatCurve=DataInterface() {propertyValues {Curve {keys [
           {arriveTangent=3.9971, interpMode=RCIM_Cubic, leaveTangent=3.9971, tangentMode=RCTM_User, time=0, value=0}
           {interpMode=RCIM_Cubic, tangentMode=RCTM_User, time=0.86, value=1}
          ]}}})
     )
    ))
   &UpdateMeshOrientation(Rotation Rate=&RandomRangeFloat(Minimum=0.5, Maximum=0.6), Rotation Vector=<0 0 1>)
   &FN_Mesh_Attributes(
    Color=&MultiplyLinearColor(
     Linear Color=<rgba 0.7704 0.6365 1 1>
     Scale Factor=&MakeVector4FromVectorAndFloat(
      "Vector (RGB)"=&VectorFromFloat(Value=&Multiply_Float(
        A=&Lerp_Float(
         A=0.5
         B=0.75
         Alpha=&FloatFromCurve(FloatCurve=DataInterface() {propertyValues {Curve {keys [
              {interpMode=RCIM_Cubic, tangentMode=RCTM_User, time=0, value=0}
              {interpMode=RCIM_Cubic, tangentMode=RCTM_User, time=0.1, value=1}
              {interpMode=RCIM_Cubic, tangentMode=RCTM_User, time=1, value=0}
             ]}}})
        )
        B=1
       ))
      "Float (Alpha)"=&FloatFromCurve(
       FloatCurve=DataInterface() {propertyValues {Curve {keys [
           {arriveTangent=6.2301, interpMode=RCIM_Cubic, leaveTangent=6.2301, tangentMode=RCTM_User, time=0, value=0}
           {arriveTangent=-6.912, interpMode=RCIM_Cubic, leaveTangent=-6.912, tangentMode=RCTM_User, time=0.1, value=3}
           {arriveTangent=0.4048, interpMode=RCIM_Cubic, leaveTangent=0.4048, tangentMode=RCTM_User, time=0.7, value=0}
          ]}}}
       Scale Curve=&RandomRangeFloat(Minimum=0.6, Maximum=0.8)
      )
     )
    )
    Create Color Attribute=true
    Write to Module Namespace=true
   )
   /SetVariables(Particles.WindGutsPos:NiagaraPosition=&Lerp_Vector(
     A=<0 0 0>
     B=&Lerp_Vector(A=<0 0 1>, B=<0 0 30>, Alpha=&ReturnNormalizedExecIndex())
     Alpha=&FloatFromCurve(FloatCurve=DataInterface() {propertyValues {Curve {keys [<at 0 0>,<at 1 1>]}}})
    )) {isSetParameters=true}
  ]
  rendererClasses [&NiagaraMeshRendererProperties]
  renderers [{rendererClass=&NiagaraMeshRendererProperties, rendererIndex=0}]
 }
}
summary {
 systemName=NS_Emote_PolarityWinn_Soundwave_ground
 emitters [
  {emitterName=Multiple_ripples001, rendererClasses [&NiagaraMeshRendererProperties]}
  {emitterName=Ground_ripples001, rendererClasses [&NiagaraSpriteRendererProperties]}
  {emitterName=Square_Sprites, rendererClasses [&NiagaraSpriteRendererProperties]}
  {emitterName=Light, rendererClasses [&NiagaraLightRendererProperties]}
  {emitterName=wind_streaks, rendererClasses [&NiagaraMeshRendererProperties]}
  {emitterName=Smoke, rendererClasses [&NiagaraSpriteRendererProperties]}
  {emitterName=Energy_Pulse, rendererClasses [&NiagaraMeshRendererProperties]}
 ]
 userVariables [
  User.Anim_Curve_InitialOpacity {type=NiagaraFloat, defaultValue=1, description=""}
  User.Anim_Curve_SpawnRate {type=NiagaraFloat, defaultValue=0, description=""}
  User.bIsADS {type=NiagaraBool, defaultValue=true}
  User.bIsFrontEnd {type=NiagaraBool}
  User.bIsFrontEndPreview {type=NiagaraBool}
  User.DebugReactivity {type=NiagaraBool, defaultValue=true}
  User.FrontEndLooping {type=NiagaraFloat, defaultValue=0, description=""}
  User.Hitted {type=NiagaraFloat, defaultValue=0, description=""}
  User.ParaGlideLeanAlpha {type=NiagaraFloat, defaultValue=0, description=""}
  User.Scale {type=NiagaraFloat, defaultValue=1, description=""}
  User.Variant {type=NiagaraFloat, defaultValue=0, description=""}
  User.VariantSwitch {type=NiagaraFloat, defaultValue=0, description=""}
 ]
}
