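# Niagara system for the foot-impact effect of the PolarityWinn emote, built from two active burst emitters in a purple, magenta and cyan palette.
# Rings spawns short-lived meshes, colored by partition index from a color array, that expand and fade.
# Squares throws a few distance-culled sprites upward in a cone, slowed by drag, with curve-driven shrinking and time-of-day brightness.
# Sparks001 and both LWE emitters are disabled, as is Rings' dynamic material module.
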NiagaraSystem /BRCosmetics/Animation/Game/MainPlayer/Emotes/Polaritywinn/FX/NS_Emote_PolarityWinn_feet_impact
dependencies {
 usedDataInterfaces [
  &NiagaraDataInterfaceCurve
  &NiagaraDataInterfaceMeshRendererInfo
  &NiagaraDataInterfaceVector2DCurve
  &NiagaraDataInterfaceColorCurve
  &NiagaraDataInterfaceArrayColor
  &NiagaraDataInterfacePlatformSet
 ]
 usedDynamicInputs [
  &MakeCustomFloatFromBool
  /Niagara/DynamicInputs/Lerp/SmoothLerpOverTime_Float
  &RemapRange
  &SetBoolByFloatComparison
 ]
 usedModules [
  &self:SystemSpawnScript.SystemScriptSource.SystemScriptGraph.NiagaraNodeAssignment_3.SetVariables_#1
  &self:SystemSpawnScript.SystemScriptSource.SystemScriptGraph.NiagaraNodeAssignment_4.SetVariables_#2
  &SystemState
  &self:SystemSpawnScript.SystemScriptSource.SystemScriptGraph.NiagaraNodeAssignment_0.SetVariables_#3
  &self:SystemSpawnScript.SystemScriptSource.SystemScriptGraph.NiagaraNodeAssignment_1.SetVariables_#4
  &self:SystemSpawnScript.SystemScriptSource.SystemScriptGraph.NiagaraNodeAssignment_2.SetVariables_#5
 ]
}
emitters {
 Rings {
  emitterUpdateScript [
   &EmitterState()
   &SpawnBurst_Instantaneous(Spawn Count=8, Use Spawn Probability=false, UseLoopCountLimit=false)
  ]
  particleSpawnScript [
   &PartitionParticles(Number Of Partitions=1, Particles Per Partition=5, Partition Mode:ENiagara_PartitionMode_v2=::NewEnumerator1)
   &InitializeParticle(
    Color=&SelectLinearColorFromArray(
     Array Sampling Mode=::NewEnumerator1
     Direct Array Index=$Particles.PartitionParticles.Index In Partition
     Color Selection Array=DataInterface() {propertyValues {ColorData [<rgba 1.0237 0.195 1.3 0.3>, <rgba 0.227 0.107 0.4072 0.5>, <rgba 0 0.6823 1.5 0.7>, <rgba 0 1.5 1.2526 0.9>]}}
     Random Seed=?
    )
    Lifetime Max=0.8
    Lifetime Min=0.01
    Lifetime Mode=::NewEnumerator1
    Mesh Scale Mode=::NewEnumerator3
    Mesh Uniform Scale=0.25
    Position Offset=<0 0 0.5>
    UsePositionOffset=true
    Write Material Random=true
   )
  ]
  particleUpdateScript [
   &ParticleState()
   &ScaleMeshSize(Scale Factor=&VectorFromFloat(Value=&FloatFromCurve(FloatCurve=DataInterface() {propertyValues {Curve {keys [<at 0 0.2 RCIM_Cubic>, {arriveTangent=0.8, interpMode=RCIM_Cubic, leaveTangent=0.8, time=0.2041, value=0.6997}, <at 1 1 RCIM_Cubic>]}}})))
   &ScaleColor(
    Scale Alpha=&FloatFromCurve(
     FloatCurve=DataInterface() {propertyValues {Curve {
        defaultValue=2
        keys [{interpMode=RCIM_Cubic, tangentMode=RCTM_User, time=0, value=1}, {interpMode=RCIM_Cubic, tangentMode=RCTM_User, time=0.5, value=0}]
       }}}
     Scale Curve=0.5
    )
    Scale RGB=&VectorFromFloat(Value=&FloatFromCurve(
      FloatCurve=DataInterface() {propertyValues {Curve {
         defaultValue=1
         keys [{interpMode=RCIM_Cubic, tangentMode=RCTM_User, time=0, value=1}, {interpMode=RCIM_Cubic, tangentMode=RCTM_User, time=0.5, value=0}]
        }}}
      Scale Curve=&FNTimeOfDayScalar(Night Scale=1, Morning Scale=1, Evening Scale=1)
     ))
    ScaleA=true
    ScaleRGB=true
   ) {module=ScaleColor001}
   &DynamicMaterialParameters(
    Index 0 Param 1=&FloatFromCurve(FloatCurve=DataInterface() {propertyValues {Curve {keys [<at 0 1>,<at 0.2 1.375>,<at 1 0>]}}}, Scale Curve=1.2)
    Index 0 Param 2=2
    Index 0 Param 3=&FloatFromCurve(
     FloatCurve=DataInterface() {propertyValues {Curve {keys [<at 0.1 1>,<at 1 0>]}}}
     Scale Curve=&FN_FrontEndFloatModifier(Gameplay=6, Front End=3)
    )
    Index 0 Param 4=&FloatFromCurve(FloatCurve=DataInterface() {propertyValues {Curve {keys [
         <at 0 1>
         {arriveTangent=0.3079, interpMode=RCIM_Cubic, leaveTangent=0.3079, tangentMode=RCTM_User, time=0.3, value=0.9}
         <at 1 1 RCIM_Cubic>
        ]}}})
    Index 1 Param 1=&FN_PickFloatDependOnPlatform(TargetPlatformActive=180, TargetPlatformNoActive=100)
    Index 1 Param 2=&FN_PickFloatDependOnPlatform(TargetPlatformActive=2.5, TargetPlatformNoActive=5)
    Module.Index0Param0Write=true
    Module.Index0Param1Write=true
    Module.Index0Param2Write=true
    Module.Index0Param3Write=true
    Module.Index1Param0Write=true
    Module.Index1Param1Write=true
    Module.Index1Param2Write=true
    Module.Index1Param3Write=true
    Write Parameter Index 1=true
   ) {enabled=false}
   &SolveForcesAndVelocity(Clamp Velocity=false, Limit Acceleration=false)
  ]
  rendererClasses [&NiagaraMeshRendererProperties]
  renderers [{rendererClass=&NiagaraMeshRendererProperties, rendererIndex=0}]
 }
 Sparkes_LWE {bEnabled=false, rendererClasses [], renderers []}
 Sparks001 {
  bEnabled=false
  emitterUpdateScript [
   &EmitterState(
    Enable Distance Culling=true
    Max Distance Response=::KillImmediately
    Module.bApplySpawnCountScaleByDistance=true
    Module.bUseMaxDistance=true
    Module.bUseMinDistance=false
    Scalability Mode=::NewEnumerator1
    Scale Spawn Count=true
    Spawn Count Scale By Distance=DataInterface() {propertyValues {Curve {keys [<at 0.3676 0.9657>,<at 0.9962 0.4982>]}}}
   )
   &SpawnBurst_Instantaneous(Spawn Count=&UniformRangedInt(Minimum=3, Maximum=5), Use Spawn Probability=false)
  ]
  particleSpawnScript [
   &InitializeParticle(
    Color=<rgba 1 1 1 0.5>
    Lifetime Max=0.35
    Lifetime Min=0.2
    Lifetime Mode=::NewEnumerator1
    Sprite Size Mode=::NewEnumerator1
    Uniform Sprite Size Max=15
    Uniform Sprite Size Min=8
    UsePositionOffset=false
    Write Material Random=true
   )
   &SphereLocation(
    MaskSpawnGroup=false
    Override Seed=false
    Sphere Radius=5
    Surface Only=false
    UseNonUniformScale=false
    UseOffset=false
    UseSphereOrientation=false
   ) {module=SphereLocation001}
   &AddVelocity(
    Cone Angle=32
    Cone Axis=<0 0 1>
    Distribution Along Cone Axis=0.25
    Inner Cone Angle=31.8452
    Module.UseConeFalloff=true
    Module.UseVelDistribution=true
    Override Seed=false
    Speed Falloff From Cone Axis=0.5
    Velocity Mode=::NewEnumerator2
    Velocity Speed=&RandomRangeFloat(Minimum=50, Maximum=300)
   )
  ]
  particleUpdateScript [
   &ParticleState()
   &Drag(Drag=&UniformRangedFloat(Minimum=0.2, Maximum=12), Use Linear Drag=true, Use Rotational Drag=false)
   &ScaleSpriteSize(Scale Factor=&Vector2DFromCurve(
     Vector2Curve=DataInterface() {propertyValues {
       XCurve {keys [
         {arriveTangent=9.231, interpMode=RCIM_Cubic, leaveTangent=9.231, tangentMode=RCTM_User, time=0, value=0.1}
         {arriveTangent=-0.1, interpMode=RCIM_Cubic, leaveTangent=-0.1, tangentMode=RCTM_User, time=0.18, value=0.8}
         {arriveTangent=-2.4373, interpMode=RCIM_Cubic, leaveTangent=-2.4373, tangentMode=RCTM_User, time=1, value=0}
        ]}
       YCurve {keys [
         {arriveTangent=9.7317, interpMode=RCIM_Cubic, leaveTangent=9.7317, tangentMode=RCTM_User, time=0, value=0.1}
         {arriveTangent=-0.1001, interpMode=RCIM_Cubic, leaveTangent=-0.1001, tangentMode=RCTM_User, time=0.18, value=0.8}
         {arriveTangent=-2.4373, interpMode=RCIM_Cubic, leaveTangent=-2.4373, tangentMode=RCTM_User, time=0.9988, value=0}
        ]}
      }}
     ScaleCurve=<1 1>
    ))
   &ScaleColor(
    Scale RGB=&MakeVectorFromLinearColorRGB(Color=&HueShiftLinearColor(
      Hue=&RandomRangeFloat(Minimum=-0.1, Maximum=0)
      Saturation=1
      Value=1
      Linear Color=&ColorFromCurve(ColorCurve=DataInterface() {propertyValues {
         AlphaCurve {keys [<at 0 1>,<at 1 0>]}
         BlueCurve {keys [<at 0.17 15>,<at 0.2011 2>,<at 0.33 1.35>,<at 0.85 1.1>]}
         GreenCurve {keys [<at 0.17 7.6606>,<at 0.2011 0.722>,<at 0.33 0.2547>,<at 0.85 0.3056>]}
         RedCurve {keys [<at 0.17 0>,<at 0.2011 0>,<at 0.33 0>,<at 0.85 0>]}
        }})
     ))
    ScaleA=true
    ScaleRGB=true
   )
   /SetVariables(
    Particles.Color2=&MultiplyLinearColor(
     Linear Color=$Particles.Color
     Scale Factor=&MakeFloat4FromVectorAndFloat("Vector (RGB)"=&VectorFromFloat(Value=7), "Float (Alpha)"=1)
    )
    Particles.SpriteSize2=&Multiply_Vector2D_ByFloat(Float=0.3, Vector2D=$Particles.SpriteSize)
   ) {isSetParameters=true, module=SetVariables_#6}
   &SolveForcesAndVelocity(Clamp Velocity=false, Limit Acceleration=false)
  ]
  rendererClasses [&NiagaraSpriteRendererProperties]
  renderers [
   {rendererClass=&NiagaraSpriteRendererProperties}
   {rendererClass=&NiagaraSpriteRendererProperties, rendererIndex=1}
  ]
 }
 Squares {
  emitterUpdateScript [
   &EmitterState(
    Enable Distance Culling=true
    Max Distance Response=::KillImmediately
    Module.bApplySpawnCountScaleByDistance=true
    Module.bUseMaxDistance=true
    Module.bUseMinDistance=false
    Scalability Mode=::NewEnumerator1
    Scale Spawn Count=true
    Spawn Count Scale By Distance=DataInterface() {propertyValues {Curve {keys [<at 0.3676 0.9657>,<at 0.9962 0.4982>]}}}
   )
   &SpawnBurst_Instantaneous(Spawn Count=&UniformRangedInt(Minimum=3, Maximum=5), Use Spawn Probability=false)
  ]
  particleSpawnScript [
   &InitializeParticle(
    Color=&SelectLinearColorFromArray(
     Array Sampling Mode=::NewEnumerator0
     Direct Array Index=?
     Color Selection Array=DataInterface() {propertyValues {ColorData [<rgba 1.0237 0.195 1.3 5>, <rgba 1.1813 0.225 1.5 5>, <rgba 0 0.2667 0.4 5>, <rgba 0 0.6823 1.5 5>, <rgba 1.05 0 1.5 5>]}}
     Random Seed=0
    )
    Lifetime Max=0.35
    Lifetime Min=0.2
    Lifetime Mode=::NewEnumerator1
    Sprite Size Mode=::NewEnumerator1
    Uniform Sprite Size Min=2
    UsePositionOffset=false
    Write Material Random=true
   )
   &SphereLocation(
    MaskSpawnGroup=false
    Override Seed=false
    Sphere Radius=5
    Surface Only=false
    UseNonUniformScale=false
    UseOffset=false
    UseSphereOrientation=false
   ) {module=SphereLocation001}
   &AddVelocity(
    Cone Angle=32
    Cone Axis=<0 0 1>
    Distribution Along Cone Axis=0.25
    Inner Cone Angle=31.8452
    Module.UseConeFalloff=true
    Module.UseVelDistribution=true
    Override Seed=false
    Speed Falloff From Cone Axis=0.5
    Velocity Mode=::NewEnumerator2
    Velocity Speed=&RandomRangeFloat(Minimum=50, Maximum=300)
   )
  ]
  particleUpdateScript [
   &ParticleState()
   &Drag(Drag=&UniformRangedFloat(Minimum=0.2, Maximum=12), Use Linear Drag=true, Use Rotational Drag=false)
   &ScaleSpriteSize(Scale Factor=&Vector2DFromCurve(
     Vector2Curve=DataInterface() {propertyValues {
       XCurve {keys [
         {arriveTangent=9.231, interpMode=RCIM_Cubic, leaveTangent=9.231, tangentMode=RCTM_User, time=0, value=0.1}
         {arriveTangent=-0.1, interpMode=RCIM_Cubic, leaveTangent=-0.1, tangentMode=RCTM_User, time=0.18, value=0.8}
         {arriveTangent=-2.4373, interpMode=RCIM_Cubic, leaveTangent=-2.4373, tangentMode=RCTM_User, time=1, value=0}
        ]}
       YCurve {keys [
         {arriveTangent=9.7317, interpMode=RCIM_Cubic, leaveTangent=9.7317, tangentMode=RCTM_User, time=0, value=0.1}
         {arriveTangent=-0.1001, interpMode=RCIM_Cubic, leaveTangent=-0.1001, tangentMode=RCTM_User, time=0.18, value=0.8}
         {arriveTangent=-2.4373, interpMode=RCIM_Cubic, leaveTangent=-2.4373, tangentMode=RCTM_User, time=0.9988, value=0}
        ]}
      }}
     ScaleCurve=<1 1>
    ))
   &ScaleColor(
    Scale Alpha=0.7
    Scale RGB=&VectorFromFloat(Value=&FNTimeOfDayScalar(Night Scale=1, Morning Scale=1, Day Scale=5, Evening Scale=1))
    ScaleA=true
    ScaleRGB=true
   ) {module=ScaleColor003}
   &ScaleColor(
    Scale RGB=&MakeVectorFromLinearColorRGB(Color=&HueShiftLinearColor(
      Hue=&RandomRangeFloat(Minimum=-0.1, Maximum=0)
      Saturation=1
      Value=1
      Linear Color=&ColorFromCurve(ColorCurve=DataInterface() {propertyValues {
         AlphaCurve {keys [<at 0 1>,<at 1 0>]}
         BlueCurve {keys [<at 0.17 15>,<at 0.2011 2>,<at 0.33 1.35>,<at 0.85 1.1>]}
         GreenCurve {keys [<at 0.17 7.6606>,<at 0.2011 0.722>,<at 0.33 0.2547>,<at 0.85 0.3056>]}
         RedCurve {keys [<at 0.17 0>,<at 0.2011 0>,<at 0.33 0>,<at 0.85 0>]}
        }})
     ))
    ScaleA=true
    ScaleRGB=true
   ) {enabled=false}
   /SetVariables(
    Particles.Color2=&MultiplyLinearColor(
     Linear Color=$Particles.Color
     Scale Factor=&MakeFloat4FromVectorAndFloat("Vector (RGB)"=&VectorFromFloat(Value=7), "Float (Alpha)"=1)
    )
    Particles.SpriteSize2=&Multiply_Vector2D_ByFloat(Float=0.3, Vector2D=$Particles.SpriteSize)
   ) {isSetParameters=true, module=SetVariables_#6}
   &SolveForcesAndVelocity(Clamp Velocity=false, Limit Acceleration=false)
   &FN_Sprite_Attributes(
    Color=&MakeLinearColorFromVectorAndFloat(
     "Vector (RGB)"=&MakeVectorFromLinearColorRGB(Color=&Multiply_LinearColorByFloat(
       Linear Color=&ColorFromCurve(
        ColorCurve=DataInterface() {propertyValues {
          AlphaCurve {keys [<at 0 1 RCIM_Cubic>]}
          BlueCurve {keys [<at 0.3173 1 RCIM_Cubic>,<at 0.3548 1>,<at 0.6928 1>,<at 0.7245 0.6864>]}
          GreenCurve {keys [<at 0.3173 0.1714 RCIM_Cubic>,<at 0.3548 0.3565>,<at 0.6928 0.2891>,<at 0.7245 0.1519>]}
          RedCurve {keys [<at 0.3173 0.3432 RCIM_Cubic>,<at 0.3548 0>,<at 0.6928 0>,<at 0.7245 1>]}
         }}
        CurveIndex=&RandomRangeFloat(Minimum=0, Maximum=1)
       )
       Float=&Lerp_Float(
        A=1
        B=20
        Alpha=&FloatFromCurve(FloatCurve=DataInterface() {propertyValues {Curve {keys [
             {arriveTangent=-0.4493, interpMode=RCIM_Cubic, leaveTangent=-0.4493, tangentMode=RCTM_User, time=0, value=1}
             {interpMode=RCIM_Cubic, tangentMode=RCTM_User, time=0.7, value=0}
            ]}}})
       )
      ))
     "Float (Alpha)"=&FloatFromCurve(FloatCurve=DataInterface() {propertyValues {Curve {keys [
          {interpMode=RCIM_Cubic, tangentMode=RCTM_User, time=0, value=1}
          {arriveTangent=-2.0246, interpMode=RCIM_Cubic, leaveTangent=-2.0246, tangentMode=RCTM_User, time=1, value=0}
         ]}}})
    )
    Create Color Attribute=true
    Create SpriteSize Attribute=true
    SpriteSize=&Multiply_Vector2D_ByFloat(
     Float=&Multiply_Float(
      A=&RandomRangeFloat(Minimum=0.5, Maximum=1.5)
      B=&FloatFromCurve(FloatCurve=DataInterface() {propertyValues {Curve {keys [
           {interpMode=RCIM_Cubic, tangentMode=RCTM_User, time=0.2, value=1}
           {arriveTangent=-2.4619, interpMode=RCIM_Cubic, leaveTangent=-2.4619, tangentMode=RCTM_User, time=1, value=0}
          ]}}})
     )
     Vector2D=<5 5>
    )
    Write to Module Namespace=true
   )
  ]
  rendererClasses [&NiagaraSpriteRendererProperties]
  renderers [
   {rendererClass=&NiagaraSpriteRendererProperties}
   {rendererClass=&NiagaraSpriteRendererProperties, rendererIndex=1}
   {rendererClass=&NiagaraSpriteRendererProperties, rendererIndex=2}
  ]
 }
 Squares_LWE {bEnabled=false, rendererClasses [], renderers []}
}
summary {
 systemName=NS_Emote_PolarityWinn_feet_impact
 emitters [
  {bEnabled=false, emitterName=Sparks001, rendererClasses [&NiagaraSpriteRendererProperties]}
  {emitterName=Rings, rendererClasses [&NiagaraMeshRendererProperties]}
  {bEnabled=false, emitterName=Sparkes_LWE, rendererClasses []}
  {bEnabled=false, emitterName=Squares_LWE, rendererClasses []}
  {emitterName=Squares, rendererClasses [&NiagaraSpriteRendererProperties]}
 ]
 userVariables [
  User.bIsADS {type=NiagaraBool, defaultValue=true}
  User.bIsFrontEnd {type=NiagaraBool}
  User.bIsFrontEndPreview {type=NiagaraBool}
  User.DebugReactivity {type=NiagaraBool, defaultValue=true}
  User.FEOffset {type=NiagaraFloat, defaultValue=10, description=""}
  User.FEScale {type=NiagaraFloat, defaultValue=0.6, description=""}
  User.FrontEndLooping {type=NiagaraFloat, defaultValue=0, description=""}
  User.Hit {type=NiagaraBool, defaultValue=true}
  User.Hitted {type=NiagaraFloat, defaultValue=0, description=""}
 ]
}
