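# Niagara system for the PolarityWinn emote's feet landing: a short burst of neon blue-magenta rings, flare and sparks.
# BurstRing spawns growing, fading meshes with dynamic material parameters partly indexed by System.AudioReactivity.
# Flare layers sprites with a fading light, while Sparks001 places sprites on a widening ring under drag and rising upward gravity.
# Rings and sparks scale brightness by time of day; the Rings and Sparkles emitters are disabled.
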NiagaraSystem /BRCosmetics/Animation/Game/MainPlayer/Emotes/Polaritywinn/FX/NS_Emote_PolarityWinn_feet_landing
dependencies {
 usedDataInterfaces [
  &NiagaraDataInterfaceCurve
  &NiagaraDataInterfaceColorCurve
  &NiagaraDataInterfaceMeshRendererInfo
  &NiagaraDataInterfaceVectorCurve
  &NiagaraDataInterfaceVector2DCurve
 ]
 usedDynamicInputs [
  &MakeCustomFloatFromBool
  /Niagara/DynamicInputs/Lerp/SmoothLerpOverTime_Float
  &FN_FrontEndFloatModifier
  &RemapRange
 ]
 usedModules [
  &self:SystemSpawnScript.SystemScriptSource.SystemScriptGraph.NiagaraNodeAssignment_3.SetVariables_#1
  &self:SystemSpawnScript.SystemScriptSource.SystemScriptGraph.NiagaraNodeAssignment_4.SetVariables_#2
  &SystemState
  &self:SystemSpawnScript.SystemScriptSource.SystemScriptGraph.NiagaraNodeAssignment_0.SetVariables_#3
  &self:SystemSpawnScript.SystemScriptSource.SystemScriptGraph.NiagaraNodeAssignment_1.SetVariables_#4
  &self:SystemSpawnScript.SystemScriptSource.SystemScriptGraph.NiagaraNodeAssignment_2.SetVariables_#5
 ]
}
emitters {
 BurstRing {
  emitterUpdateScript [
   &EmitterState(
    Inactive Response=::NewEnumerator1
    Life Cycle Mode=::NewEnumerator1
    Loop Behavior=::NewEnumerator1
    Loop Duration Mode=::NewEnumerator1
    UseLoopDelay=false
   )
   &SpawnBurst_Instantaneous(Spawn Count=8, Spawn Time=0.01, Use Spawn Probability=false, UseLoopCountLimit=false)
   &SpawnRate(SpawnRate=5, Use Spawn Probability=false) {enabled=false}
  ]
  particleSpawnScript [
   &InitializeParticle(
    Lifetime Max=0.5
    Lifetime Min=0.1
    Lifetime Mode=::NewEnumerator1
    Mesh Scale Mode=::NewEnumerator1
    Mesh Uniform Scale Max=0.3
    Mesh Uniform Scale Min=0.1
    Sprite Size Mode=::NewEnumerator3
    UsePositionOffset=false
    Write Material Random=true
   )
   &Color(Color=&SelectLinearColorFromArray(
     Array Sampling Mode=::NewEnumerator0
     Direct Array Index=?
     Direct Array Mode=::NewEnumerator1
     Color Selection Array=DataInterface() {propertyValues {ColorData [<rgba 0 0.3319 1 1>, <rgba 0.9226 0.0503 1 1>, <rgba 0 0 0 0>]}}
     Random Seed=0
    )) {enabled=false}
  ]
  particleUpdateScript [
   &ParticleState()
   /SetVariables(Particles.MeshOrientation=<1 0 0 0>) {isSetParameters=true, module=SetVariables_#6}
   /SetVariables(Particles.MeshOrientation=<0 1 1 0>) {enabled=false, isSetParameters=true, module=SetVariables_#7}
   &Timeline(Initial Start Time=256.258, Play Rate=0.7, Timeline Beginning=256.258, Timeline End=100000) {enabled=false}
   &DynamicMaterialParameters(
    Index 0 Param 1=&MakeFloatFromVector(Vector=$Particles.Scale, Channel=::NewEnumerator0)
    Index 0 Param 2=2
    Index 0 Param 3=&FloatFromCurve(FloatCurve=DataInterface() {propertyValues {Curve {keys [<at 0 0>,<at 1 0>]}}}, CurveIndex=$System.AudioReactivity)
    Index 0 Param 4=15
    Index 1 Param 1=$Particles.Timeline.Time
    Index 1 Param 2=1
    Index 1 Param 3=0.5
    Index 1 Param 4=$Particles.Timeline.Time
    Index 2 Param 1=2
    Index 2 Param 2=5
    Index 2 Param 3=&FloatFromCurve(FloatCurve=DataInterface() {propertyValues {Curve {keys [<at 0 0.5>,<at 1 1>]}}}, CurveIndex=$System.AudioReactivity)
    Module.Index0Param0Write=true
    Module.Index0Param1Write=true
    Module.Index0Param2Write=true
    Module.Index0Param3Write=true
    Module.Index1Param0Write=true
    Module.Index1Param1Write=true
    Module.Index1Param2Write=true
    Module.Index1Param3Write=true
    Module.Index2Param0Write=true
    Module.Index2Param1Write=true
    Module.Index2Param2Write=true
    Module.Index2Param3Write=true
    Write Parameter Index 1=true
    Write Parameter Index 2=true
   )
   &ScaleMeshSize(Scale Factor=&VectorFromFloat(Value=&FloatFromCurve(FloatCurve=DataInterface() {propertyValues {Curve {keys [
          {interpMode=RCIM_Cubic, leaveTangent=0.6196, tangentMode=RCTM_Break, time=0, value=0.5}
          {arriveTangent=0.7561, interpMode=RCIM_Cubic, tangentMode=RCTM_Break, time=1, value=1.7}
         ]}}})))
   &ScaleColor(
    Scale Alpha=&FloatFromCurve(FloatCurve=DataInterface() {propertyValues {Curve {keys [
         {interpMode=RCIM_Cubic, tangentMode=RCTM_User, time=0, value=0}
         {arriveTangent=-2.0542, interpMode=RCIM_Cubic, leaveTangent=-2.0542, tangentMode=RCTM_User, time=0.2, value=1}
         {arriveTangent=-0.4397, tangentMode=RCTM_User, time=0.5, value=0}
        ]}}})
    Scale RGB=&VectorFromFloat(Value=&FloatFromCurve(FloatCurve=DataInterface() {propertyValues {Curve {
         defaultValue=1
         keys [{interpMode=RCIM_Cubic, tangentMode=RCTM_User, time=0, value=1}, {interpMode=RCIM_Cubic, tangentMode=RCTM_User, time=0.5, value=0}]
        }}}))
    ScaleA=true
    ScaleRGB=true
   )
   &ScaleColor(
    Scale Alpha=0.7
    Scale RGB=&VectorFromFloat(Value=&FNTimeOfDayScalar(Night Scale=1, Morning Scale=1, Day Scale=5, Evening Scale=1))
    ScaleA=false
    ScaleRGB=true
   ) {module=ScaleColor003}
   &Color(
    Color=&ColorFromCurve(ColorCurve=DataInterface() {propertyValues {
       AlphaCurve {keys [<at 0 1>,<at 1 0>]}
       BlueCurve {keys [<at 0.0238 1>,<at 0.1259 1 RCIM_Cubic>,<at 0.4 1>,<at 0.5655 0.4072>,<at 1 1>]}
       GreenCurve {keys [
         <at 0.0238 0.0503>
         {arriveTangent=0.7488, interpMode=RCIM_Cubic, leaveTangent=0.7488, time=0.1259, value=0.9157}
         <at 0.4 0.3319>
         <at 0.5655 0.107>
         <at 1 0.3357>
        ]}
       RedCurve {keys [
         <at 0.0238 0.9226>
         {arriveTangent=-2.4526, interpMode=RCIM_Cubic, leaveTangent=-2.4526, time=0.1259, value=0}
         <at 0.4 0>
         <at 0.5655 0.227>
         <at 1 0>
        ]}
      }})
    Scale Alpha=&FloatFromCurve(FloatCurve=DataInterface() {propertyValues {Curve {keys [
         {interpMode=RCIM_Cubic, tangentMode=RCTM_User, time=0, value=0}
         {arriveTangent=-2.0542, interpMode=RCIM_Cubic, leaveTangent=-2.0542, tangentMode=RCTM_User, time=0.2, value=1}
         {arriveTangent=-0.4397, tangentMode=RCTM_User, time=0.8, value=0}
        ]}}})
   ) {module=Color002}
   &ScaleColor(
    Scale Alpha=0.7
    Scale RGB=&VectorFromFloat(Value=&FNTimeOfDayScalar(Night Scale=5, Morning Scale=1, Evening Scale=1))
    ScaleA=true
    ScaleRGB=true
   ) {enabled=false, module=ScaleColor001}
   &ScaleColor(
    Scale RGB=&VectorFromFloat(Value=&FloatFromCurve(
      FloatCurve=DataInterface() {propertyValues {Curve {keys [{interpMode=RCIM_Cubic, leaveTangent=1.4829, tangentMode=RCTM_Break, time=0, value=1}, <at 1 2.3>]}}}
      CurveIndex=$User.Hitted
     ))
    ScaleA=true
    ScaleRGB=true
   ) {enabled=false, module=ScaleColor002}
   &Color(Color=&ColorFromCurve(ColorCurve=DataInterface() {propertyValues {
       AlphaCurve {keys [<at 0 1 RCIM_Cubic>,<at 1 0 RCIM_Cubic>]}
       BlueCurve {keys [<at 0 1 RCIM_Cubic>,<at 0.2697 0.8863>,<at 0.6621 0.8058>,<at 1 0.7366 RCIM_Cubic>]}
       GreenCurve {keys [<at 0 0.3319 RCIM_Cubic>,<at 0.2697 0.0493>,<at 0.6621 0.7096>,<at 1 1 RCIM_Cubic>]}
       RedCurve {keys [<at 0 0 RCIM_Cubic>,<at 0.2697 1>,<at 0.6621 0.0014>,<at 1 0 RCIM_Cubic>]}
      }})) {enabled=false, module=Color001}
   &KillParticles(Kill Particles=&SetBoolByFloatComparison(A=$Particles.NormalizedAge, B=0.7))
   /SetVariables(Particles.MeshOrientation=&FindQuatBetweenTwoVectors(From Vector=<1 0 1>, To Vector=<0 0 1>)) {enabled=false, isSetParameters=true, module=SetVariables_#8}
  ]
  rendererClasses [&NiagaraMeshRendererProperties]
  renderers [
   {rendererClass=&NiagaraMeshRendererProperties, rendererIndex=0}
   {rendererClass=&NiagaraMeshRendererProperties, rendererIndex=1}
  ]
 }
 Flare {
  emitterUpdateScript [
   &EmitterState()
   &SpawnBurst_Instantaneous(Use Spawn Probability=false, UseLoopCountLimit=false)
  ]
  particleSpawnScript [
   &InitializeParticle(
    Color=<rgba 0.6771 0.6458 1 1>
    Lifetime=0.45
    Mass Mode=::NewEnumerator1
    Mesh Scale Mode=::NewEnumerator3
    Mesh Uniform Scale=0.2
    Position Mode=::NewEnumerator1
    Sprite Rotation Mode=::NewEnumerator2
    Sprite Size=&Vector2DFromFloat(Value=&Multiply_Float(A=550, B=1))
    Sprite Size Mode=::NewEnumerator4
    UsePositionOffset=true
    Write Material Random=true
   ) {module=InitializeParticle001}
   &CameraOffset(Camera Offset Amount=75)
  ]
  particleUpdateScript [
   &ParticleState()
   &SolveForcesAndVelocity(Clamp Velocity=false, Limit Acceleration=false)
   &ScaleSpriteSize(Scale Factor=&Multiply_Vector2D_ByFloat(
     Float=&FloatFromCurve(
      FloatCurve=DataInterface() {propertyValues {Curve {keys [
          {arriveTangent=0.7585, interpMode=RCIM_Cubic, leaveTangent=0.7585, tangentMode=RCTM_User, time=0, value=0.4}
          {interpMode=RCIM_Cubic, tangentMode=RCTM_User, time=1, value=1}
         ]}}}
      Scale Curve=0.1
     )
     Vector2D=<1 1>
    ))
   &SpriteRotationRate(Rotation Rate=&UniformRangedFloat(Minimum=-50, Maximum=50))
   /SetVariables(Particles.DarkGlowSize:Vector2f=&Multiply_Vector2D_ByFloat(Float=3, Vector2D=$Particles.SpriteSize)) {isSetParameters=true}
   /SetVariables(Particles.DargGlowColor:LinearColor=&MakeLinearColorFromVectorAndFloat(
     "Vector (RGB)"=&MakeVectorFromLinearColorRGB(Color=<rgba 0.0452 0.0406 0.05 1>)
     "Float (Alpha)"=&FloatFromCurve(FloatCurve=DataInterface() {propertyValues {Curve {keys [
          {arriveTangent=-0.0975, interpMode=RCIM_Cubic, leaveTangent=-0.0975, tangentMode=RCTM_User, time=0, value=1}
          {interpMode=RCIM_Cubic, leaveTangent=0.002, time=1, value=0}
         ]}}})
    )) {isSetParameters=true}
   /SetVariables(Particles.CentralFlareSize:Vector2f=&Multiply_Vector2D_ByFloat(Float=3, Vector2D=$Particles.SpriteSize)) {isSetParameters=true}
   /SetVariables(Particles.CentralFlareColor:LinearColor=&MakeLinearColorFromVectorAndFloat(
     "Vector (RGB)"=&MakeVectorFromLinearColorRGB(Color=&ColorFromCurve(ColorCurve=DataInterface() {propertyValues {
         AlphaCurve {keys [<at 0 1 RCIM_Cubic>,<at 1 0 RCIM_Cubic>]}
         BlueCurve {keys [<at 0 2>, {interpMode=RCIM_Cubic, leaveTangent=1.3364, time=1, value=1}]}
         GreenCurve {keys [<at 0 0.5272>, {interpMode=RCIM_Cubic, leaveTangent=-5.7627, time=1, value=0.3351}]}
         RedCurve {keys [<at 0 0.0903>, {interpMode=RCIM_Cubic, leaveTangent=-13.364, time=1, value=0.8022}]}
        }}))
     "Float (Alpha)"=&FloatFromCurve(
      FloatCurve=DataInterface() {propertyValues {Curve {keys [
          <at 0 0 RCIM_Cubic>
          {arriveTangent=0.3854, interpMode=RCIM_Cubic, leaveTangent=0.3854, tangentMode=RCTM_User, time=0.1, value=1}
          <at 1 0 RCIM_Cubic>
         ]}}}
      Scale Curve=0.5
     )
    )) {isSetParameters=true}
   &ScaleColor(
    Scale Alpha=&FloatFromCurve(
     FloatCurve=DataInterface() {propertyValues {Curve {keys [
         {arriveTangent=6.4019, interpMode=RCIM_Cubic, leaveTangent=6.4019, tangentMode=RCTM_User, time=0, value=1}
         {interpMode=RCIM_Cubic, leaveTangent=0.002, time=1, value=0}
        ]}}}
     Scale Curve=0.02
    )
    Scale RGB=&MakeVectorFromLinearColorRGB(Color=$Particles.CentralFlareColor)
    ScaleA=true
    ScaleRGB=true
   )
   &FN_Light_Attributes(
    Color=$Particles.CentralFlareColor
    Create Color Attribute=true
    Create Exponent Attribute=true
    Create Radius Attribute=true
    Exponent=1
    Radius=&Multiply_Float(
     A=&FloatFromCurve(
      FloatCurve=DataInterface() {propertyValues {Curve {keys [
          {arriveTangent=2.3282, interpMode=RCIM_Cubic, leaveTangent=2.3282, tangentMode=RCTM_User, time=0, value=0}
          {interpMode=RCIM_Cubic, tangentMode=RCTM_User, time=1, value=1}
         ]}}}
      Scale Curve=150
     )
     B=0.5
    )
    Scale Color=&FloatFromCurve(
     FloatCurve=DataInterface() {propertyValues {Curve {keys [
         {interpMode=RCIM_Cubic, tangentMode=RCTM_User, time=0, value=1}
         {arriveTangent=-3.4524, interpMode=RCIM_Cubic, leaveTangent=-3.4524, tangentMode=RCTM_User, time=1, value=0}
        ]}}}
     Scale Curve=7
    )
   )
   &DynamicMaterialParameters(
    Index 1 Param 1=0.3
    Index 1 Param 4=&Lerp_Float(A=0, B=1, Alpha=0.5)
    Module.Index1Float0Write=true
    Module.Index1Float1Write=true
    Module.Index1Float2Write=true
    Module.Index1Float3Write=true
    Module.Index3Float0Write=true
    Module.Index3Float1Write=true
    Module.Index3Float2Write=true
    Module.Index3Float3Write=true
    Write Parameter Index 0=::NewEnumerator0
    Write Parameter Index 1=::NewEnumerator1
    Write Parameter Index 3=::NewEnumerator1
   )
   &FN_Sprite_Attributes(
    Create Sprite Rotation Attribute=true
    Create SpriteSize Attribute=true
    SpriteSize=&Multiply_Vector2D(A=$Particles.CentralFlareSize, B=<2 0.15>)
    Write to Module Namespace=true
   )
  ]
  rendererClasses [&NiagaraSpriteRendererProperties, &NiagaraLightRendererProperties]
  renderers [
   {rendererClass=&NiagaraSpriteRendererProperties}
   {rendererClass=&NiagaraSpriteRendererProperties, rendererIndex=1}
   {rendererClass=&NiagaraLightRendererProperties, rendererIndex=2}
   {rendererClass=&NiagaraSpriteRendererProperties, rendererIndex=3}
   {rendererClass=&NiagaraSpriteRendererProperties, rendererIndex=4}
  ]
 }
 Rings {
  bEnabled=false
  emitterUpdateScript [
   &EmitterState(
    Inactive Response=::NewEnumerator1
    Life Cycle Mode=::NewEnumerator1
    Loop Behavior=::NewEnumerator1
    Loop Duration Mode=::NewEnumerator1
    UseLoopDelay=false
   )
   &SpawnBurst_Instantaneous(Spawn Count=8, Spawn Time=0.05, Use Spawn Probability=false, UseLoopCountLimit=false)
   &SpawnRate(SpawnRate=5, Use Spawn Probability=false) {enabled=false}
  ]
  particleSpawnScript [
   &InitializeParticle(
    Lifetime Max=0.5
    Lifetime Min=0.1
    Lifetime Mode=::NewEnumerator1
    Mesh Scale Mode=::NewEnumerator1
    Mesh Uniform Scale Max=0.3
    Mesh Uniform Scale Min=0.1
    Sprite Size Mode=::NewEnumerator3
    UsePositionOffset=false
    Write Material Random=true
   )
   &Color(Color=&SelectLinearColorFromArray(
     Array Sampling Mode=::NewEnumerator0
     Direct Array Index=?
     Direct Array Mode=::NewEnumerator1
     Color Selection Array=DataInterface() {propertyValues {ColorData [<rgba 0 0.3319 1 1>, <rgba 0.9226 0.0503 1 1>, <rgba 0 0 0 0>]}}
     Random Seed=0
    )) {enabled=false}
  ]
  particleUpdateScript [
   &ParticleState()
   /SetVariables(Particles.MeshOrientation=<1 0 0 0>) {isSetParameters=true, module=SetVariables_#6}
   /SetVariables(Particles.MeshOrientation=<0 1 1 0>) {enabled=false, isSetParameters=true, module=SetVariables_#7}
   &Timeline(Initial Start Time=256.258, Play Rate=0.7, Timeline Beginning=256.258, Timeline End=100000) {enabled=false}
   &DynamicMaterialParameters(
    Index 0 Param 1=&MakeFloatFromVector(Vector=$Particles.Scale, Channel=::NewEnumerator0)
    Index 0 Param 2=2
    Index 0 Param 3=&FloatFromCurve(FloatCurve=DataInterface() {propertyValues {Curve {keys [<at 0 0>,<at 1 0>]}}}, CurveIndex=$System.AudioReactivity)
    Index 0 Param 4=15
    Index 1 Param 1=$Particles.Timeline.Time
    Index 1 Param 2=1
    Index 1 Param 3=0.5
    Index 1 Param 4=$Particles.Timeline.Time
    Index 2 Param 1=2
    Index 2 Param 2=5
    Index 2 Param 3=&FloatFromCurve(FloatCurve=DataInterface() {propertyValues {Curve {keys [<at 0 0.5>,<at 1 1>]}}}, CurveIndex=$System.AudioReactivity)
    Module.Index0Param0Write=true
    Module.Index0Param1Write=true
    Module.Index0Param2Write=true
    Module.Index0Param3Write=true
    Module.Index1Param0Write=true
    Module.Index1Param1Write=true
    Module.Index1Param2Write=true
    Module.Index1Param3Write=true
    Module.Index2Param0Write=true
    Module.Index2Param1Write=true
    Module.Index2Param2Write=true
    Module.Index2Param3Write=true
    Write Parameter Index 1=true
    Write Parameter Index 2=true
   )
   &ScaleMeshSize(Scale Factor=&VectorFromFloat(Value=&FloatFromCurve(FloatCurve=DataInterface() {propertyValues {Curve {keys [
          {interpMode=RCIM_Cubic, leaveTangent=0.6196, tangentMode=RCTM_Break, time=0, value=0.5}
          {arriveTangent=0.7561, interpMode=RCIM_Cubic, tangentMode=RCTM_Break, time=1, value=1.7}
         ]}}})))
   &ScaleColor(
    Scale Alpha=&FloatFromCurve(FloatCurve=DataInterface() {propertyValues {Curve {keys [
         {interpMode=RCIM_Cubic, tangentMode=RCTM_User, time=0, value=0}
         {arriveTangent=-2.0542, interpMode=RCIM_Cubic, leaveTangent=-2.0542, tangentMode=RCTM_User, time=0.2, value=1}
         {arriveTangent=-0.4397, tangentMode=RCTM_User, time=0.5, value=0}
        ]}}})
    Scale RGB=&VectorFromFloat(Value=&FloatFromCurve(FloatCurve=DataInterface() {propertyValues {Curve {
         defaultValue=1
         keys [{interpMode=RCIM_Cubic, tangentMode=RCTM_User, time=0, value=1}, {interpMode=RCIM_Cubic, tangentMode=RCTM_User, time=0.5, value=0}]
        }}}))
    ScaleA=true
    ScaleRGB=true
   )
   &ScaleColor(
    Scale Alpha=0.7
    Scale RGB=&VectorFromFloat(Value=&FNTimeOfDayScalar(Night Scale=1, Morning Scale=1, Day Scale=5, Evening Scale=1))
    ScaleA=false
    ScaleRGB=true
   ) {module=ScaleColor003}
   &Color(
    Color=&ColorFromCurve(ColorCurve=DataInterface() {propertyValues {
       AlphaCurve {keys [<at 0 1>,<at 1 0>]}
       BlueCurve {keys [<at 0.0238 1>,<at 0.1259 1 RCIM_Cubic>,<at 0.4 1>,<at 0.5655 0.4072>,<at 1 1>]}
       GreenCurve {keys [
         <at 0.0238 0.0503>
         {arriveTangent=0.7488, interpMode=RCIM_Cubic, leaveTangent=0.7488, time=0.1259, value=0.9157}
         <at 0.4 0.3319>
         <at 0.5655 0.107>
         <at 1 0.3357>
        ]}
       RedCurve {keys [
         <at 0.0238 0.9226>
         {arriveTangent=-2.4526, interpMode=RCIM_Cubic, leaveTangent=-2.4526, time=0.1259, value=0}
         <at 0.4 0>
         <at 0.5655 0.227>
         <at 1 0>
        ]}
      }})
    Scale Alpha=&FloatFromCurve(FloatCurve=DataInterface() {propertyValues {Curve {keys [
         {interpMode=RCIM_Cubic, tangentMode=RCTM_User, time=0, value=0}
         {arriveTangent=-2.0542, interpMode=RCIM_Cubic, leaveTangent=-2.0542, tangentMode=RCTM_User, time=0.2, value=1}
         {arriveTangent=-0.4397, tangentMode=RCTM_User, time=0.8, value=0}
        ]}}})
   ) {module=Color002}
   &ScaleColor(
    Scale Alpha=0.7
    Scale RGB=&VectorFromFloat(Value=&FNTimeOfDayScalar(Night Scale=5, Morning Scale=1, Evening Scale=1))
    ScaleA=true
    ScaleRGB=true
   ) {enabled=false, module=ScaleColor001}
   &ScaleColor(
    Scale RGB=&VectorFromFloat(Value=&FloatFromCurve(
      FloatCurve=DataInterface() {propertyValues {Curve {keys [{interpMode=RCIM_Cubic, leaveTangent=1.4829, tangentMode=RCTM_Break, time=0, value=1}, <at 1 2.3>]}}}
      CurveIndex=$User.Hitted
     ))
    ScaleA=true
    ScaleRGB=true
   ) {enabled=false, module=ScaleColor002}
   &Color(Color=&ColorFromCurve(ColorCurve=DataInterface() {propertyValues {
       AlphaCurve {keys [<at 0 1 RCIM_Cubic>,<at 1 0 RCIM_Cubic>]}
       BlueCurve {keys [<at 0 1 RCIM_Cubic>,<at 0.2697 0.8863>,<at 0.6621 0.8058>,<at 1 0.7366 RCIM_Cubic>]}
       GreenCurve {keys [<at 0 0.3319 RCIM_Cubic>,<at 0.2697 0.0493>,<at 0.6621 0.7096>,<at 1 1 RCIM_Cubic>]}
       RedCurve {keys [<at 0 0 RCIM_Cubic>,<at 0.2697 1>,<at 0.6621 0.0014>,<at 1 0 RCIM_Cubic>]}
      }})) {enabled=false, module=Color001}
   &KillParticles(Kill Particles=&SetBoolByFloatComparison(A=$Particles.NormalizedAge, B=0.7))
   /SetVariables(Particles.MeshOrientation=&FindQuatBetweenTwoVectors(From Vector=<1 0 1>, To Vector=<0 0 1>)) {enabled=false, isSetParameters=true, module=SetVariables_#8}
  ]
  rendererClasses [&NiagaraMeshRendererProperties]
  renderers [
   {rendererClass=&NiagaraMeshRendererProperties, rendererIndex=0}
   {rendererClass=&NiagaraMeshRendererProperties, rendererIndex=1}
  ]
 }
 Sparkles {
  bEnabled=false
  emitterUpdateScript [
   &EmitterState(
    Enable Distance Culling=true
    Max Distance Response=::KillImmediately
    Module.bApplySpawnCountScaleByDistance=true
    Module.bUseMaxDistance=true
    Module.bUseMinDistance=false
    Scalability Mode=::NewEnumerator1
    Scale Spawn Count=true
    Spawn Count Scale By Distance=DataInterface() {propertyValues {Curve {keys [<at 0.3676 0.9657>,<at 0.9962 0.4982>]}}}
   )
   &SpawnBurst_Instantaneous(Spawn Count=&UniformRangedInt(Minimum=3, Maximum=20), Spawn Probability=0.85, Use Spawn Probability=false)
  ]
  particleSpawnScript [
   &InitializeParticle(
    Color=<rgba 1 1 1 0.5>
    Lifetime Max=0.5
    Lifetime Min=0.2
    Lifetime Mode=::NewEnumerator1
    Sprite Size Mode=::NewEnumerator1
    Uniform Sprite Size Max=20
    Uniform Sprite Size Min=10
    UsePositionOffset=true
    Write Material Random=true
   )
   &SphereLocation(
    MaskSpawnGroup=false
    Override Seed=false
    Sphere Radius=5
    Surface Only=false
    UseNonUniformScale=false
    UseOffset=false
    UseSphereOrientation=false
   ) {enabled=false, module=SphereLocation001}
   &ShapeLocation(MaskSpawnGroup=false, Non Uniform Scale=<5 5 1>, Override Seed=false, Ring Radius=5, Shape Primitive=::NewEnumerator4, UseOwnerScale=true)
   &AddVelocity(Velocity=&RandomRangeVector(Minimum=<0 0 236.6787>, Maximum=<1 1 500>)) {enabled=false}
  ]
  particleUpdateScript [
   &ParticleState()
   &ShapeLocation(
    MaskSpawnGroup=false
    Override Seed=false
    Ring Radius=&FloatFromCurve(
     FloatCurve=DataInterface() {propertyValues {Curve {
        defaultValue=1
        keys [{interpMode=RCIM_Cubic, tangentMode=RCTM_User, time=0, value=0}, {interpMode=RCIM_Cubic, tangentMode=RCTM_User, time=1, value=1}]
       }}}
     Scale Curve=45
    )
    Shape Primitive=::NewEnumerator4
    UseOwnerScale=true
   ) {module=ShapeLocation001}
   &GravityForce(Gravity=&VectorFromCurve(
     VectorCurve=DataInterface() {propertyValues {
       XCurve {keys [<at 0 1>,<at 1 0>]}
       YCurve {keys [<at 0 1>,<at 1 0>]}
       ZCurve {
        defaultValue=1
        keys [{interpMode=RCIM_Cubic, tangentMode=RCTM_User, time=0, value=0}, {interpMode=RCIM_Cubic, tangentMode=RCTM_User, time=1, value=1}]
       }
      }}
     Scale Curve=<1 1 3500>
    ))
   &Drag(Drag=&UniformRangedFloat(Minimum=0.2, Maximum=12), Use Linear Drag=true, Use Rotational Drag=false)
   &SolveForcesAndVelocity(Clamp Velocity=false, Limit Acceleration=false)
   &ScaleSpriteSize(Scale Factor=&Vector2DFromCurve(
     Vector2Curve=DataInterface() {propertyValues {
       XCurve {keys [
         {arriveTangent=9.231, interpMode=RCIM_Cubic, leaveTangent=9.231, tangentMode=RCTM_User, time=0, value=0.1}
         {arriveTangent=-0.1, interpMode=RCIM_Cubic, leaveTangent=-0.1, tangentMode=RCTM_User, time=0.18, value=0.8}
         {arriveTangent=-2.4373, interpMode=RCIM_Cubic, leaveTangent=-2.4373, tangentMode=RCTM_User, time=1, value=0}
        ]}
       YCurve {keys [
         {arriveTangent=9.7317, interpMode=RCIM_Cubic, leaveTangent=9.7317, tangentMode=RCTM_User, time=0, value=0.1}
         {arriveTangent=-0.1001, interpMode=RCIM_Cubic, leaveTangent=-0.1001, tangentMode=RCTM_User, time=0.18, value=0.8}
         {arriveTangent=-2.4373, interpMode=RCIM_Cubic, leaveTangent=-2.4373, tangentMode=RCTM_User, time=0.9988, value=0}
        ]}
      }}
     ScaleCurve=<1 1>
    ))
   &ScaleColor(
    Scale RGB=&MakeVectorFromLinearColorRGB(Color=&HueShiftLinearColor(
      Hue=&RandomRangeFloat(Minimum=-0.1, Maximum=0)
      Saturation=1
      Value=1
      Linear Color=&ColorFromCurve(ColorCurve=DataInterface() {propertyValues {
         AlphaCurve {keys [<at 0 1>,<at 1 0>]}
         BlueCurve {keys [<at 0.0061 0.701>,<at 0.33 1.35>,<at 0.85 0.5354>]}
         GreenCurve {keys [<at 0.0061 0.0719>,<at 0.33 0.2547>,<at 0.85 0.5809>]}
         RedCurve {keys [<at 0.0061 1>,<at 0.33 0>,<at 0.85 0.0119>]}
        }})
     ))
    ScaleA=true
    ScaleRGB=true
   )
   /SetVariables(
    Particles.Color2=&MultiplyLinearColor(
     Linear Color=$Particles.Color
     Scale Factor=&MakeFloat4FromVectorAndFloat("Vector (RGB)"=&VectorFromFloat(Value=7), "Float (Alpha)"=1)
    )
    Particles.SpriteSize2=&Multiply_Vector2D_ByFloat(Float=0.3, Vector2D=$Particles.SpriteSize)
   ) {isSetParameters=true, module=SetVariables_#9}
  ]
  rendererClasses [&NiagaraSpriteRendererProperties]
  renderers [
   {rendererClass=&NiagaraSpriteRendererProperties}
   {rendererClass=&NiagaraSpriteRendererProperties, rendererIndex=1}
  ]
 }
 Sparks001 {
  emitterUpdateScript [
   &EmitterState(
    Enable Distance Culling=true
    Max Distance Response=::KillImmediately
    Module.bApplySpawnCountScaleByDistance=true
    Module.bUseMaxDistance=true
    Module.bUseMinDistance=false
    Scalability Mode=::NewEnumerator1
    Scale Spawn Count=true
    Spawn Count Scale By Distance=DataInterface() {propertyValues {Curve {keys [<at 0.3676 0.9657>,<at 0.9962 0.4982>]}}}
   )
   &SpawnBurst_Instantaneous(Spawn Count=&UniformRangedInt(Minimum=3, Maximum=10), Use Spawn Probability=false)
  ]
  particleSpawnScript [
   &InitializeParticle(
    Color=&SelectLinearColorFromArray(
     Array Sampling Mode=::NewEnumerator0
     Direct Array Index=?
     Color Selection Array=DataInterface() {propertyValues {ColorData [<rgba 1.0237 0.195 1.3 5>, <rgba 1.1813 0.225 1.5 5>, <rgba 0 0.2667 0.4 5>, <rgba 0 0.6823 1.5 5>, <rgba 1.05 0 1.5 5>]}}
     Random Seed=0
    )
    Lifetime Max=0.45
    Lifetime Min=0.166
    Lifetime Mode=::NewEnumerator1
    Sprite Size Mode=::NewEnumerator1
    Uniform Sprite Size Max=15
    Uniform Sprite Size Min=8
    UsePositionOffset=false
    Write Material Random=true
   )
   &ShapeLocation(MaskSpawnGroup=false, Non Uniform Scale=<5 5 1>, Override Seed=false, Ring Radius=5, Shape Primitive=::NewEnumerator4, UseOwnerScale=true)
  ]
  particleUpdateScript [
   &ParticleState()
   &ScaleSpriteSize(Scale Factor=&Vector2DFromCurve(
     Vector2Curve=DataInterface() {propertyValues {
       XCurve {keys [
         {arriveTangent=9.231, interpMode=RCIM_Cubic, leaveTangent=9.231, tangentMode=RCTM_User, time=0, value=0.1}
         {arriveTangent=-0.1, interpMode=RCIM_Cubic, leaveTangent=-0.1, tangentMode=RCTM_User, time=0.18, value=0.8}
         {arriveTangent=-2.4373, interpMode=RCIM_Cubic, leaveTangent=-2.4373, tangentMode=RCTM_User, time=1, value=0}
        ]}
       YCurve {keys [
         {arriveTangent=9.7317, interpMode=RCIM_Cubic, leaveTangent=9.7317, tangentMode=RCTM_User, time=0, value=0.1}
         {arriveTangent=-0.1001, interpMode=RCIM_Cubic, leaveTangent=-0.1001, tangentMode=RCTM_User, time=0.18, value=0.8}
         {arriveTangent=-2.4373, interpMode=RCIM_Cubic, leaveTangent=-2.4373, tangentMode=RCTM_User, time=0.9988, value=0}
        ]}
      }}
     ScaleCurve=<1 1>
    ))
   &ShapeLocation(
    MaskSpawnGroup=false
    Override Seed=false
    Ring Radius=&FloatFromCurve(
     FloatCurve=DataInterface() {propertyValues {Curve {
        defaultValue=1
        keys [{interpMode=RCIM_Cubic, tangentMode=RCTM_User, time=0, value=0}, {interpMode=RCIM_Cubic, tangentMode=RCTM_User, time=1, value=1}]
       }}}
     Scale Curve=45
    )
    Shape Primitive=::NewEnumerator4
    UseOwnerScale=true
   ) {module=ShapeLocation001}
   &ScaleColor(
    Scale Alpha=0.7
    Scale RGB=&VectorFromFloat(Value=&FNTimeOfDayScalar(Night Scale=1, Morning Scale=1, Day Scale=5, Evening Scale=1))
    ScaleA=true
    ScaleRGB=true
   ) {module=ScaleColor003}
   &ScaleColor(
    Scale RGB=&MakeVectorFromLinearColorRGB(Color=&HueShiftLinearColor(
      Hue=&RandomRangeFloat(Minimum=-0.1, Maximum=0)
      Saturation=1
      Value=1
      Linear Color=&ColorFromCurve(ColorCurve=DataInterface() {propertyValues {
         AlphaCurve {keys [<at 0 1>,<at 1 0>]}
         BlueCurve {keys [<at 0.17 15>,<at 0.2011 2>,<at 0.33 1.35>,<at 0.85 1.1>]}
         GreenCurve {keys [<at 0.17 7.6606>,<at 0.2011 0.722>,<at 0.33 0.2547>,<at 0.85 0.3056>]}
         RedCurve {keys [<at 0.17 0>,<at 0.2011 0>,<at 0.33 0>,<at 0.85 0>]}
        }})
     ))
    ScaleA=true
    ScaleRGB=true
   ) {enabled=false}
   /SetVariables(
    Particles.Color2=&MultiplyLinearColor(
     Linear Color=$Particles.Color
     Scale Factor=&MakeFloat4FromVectorAndFloat("Vector (RGB)"=&VectorFromFloat(Value=7), "Float (Alpha)"=1)
    )
    Particles.SpriteSize2=&Multiply_Vector2D_ByFloat(Float=0.3, Vector2D=$Particles.SpriteSize)
   ) {isSetParameters=true, module=SetVariables_#9}
   &Drag(Drag=&UniformRangedFloat(Minimum=6, Maximum=12), Use Linear Drag=true, Use Rotational Drag=false)
   &GravityForce(Gravity=&VectorFromCurve(
     VectorCurve=DataInterface() {propertyValues {
       XCurve {keys [<at 0 1>,<at 1 0>]}
       YCurve {keys [<at 0 1>,<at 1 0>]}
       ZCurve {
        defaultValue=1
        keys [{interpMode=RCIM_Cubic, tangentMode=RCTM_User, time=0, value=0}, {interpMode=RCIM_Cubic, tangentMode=RCTM_User, time=1, value=1}]
       }
      }}
     Scale Curve=<1 1 3500>
    ))
   &SolveForcesAndVelocity(Clamp Velocity=false, Limit Acceleration=false)
   &FN_Sprite_Attributes(
    Color=&MakeLinearColorFromVectorAndFloat(
     "Vector (RGB)"=&MakeVectorFromLinearColorRGB(Color=&Multiply_LinearColorByFloat(
       Linear Color=&ColorFromCurve(
        ColorCurve=DataInterface() {propertyValues {
          AlphaCurve {keys [<at 0 1 RCIM_Cubic>]}
          BlueCurve {keys [<at 0.3173 1 RCIM_Cubic>,<at 0.3548 1>,<at 0.6928 1>,<at 0.7245 0.6864>]}
          GreenCurve {keys [<at 0.3173 0.1714 RCIM_Cubic>,<at 0.3548 0.3565>,<at 0.6928 0.2891>,<at 0.7245 0.1519>]}
          RedCurve {keys [<at 0.3173 0.3432 RCIM_Cubic>,<at 0.3548 0>,<at 0.6928 0>,<at 0.7245 1>]}
         }}
        CurveIndex=&RandomRangeFloat(Minimum=0, Maximum=1)
       )
       Float=&Lerp_Float(
        A=1
        B=20
        Alpha=&FloatFromCurve(FloatCurve=DataInterface() {propertyValues {Curve {keys [
             {arriveTangent=-0.4493, interpMode=RCIM_Cubic, leaveTangent=-0.4493, tangentMode=RCTM_User, time=0, value=1}
             {interpMode=RCIM_Cubic, tangentMode=RCTM_User, time=0.7, value=0}
            ]}}})
       )
      ))
     "Float (Alpha)"=&FloatFromCurve(FloatCurve=DataInterface() {propertyValues {Curve {keys [
          {interpMode=RCIM_Cubic, tangentMode=RCTM_User, time=0, value=1}
          {arriveTangent=-2.0246, interpMode=RCIM_Cubic, leaveTangent=-2.0246, tangentMode=RCTM_User, time=1, value=0}
         ]}}})
    )
    Create Color Attribute=true
    Create SpriteSize Attribute=true
    SpriteSize=&Multiply_Vector2D_ByFloat(
     Float=&Multiply_Float(
      A=&RandomRangeFloat(Minimum=0.5, Maximum=1.5)
      B=&FloatFromCurve(FloatCurve=DataInterface() {propertyValues {Curve {keys [
           {interpMode=RCIM_Cubic, tangentMode=RCTM_User, time=0.2, value=1}
           {arriveTangent=-2.4619, interpMode=RCIM_Cubic, leaveTangent=-2.4619, tangentMode=RCTM_User, time=1, value=0}
          ]}}})
     )
     Vector2D=<5 5>
    )
    Write to Module Namespace=true
   )
  ]
  rendererClasses [&NiagaraSpriteRendererProperties]
  renderers [
   {rendererClass=&NiagaraSpriteRendererProperties}
   {rendererClass=&NiagaraSpriteRendererProperties, rendererIndex=1}
   {rendererClass=&NiagaraSpriteRendererProperties, rendererIndex=2}
  ]
 }
}
summary {
 systemName=NS_Emote_PolarityWinn_feet_landing
 emitters [
  {bEnabled=false, emitterName=Rings, rendererClasses [&NiagaraMeshRendererProperties]}
  {bEnabled=false, emitterName=Sparkles, rendererClasses [&NiagaraSpriteRendererProperties]}
  {emitterName=BurstRing, rendererClasses [&NiagaraMeshRendererProperties]}
  {emitterName=Flare, rendererClasses [&NiagaraSpriteRendererProperties, &NiagaraLightRendererProperties]}
  {emitterName=Sparks001, rendererClasses [&NiagaraSpriteRendererProperties]}
 ]
 userVariables [
  User.bIsADS {type=NiagaraBool, defaultValue=true}
  User.bIsFrontEnd {type=NiagaraBool}
  User.bIsFrontEndPreview {type=NiagaraBool}
  User.BoneIndex {type=NiagaraFloat, defaultValue=0, description=""}
  User.DebugReactivity {type=NiagaraBool, defaultValue=true}
  User.FEOffset {type=NiagaraFloat, defaultValue=10, description=""}
  User.FEScale {type=NiagaraFloat, defaultValue=0.6, description=""}
  User.FrontEndLooping {type=NiagaraFloat, defaultValue=0, description=""}
  User.Hit {type=NiagaraBool, defaultValue=true}
  User.Hitted {type=NiagaraFloat, defaultValue=0, description=""}
  User.SpotlightColorArray {
   type=NiagaraDataInterfaceArrayColor
   description=""
   defaultValue {
    struct=NiagaraExt_VariableValue_DataInterface
    value {
     dataInterface=&self:NiagaraDataInterfaceArrayColor_0
     dataInterfaceClass=&NiagaraDataInterfaceArrayColor
    }
   }
  }
  User.SpotlightsPositionArray {
   type=NiagaraDataInterfaceArrayFloat3
   description=""
   defaultValue {
    struct=NiagaraExt_VariableValue_DataInterface
    value {
     dataInterface=&self:NiagaraDataInterfaceArrayFloat3_0
     dataInterfaceClass=&NiagaraDataInterfaceArrayFloat3
    }
   }
  }
 ]
}
