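# Niagara effect for the Polaritywinn emote, adding glowing auras and swirling shooting-star trails to the character's skeletal mesh.
# ShootingStarHead001 bursts two bone-sampled star sprites, launched to opposite sides by partition and curved by vortex, point and gravity forces.
# ShootingStarRibbon001 reads those heads to spawn fading color-shifting ribbons, and initializes a comet-sound audio player.
# aura_v2 pins two bone-following glow sprites; ribbon and aura brightness scale with time of day.
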NiagaraSystem /BRCosmetics/Animation/Game/MainPlayer/Emotes/Polaritywinn/FX/NS_Emote_Polaritywinn_hand_trail
dependencies {
 usedDataInterfaces [
  &NiagaraDataInterfaceCurve
  &NiagaraDataInterfaceVector2DCurve
  &NiagaraDataInterfaceColorCurve
  &NiagaraDataInterfaceArrayColor
  &NiagaraDataInterfaceAudioPlayer
  &NiagaraDataInterfaceParticleRead
 ]
 usedDynamicInputs [&RandomRangeVector2D, &FNTimeOfDayColor]
 usedModules [
  &SystemState
  &self:SystemSpawnScript.SystemScriptSource.SystemScriptGraph.NiagaraNodeAssignment_1.SetVariables_#1
  &self:SystemSpawnScript.SystemScriptSource.SystemScriptGraph.NiagaraNodeAssignment_0.SetVariables_#2
  &self:SystemSpawnScript.SystemScriptSource.SystemScriptGraph.NiagaraNodeAssignment_3.SetVariables_#3
 ]
}
emitters {
 ShootingStarHead001 {
  emitterUpdateScript [
   &EmitterState(Life Cycle Mode=::NewEnumerator1, Loop Behavior=::NewEnumerator1, Loop Duration=1.3, UseLoopDelay=false)
   &SpawnBurst_Instantaneous(Spawn Count=2, Use Spawn Probability=false, UseLoopCountLimit=false)
   &SpawnRate(
    Spawn Probability=0.6
    SpawnRate=&FloatFromCurve(
     FloatCurve=DataInterface() {propertyValues {Curve {keys [<at 0 0>,<at 0.075 1>,<at 0.3 1>,<at 0.5 0.7>,<at 0.75 0.7>,<at 0.8 0>]}}}
     CurveIndex=$Emitter.NormalizedLoopAge
     Scale Curve=3.35
    )
    Use Spawn Probability=true
   ) {enabled=false}
  ]
  particleSpawnScript [
   &PartitionParticles()
   &InitializeParticle(
    Color=&SelectLinearColorFromArray(
     Array Sampling Mode=::NewEnumerator0
     Direct Array Index=?
     Color Selection Array=DataInterface() {propertyValues {ColorData [<rgba 1.1813 0.225 1.5 5>, <rgba 0.0194 0.8077 1 5>, <rgba 0 0.6823 1.5 5>, <rgba 1.05 0 1.5 5>]}}
     Random Seed=0
    )
    Lifetime Max=1
    Lifetime Min=0.6
    Lifetime Mode=::NewEnumerator1
    Sprite Size Mode=::NewEnumerator1
    Uniform Sprite Size Max=0
    Uniform Sprite Size Min=0
    UsePositionOffset=false
    Write Material Random=true
   ) {module=InitializeParticle001}
   &InitialMeshOrientation(Mesh Orientation Mode=::NewEnumerator1, Rotation=<0 0 0.25>, Use Rotation Vector=true)
   &SkeletalMeshLocation(
    Bone / Socket Index=2
    Bone Sampling Mode=::NewEnumerator4
    "Inherit Velocity (Scale)"=0.1
    MaskSpawnGroup=false
    Orientation Sampling=::NewEnumerator1
    Sampled Position Offset=&MaskVector(Mask Source=1, Mask Parameter=$Particles.PartitionParticles.Partition, Mask=0, Masked Vector=<35 0 125>, Passthrough Vector=<-25 0 125>)
    Skeletal Mesh=$System.Skeletal Mesh
    Velocity Sampling=::NewEnumerator0
    VelocitySpeedLimit=false
   )
   &AddVelocity(Velocity=&MaskVector(Mask Source=1, Mask Parameter=$Particles.PartitionParticles.Partition, Mask=0, Masked Vector=<0 -200 -210>, Passthrough Vector=<0 250 -210>))
  ]
  particleUpdateScript [
   &ParticleState()
   &SpriteRotationRate(Rotation Rate=-3)
   &ScaleMeshSize(Scale Factor=&VectorFromFloat(Value=&FloatFromCurve(FloatCurve=DataInterface() {propertyValues {Curve {
         defaultValue=0
         keys [
          {arriveTangent=9.0725, interpMode=RCIM_Cubic, leaveTangent=9.0725, tangentMode=RCTM_User, time=0, value=0}
          {arriveTangent=-2.5249, interpMode=RCIM_Cubic, leaveTangent=-2.5249, tangentMode=RCTM_User, time=0.1408, value=0.9847}
          {arriveTangent=-0.0405, interpMode=RCIM_Cubic, leaveTangent=-0.0405, tangentMode=RCTM_User, time=1, value=0}
         ]
        }}})))
   &ScaleSpriteSize(Uniform Curve Sprite Scale=DataInterface() {propertyValues {Curve {
       defaultValue=0
       keys [
        {arriveTangent=9.0725, interpMode=RCIM_Cubic, leaveTangent=9.0725, tangentMode=RCTM_User, time=0, value=0}
        {arriveTangent=-2.5249, interpMode=RCIM_Cubic, leaveTangent=-2.5249, tangentMode=RCTM_User, time=0.1408, value=0.9847}
        {arriveTangent=-0.0405, interpMode=RCIM_Cubic, leaveTangent=-0.0405, tangentMode=RCTM_User, time=1, value=0}
       ]
      }}})
   &MeshRotationForce(Rotation=&RandomRangeVector(Minimum=<-15 0 0>, Maximum=<-25 0 0>))
   &PointForce(
    Force Origin=&AddVectorToPosition(
     Position=$Particles.SkeletalMeshLocation.SampledPosition
     Vector=&MakeVector(
      X=0
      Y=0
      Z=&FloatFromCurve(
       FloatCurve=DataInterface() {propertyValues {Curve {defaultValue=1, keys [{tangentMode=RCTM_User, time=0, value=0}, {tangentMode=RCTM_User, time=1, value=1}]}}}
       Scale Curve=100
      )
     )
    )
    Force Strength=&FloatFromCurve(
     FloatCurve=DataInterface() {propertyValues {Curve {defaultValue=1, keys [{tangentMode=RCTM_User, time=0, value=0}, {tangentMode=RCTM_User, time=0.4, value=1}]}}}
     Scale Curve=300
    )
    Module.Use Falloff=false
    Module.UseLinearForce=false
   )
   &VortexForce(
    Origin Pull Amount:NiagaraFloat=&FloatFromCurve(FloatCurve=DataInterface() {propertyValues {Curve {keys [<at 0 1>,<at 1 10>]}}}, Scale Curve=1000)
    UseOriginOffset=false
    Vortex Force Amount=&FloatFromCurve(
     FloatCurve=DataInterface() {propertyValues {Curve {defaultValue=1, keys [{tangentMode=RCTM_User, time=0, value=1}, {tangentMode=RCTM_User, time=1, value=2}]}}}
     Scale Curve=-1000
    )
   )
   &GravityForce(Gravity=&MakeVector(
     X=0
     Y=0
     Z=&FloatFromCurve(
      FloatCurve=DataInterface() {propertyValues {Curve {defaultValue=1, keys [{tangentMode=RCTM_User, time=0, value=0}, {tangentMode=RCTM_User, time=1, value=1}]}}}
      Scale Curve=200
     )
    ))
   &SolveForcesAndVelocity(Clamp Velocity=false, Limit Acceleration=false)
   &ScaleColor(
    Scale Alpha=&FloatFromCurve(FloatCurve=DataInterface() {propertyValues {Curve {
        defaultValue=0
        keys [
         {arriveTangent=9.0725, interpMode=RCIM_Cubic, leaveTangent=9.0725, tangentMode=RCTM_User, time=0, value=0}
         {arriveTangent=-2.5249, interpMode=RCIM_Cubic, leaveTangent=-2.5249, tangentMode=RCTM_User, time=0.1408, value=0.9847}
         {arriveTangent=-0.0405, interpMode=RCIM_Cubic, leaveTangent=-0.0405, tangentMode=RCTM_User, time=1, value=0}
        ]
       }}})
    Scale RGB=&VectorFromFloat(Value=&FloatFromCurve(FloatCurve=DataInterface() {propertyValues {Curve {
         defaultValue=0
         keys [
          {arriveTangent=9.0725, interpMode=RCIM_Cubic, leaveTangent=9.0725, tangentMode=RCTM_User, time=0, value=0}
          {arriveTangent=-2.5249, interpMode=RCIM_Cubic, leaveTangent=-2.5249, tangentMode=RCTM_User, time=0.1408, value=0.9847}
         ]
        }}}))
    ScaleA=true
    ScaleRGB=true
   )
   /SetVariables(
    Particles.InnerStarCO=5
    Particles.InnerStarSize:Vector3f=&VectorFromFloat(Value=&FloatFromCurve(
      FloatCurve=DataInterface() {propertyValues {Curve {
         defaultValue=0
         keys [
          {arriveTangent=9.0725, interpMode=RCIM_Cubic, leaveTangent=9.0725, tangentMode=RCTM_User, time=0, value=0}
          {arriveTangent=-2.5249, interpMode=RCIM_Cubic, leaveTangent=-2.5249, tangentMode=RCTM_User, time=0.1408, value=0.9847}
          {arriveTangent=-0.0405, interpMode=RCIM_Cubic, leaveTangent=-0.0405, tangentMode=RCTM_User, time=1, value=0}
         ]
        }}}
      Scale Curve=&RandomRangeFloat(Minimum=0.3, Maximum=0.4)
     ))
   ) {isSetParameters=true}
  ]
  rendererClasses [&NiagaraSpriteRendererProperties]
  renderers [{rendererClass=&NiagaraSpriteRendererProperties}]
 }
 ShootingStarRibbon001 {
  emitterSpawnScript [&InitializeAudioPlayer(Audio Player=DataInterface() {propertyValues {
      Attenuation=null
      Concurrency=null
      MaxPlaysPerTick=10
      SoundToPlay=/BRCosmetics/Sounds/Emotes/WishingStar/MSS_Emote_Foley_WishingStar_CometBys
      bAllowLoopingOneShotSounds=false
      bLimitPlaysPerTick=true
      bOnlyActiveDuringGameplay=false
      bStopWhenComponentIsDestroyed=true
      ConfigurationUserParameter {parameter {name=None, typeDefHandle {registeredTypeIndex=1}, varData []}}
      ParameterNames []
     }})]
  emitterUpdateScript [
   &EmitterState(Life Cycle Mode=::NewEnumerator1, Loop Behavior=::NewEnumerator1, Loop Duration=1.3, UseLoopDelay=false)
   &SpawnParticlesFromOtherEmitter(
    Attribute Reader=DataInterface() {propertyValues {EmitterBinding {bindingMode=Other, emitterName=ShootingStarHead001}}}
    Spawn Rate=60
    Use Maximum Age=false
    Use Minimum Age=false
   )
  ]
  particleSpawnScript [
   &InitializeRibbon(
    Color=<rgba 1 1 1 3>
    Lifetime=&FloatFromCurve(FloatCurve=DataInterface() {propertyValues {Curve {keys [<at 0 1>,<at 1 0>]}}})
    Position Mode:ENiagara_PositionInitializationMode=::NewEnumerator0
    Ribbon Facing Mode:ENiagara_UnsetDirectSet=::NewEnumerator1
    Ribbon Width=40
    Write Material Random=true
   )
   &InitializeParticle(UsePositionOffset=false, Write Material Random=true)
   &SampleParticlesFromOtherEmitter(
    Age Sampling:ENiagara_DirectReadSamplingMode=::NewEnumerator0
    Applied Position Offset:Vector3f=<-2 0 0>
    Color Sampling=::NewEnumerator1
    Particle ID Sampling=::NewEnumerator1
   )
  ]
  particleUpdateScript [
   &ParticleState()
   &SolveForcesAndVelocity(Clamp Velocity=false, Limit Acceleration=false)
   &ScaleColor(
    Scale Alpha=&FloatFromCurve(
     FloatCurve=DataInterface() {propertyValues {Curve {keys [
         {arriveTangent=3.1524, interpMode=RCIM_Cubic, leaveTangent=3.1524, tangentMode=RCTM_User, time=0, value=0}
         {arriveTangent=0.3962, interpMode=RCIM_Cubic, leaveTangent=0.3962, tangentMode=RCTM_User, time=0.3, value=1}
         {arriveTangent=-1.7441, interpMode=RCIM_Cubic, leaveTangent=-1.7441, tangentMode=RCTM_User, time=0.4, value=0.9}
         {arriveTangent=0.0107, interpMode=RCIM_Cubic, leaveTangent=0.0107, tangentMode=RCTM_User, time=0.95, value=0}
        ]}}}
     CurveIndex=$Emitter.Age
    )
    Scale RGB=&MakeVectorFromLinearColorRGB(Color=&ColorFromCurve(ColorCurve=DataInterface() {propertyValues {
        AlphaCurve {keys [<at 0 0 RCIM_Cubic>,<at 0.5 1>,<at 0.9949 0>]}
        BlueCurve {keys [{arriveTangent=4.0072, interpMode=RCIM_Cubic, time=0, value=0.8361}, <at 0.1526 1>, <at 0.5036 1>, <at 0.6607 1>, <at 1 1 RCIM_Cubic>]}
        GreenCurve {keys [{arriveTangent=-0.4007, interpMode=RCIM_Cubic, time=0, value=1}, <at 0.1526 0>, <at 0.5036 0>, <at 0.6607 0.2>, <at 1 0.4 RCIM_Cubic>]}
        RedCurve {keys [
          {arriveTangent=-11.8213, interpMode=RCIM_Cubic, time=0, value=0}
          <at 0.1526 0.2883>
          <at 0.5036 0.6667>
          <at 0.6607 0.2>
          <at 1 0.1 RCIM_Cubic>
         ]}
       }}))
    ScaleA=true
    ScaleRGB=true
   )
   &ScaleColor(
    Scale Alpha=0.7
    Scale RGB=&VectorFromFloat(Value=&FNTimeOfDayScalar(Night Scale=25, Morning Scale=1, Day Scale=25, Evening Scale=1))
    ScaleA=false
    ScaleRGB=true
   ) {module=ScaleColor001}
   &ScaleRibbonWidth(Ribbon Width Scale=&FloatFromCurve(FloatCurve=DataInterface() {propertyValues {Curve {
        defaultValue=0.5
        keys [{interpMode=RCIM_Cubic, tangentMode=RCTM_User, time=0, value=1}, {interpMode=RCIM_Cubic, tangentMode=RCTM_User, time=1, value=0}]
       }}}))
   &DynamicMaterialParameters(
    Index 0 Param 1=&FloatFromCurve(FloatCurve=DataInterface() {propertyValues {Curve {keys [
         {arriveTangent=0.2015, interpMode=RCIM_Cubic, leaveTangent=0.2015, tangentMode=RCTM_User, time=0, value=1}
         {arriveTangent=-1.5941, interpMode=RCIM_Cubic, leaveTangent=-1.5941, tangentMode=RCTM_User, time=0.3, value=0.5}
         {interpMode=RCIM_Cubic, tangentMode=RCTM_User, time=1, value=0}
        ]}}})
    Index 0 Param 3=-1
    Module.Index0Float0Write=true
    Module.Index0Float1Write=false
    Module.Index0Float2Write=true
    Module.Index0Float3Write=true
   ) {module=DynamicMaterialParameters001}
   &ScaleColor(
    Scale Alpha=&FloatFromCurve(FloatCurve=DataInterface() {propertyValues {Curve {keys [<at 0 0>,<at 1 1>]}}})
    ScaleA=true
    ScaleRGB=true
   ) {module=ScaleColor002}
  ]
  rendererClasses [&NiagaraRibbonRendererProperties]
  renderers [
   {rendererClass=&NiagaraRibbonRendererProperties}
   {rendererClass=&NiagaraRibbonRendererProperties, rendererIndex=1}
   {rendererClass=&NiagaraRibbonRendererProperties, rendererIndex=2}
  ]
 }
 aura_v2 {
  emitterUpdateScript [
   &EmitterState(Life Cycle Mode=::NewEnumerator1, Loop Behavior=::NewEnumerator1, Loop Duration=2.4, UseLoopDelay=false)
   &SpawnBurst_Instantaneous(Spawn Count=2, Use Spawn Probability=false, UseLoopCountLimit=false)
  ]
  particleSpawnScript [
   &PartitionParticles()
   &InitializeParticle(
    Color=<rgba 1 1 1 0.4>
    Lifetime=0.9
    Sprite Rotation Mode=::NewEnumerator1
    Sprite Size Mode=::NewEnumerator3
    Sprite UV Mode=::NewEnumerator1
    Uniform Sprite Size=35
    UsePositionOffset=false
    Write Material Random=true
   )
   &CameraOffset(Camera Offset Amount=5)
   /SetVariables(Particles.PivotOffset:Vector2f=/Niagara/DynamicInputs/Helpers/MaskVector2d(
     Mask Source:NiagaraInt32=0
     Mask Parameter:NiagaraInt32=$Particles.PartitionParticles.Partition
     Mask:NiagaraInt32=1
     Masked Vector2D:Vector2f=<0.5 0.4>
     Passthrough Vector2D:Vector2f=<0.5 0.6>
    )) {isSetParameters=true}
  ]
  particleUpdateScript [
   &SkeletalMeshLocation(
    Bone / Socket Index=$Particles.PartitionParticles.Partition
    Bone Sampling Mode=::NewEnumerator4
    MaskSpawnGroup=false
    Orientation Sampling=::NewEnumerator1
    Sampled Position Offset=&MaskVector(Mask Source=1, Mask Parameter=$Particles.PartitionParticles.Partition, Mask=0, Masked Vector=<0 0 0>, Passthrough Vector=<-15 1 1>)
    Skeletal Mesh=$System.Skeletal Mesh
   )
   &SpriteFacingAndAlignment(
    Sprite Alignment=&RotateVectorByQuaternion(Vector To Rotate=<1 0 0>, Quaternion=$Particles.SkeletalMeshLocation.SampledRotation)
    WriteAlignment=true
    WriteFacing=false
   )
   &ParticleState()
   &ScaleSpriteSize(
    Non-Uniform Curve Scale=<1.8 1.8>
    Non-Uniform Curve Sprite Scale=DataInterface() {propertyValues {
      XCurve {keys [{interpMode=RCIM_Cubic, tangentMode=RCTM_User, time=0, value=0}, {interpMode=RCIM_Cubic, tangentMode=RCTM_User, time=0.05, value=0.6085}]}
      YCurve {keys [{interpMode=RCIM_Cubic, tangentMode=RCTM_User, time=0, value=0}, {interpMode=RCIM_Cubic, tangentMode=RCTM_User, time=0.05, value=0.9144}]}
     }}
    Scale Sprite Size Mode=::NewEnumerator3
   ) {module=ScaleSpriteSize001}
   &ScaleColor(
    Scale Alpha=&FloatFromCurve(FloatCurve=DataInterface() {propertyValues {Curve {keys [
         <at 0 0 RCIM_Cubic>
         {arriveTangent=1.0349, interpMode=RCIM_Cubic, leaveTangent=1.0349, time=0.1424, value=1}
         {interpMode=RCIM_Cubic, tangentMode=RCTM_User, time=0.5993, value=0.6202}
         {interpMode=RCIM_Cubic, tangentMode=RCTM_User, time=1, value=0}
        ]}}})
    Scale RGB=&MakeVectorFromLinearColorRGB(Color=&ColorFromCurve(ColorCurve=DataInterface() {propertyValues {
        AlphaCurve {keys [<at 0.0867 1 RCIM_Cubic>,<at 0.5 0.4335>,<at 0.9101 0>]}
        BlueCurve {keys [{arriveTangent=4.0072, interpMode=RCIM_Cubic, time=0, value=0.8361}, <at 0.1526 1>, <at 0.5036 1>, <at 0.6607 1>, <at 1 1 RCIM_Cubic>]}
        GreenCurve {keys [{arriveTangent=-0.4007, interpMode=RCIM_Cubic, time=0, value=1}, <at 0.1526 0>, <at 0.5036 0>, <at 0.6607 0.2>, <at 1 0.4 RCIM_Cubic>]}
        RedCurve {keys [
          {arriveTangent=-11.8213, interpMode=RCIM_Cubic, time=0, value=0}
          <at 0.1526 0.2883>
          <at 0.5036 0.6667>
          <at 0.6607 0.2>
          <at 1 0.1 RCIM_Cubic>
         ]}
       }}))
    ScaleA=true
    ScaleRGB=true
   )
   /SetVariables(
    Particles.GlowColor=&ColorFromCurve(ColorCurve=DataInterface() {propertyValues {
       AlphaCurve {keys [<at 0 0 RCIM_Cubic>, {arriveTangent=0.6, interpMode=RCIM_Cubic, leaveTangent=0.6, time=0.4, value=0.3}, <at 0.5 0.3>, <at 0.75 0>]}
       BlueCurve {keys [<at 0 2 RCIM_Cubic>, {interpMode=RCIM_Cubic, tangentMode=RCTM_User, time=0.2852, value=3}, <at 0.5204 2>, <at 0.7867 0.1833>, <at 1 1>]}
       GreenCurve {keys [<at 0 0 RCIM_Cubic>, {interpMode=RCIM_Cubic, tangentMode=RCTM_User, time=0.2852, value=1.4319}, <at 0.5204 0>, <at 0.7867 0.02>, <at 1 0.05>]}
       RedCurve {keys [
         <at 0 1.484 RCIM_Cubic>
         {interpMode=RCIM_Cubic, tangentMode=RCTM_User, time=0.2852, value=0}
         <at 0.5204 0.8464>
         <at 0.7867 1>
         <at 1 0.6833>
        ]}
      }})
    Particles.GlowSize=&Multiply_Vector2D_ByFloat(Float=55, Vector2D=<1 1>)
   ) {isSetParameters=true}
   &ScaleColor(
    Scale Alpha=0.7
    Scale RGB=&VectorFromFloat(Value=&FNTimeOfDayScalar(Night Scale=25, Morning Scale=1, Day Scale=25, Evening Scale=1))
    ScaleA=true
    ScaleRGB=true
   ) {module=ScaleColor001}
   &SolveForcesAndVelocity(Clamp Velocity=false, Limit Acceleration=false)
  ]
  rendererClasses [&NiagaraSpriteRendererProperties]
  renderers [
   {rendererClass=&NiagaraSpriteRendererProperties}
   {rendererClass=&NiagaraSpriteRendererProperties, rendererIndex=1}
   {rendererClass=&NiagaraSpriteRendererProperties, rendererIndex=2}
  ]
 }
}
summary {
 systemName=NS_Emote_Polaritywinn_hand_trail
 emitters [
  {emitterName=aura_v2, rendererClasses [&NiagaraSpriteRendererProperties]}
  {emitterName=ShootingStarHead001, rendererClasses [&NiagaraSpriteRendererProperties]}
  {emitterName=ShootingStarRibbon001, rendererClasses [&NiagaraRibbonRendererProperties]}
 ]
 userVariables [
  User.Anim_Curve_InitialOpacity {type=NiagaraFloat, defaultValue=1, description=""}
  User.Anim_Curve_SpawnRate {type=NiagaraFloat, defaultValue=0, description=""}
  User.bDrums {type=NiagaraFloat, defaultValue=0, description=""}
  User.Skeletal Mesh {
   type=NiagaraDataInterfaceSkeletalMesh
   defaultValue {
    struct=NiagaraExt_VariableValue_DataInterface
    value {dataInterface=&self:NiagaraDataInterfaceSkeletalMesh_0}
   }
  }
 ]
}
