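# Niagara effect for the Polaritywinn player emote that adds glowing cyan-to-violet trails and auras to both hands.
# PartitionParticles splits particles between two skeletal mesh bones or sockets, with masked per-side offsets.
# The active emitters are a burst aura that resamples its bone every update and a continuously spawned ribbon faded by emitter age.
# Time-of-day scalars boost brightness; the older ribbon and combined trail-aura emitters are disabled.
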NiagaraSystem /BRCosmetics/Animation/Game/MainPlayer/Emotes/Polaritywinn/FX/NS_Emote_Polaritywinn_hand_trail_V2
dependencies {
 usedDataInterfaces [
  &NiagaraDataInterfaceCurve
  &NiagaraDataInterfaceVector2DCurve
  &NiagaraDataInterfaceColorCurve
  &NiagaraDataInterfacePlatformSet
 ]
 usedDynamicInputs [
  &RandomRangeVector2D
  &RandomRangeFloat
  &FNTimeOfDayColor
 ]
 usedModules [
  &SystemState
  &self:SystemSpawnScript.SystemScriptSource.SystemScriptGraph.NiagaraNodeAssignment_1.SetVariables_#1
  &self:SystemSpawnScript.SystemScriptSource.SystemScriptGraph.NiagaraNodeAssignment_0.SetVariables_#2
  &self:SystemSpawnScript.SystemScriptSource.SystemScriptGraph.NiagaraNodeAssignment_3.SetVariables_#3
 ]
}
emitters {
 New_version {
  bEnabled=false
  emitterUpdateScript [
   &EmitterState(Life Cycle Mode=::NewEnumerator1, Loop Behavior=::NewEnumerator1, Loop Delay=0.3, Loop Duration=1, UseLoopDelay=false)
   &SpawnRate(SpawnRate=30, Use Spawn Probability=false)
  ]
  particleSpawnScript [
   &InitializeParticle(Color Mode=::NewEnumerator0, Lifetime=0.25, UsePositionOffset=false, Write Material Random=true)
   &PartitionParticles()
   &SkeletalMeshLocation(
    Bone / Socket Index=$Particles.PartitionParticles.Partition
    Bone Sampling Mode=::NewEnumerator4
    "Inherit Velocity (Scale)"=0.1
    MaskSpawnGroup=false
    Orientation Sampling=::NewEnumerator1
    Sampled Position Offset=&MaskVector(
     Mask Source=1
     Mask Parameter=$Particles.PartitionParticles.Partition
     Mask=0
     Masked Vector=<-4.5799 2.1042 0>
     Passthrough Vector=<-10.5293 -1.044 1.4884>
    )
    Skeletal Mesh=$System.Skeletal Mesh
    Velocity Sampling=::NewEnumerator0
    VelocitySpeedLimit=false
   )
   &InitializeRibbon(
    Color=<rgba 1 1 1 3>
    Lifetime=&FloatFromCurve(FloatCurve=DataInterface() {propertyValues {Curve {keys [<at 0 1>,<at 1 0>]}}})
    Position Mode:ENiagara_PositionInitializationMode=::NewEnumerator0
    Ribbon Facing Mode:ENiagara_UnsetDirectSet=::NewEnumerator1
    Ribbon Facing Vector:Vector3f=&RotateVectorByQuaternion(Vector To Rotate=<1 0 0>, Quaternion=$Particles.SkeletalMeshLocation.SampledRotation)
    Ribbon Width=40
    Write Material Random=true
   )
   &AddVelocity(
    Velocity=<0 0 -15>
    Velocity Speed Scale=&FloatFromCurve(FloatCurve=DataInterface() {propertyValues {Curve {keys [<at 0 1>,<at 0.2 0>,<at 0.5 0.55>,<at 1 0>]}}})
   ) {enabled=false}
   &PointForce(
    Linear Force Vector:Vector3f=/Niagara/DynamicInputs/Helpers/MaskPosition(
     Mask Parameter:NiagaraInt32=$Particles.PartitionParticles.Partition
     Mask:NiagaraInt32=0
     Masked Position:NiagaraPosition=<1 0 0>
     Passthrough Position:NiagaraPosition=<-1 0 0>
    )
    Module.Use Falloff=true
    Module.UseLinearForce=true
   ) {enabled=false}
   &ApplyInitialForces()
  ]
  particleUpdateScript [
   &ParticleState()
   &InheritSourceMovement(ApplyForce=false, ApplyPosition=true, ApplyPositionFalloff=false)
   &DynamicMaterialParameters(
    Index 0 Param 1=&FloatFromCurve(FloatCurve=DataInterface() {propertyValues {Curve {keys [
         {arriveTangent=0.2015, interpMode=RCIM_Cubic, leaveTangent=0.2015, tangentMode=RCTM_User, time=0, value=1}
         {arriveTangent=-1.5941, interpMode=RCIM_Cubic, leaveTangent=-1.5941, tangentMode=RCTM_User, time=0.3, value=0.5}
         {interpMode=RCIM_Cubic, tangentMode=RCTM_User, time=1, value=0}
        ]}}})
    Index 0 Param 3=-1
    Module.Index0Float0Write=true
    Module.Index0Float1Write=false
    Module.Index0Float2Write=true
    Module.Index0Float3Write=true
   ) {module=DynamicMaterialParameters001}
   &ScaleColor(
    Scale Alpha=&FloatFromCurve(
     FloatCurve=DataInterface() {propertyValues {Curve {keys [
         {arriveTangent=3.1524, interpMode=RCIM_Cubic, leaveTangent=3.1524, tangentMode=RCTM_User, time=0, value=0}
         {arriveTangent=0.3962, interpMode=RCIM_Cubic, leaveTangent=0.3962, tangentMode=RCTM_User, time=0.3, value=1}
         {arriveTangent=-1.7441, interpMode=RCIM_Cubic, leaveTangent=-1.7441, tangentMode=RCTM_User, time=0.47, value=0.95}
         {arriveTangent=-0.4506, interpMode=RCIM_Cubic, leaveTangent=-0.4506, tangentMode=RCTM_User, time=0.87, value=0}
        ]}}}
     CurveIndex=$Emitter.Age
    )
    Scale RGB=&MakeVectorFromLinearColorRGB(Color=&ColorFromCurve(ColorCurve=DataInterface() {propertyValues {
        AlphaCurve {keys [<at 0 0.3116 RCIM_Cubic>,<at 0.5 1>,<at 0.9949 0>]}
        BlueCurve {keys [{arriveTangent=4.0072, interpMode=RCIM_Cubic, time=0, value=0.8361}, <at 0.1526 1>, <at 0.5036 1>, <at 0.6607 1>, <at 1 1 RCIM_Cubic>]}
        GreenCurve {keys [{arriveTangent=-0.4007, interpMode=RCIM_Cubic, time=0, value=1}, <at 0.1526 0>, <at 0.5036 0>, <at 0.6607 0.2>, <at 1 0.4 RCIM_Cubic>]}
        RedCurve {keys [
          {arriveTangent=-11.8213, interpMode=RCIM_Cubic, time=0, value=0}
          <at 0.1526 0.2883>
          <at 0.5036 0.6667>
          <at 0.6607 0.2>
          <at 1 0.1 RCIM_Cubic>
         ]}
       }}))
    ScaleA=true
    ScaleRGB=true
   )
   &ScaleColor(
    Scale Alpha=0.7
    Scale RGB=&VectorFromFloat(Value=&FNTimeOfDayScalar(Night Scale=25, Morning Scale=1, Day Scale=25, Evening Scale=1))
    ScaleA=false
    ScaleRGB=true
   ) {module=ScaleColor001}
   &ScaleRibbonWidth(Ribbon Width Scale=&FloatFromCurve(FloatCurve=DataInterface() {propertyValues {Curve {
        defaultValue=0.5
        keys [{interpMode=RCIM_Cubic, tangentMode=RCTM_User, time=0, value=1}, {interpMode=RCIM_Cubic, tangentMode=RCTM_User, time=1, value=0}]
       }}}))
   &SolveForcesAndVelocity(Clamp Velocity=false, Limit Acceleration=false)
  ]
  rendererClasses [&NiagaraRibbonRendererProperties]
  renderers [
   {rendererClass=&NiagaraRibbonRendererProperties}
   {rendererClass=&NiagaraRibbonRendererProperties, rendererIndex=1}
   {rendererClass=&NiagaraRibbonRendererProperties, rendererIndex=2}
   {rendererClass=&NiagaraRibbonRendererProperties, rendererIndex=3}
   {rendererClass=&NiagaraRibbonRendererProperties, rendererIndex=4}
   {rendererClass=&NiagaraRibbonRendererProperties, rendererIndex=5}
  ]
 }
 aura_No_flare {
  emitterUpdateScript [
   &EmitterState(Life Cycle Mode=::NewEnumerator1, Loop Behavior=::NewEnumerator1, Loop Duration=2.4, UseLoopDelay=false)
   &SpawnBurst_Instantaneous(Spawn Count=2, Use Spawn Probability=false, UseLoopCountLimit=false)
  ]
  particleSpawnScript [
   &PartitionParticles()
   &InitializeParticle(
    Color=<rgba 1 1 1 0.4>
    Lifetime=0.9
    Ribbon Width Mode=::NewEnumerator1
    Sprite Rotation Mode=::NewEnumerator1
    Sprite Size Mode=::NewEnumerator3
    Sprite UV Mode=::NewEnumerator1
    Uniform Sprite Size=35
    UsePositionOffset=false
    Write Material Random=true
   )
   &CameraOffset(Camera Offset Amount=5)
   /SetVariables(Particles.PivotOffset:Vector2f=/Niagara/DynamicInputs/Helpers/MaskVector2d(
     Mask Source:NiagaraInt32=0
     Mask Parameter:NiagaraInt32=$Particles.PartitionParticles.Partition
     Mask:NiagaraInt32=1
     Masked Vector2D:Vector2f=<0.5 0.4>
     Passthrough Vector2D:Vector2f=<0.5 0.6>
    )) {isSetParameters=true}
   &AddVelocity(Velocity=<0 0 0>) {enabled=false}
  ]
  particleUpdateScript [
   &SkeletalMeshLocation(
    Bone / Socket Index=$Particles.PartitionParticles.Partition
    Bone Sampling Mode=::NewEnumerator4
    MaskSpawnGroup=false
    Orientation Sampling=::NewEnumerator1
    Sampled Position Offset=&MaskVector(Mask Source=1, Mask Parameter=$Particles.PartitionParticles.Partition, Mask=0, Masked Vector=<0 0 0>, Passthrough Vector=<-15 0 0>)
    Skeletal Mesh=$System.Skeletal Mesh
   )
   &SpriteFacingAndAlignment(
    Sprite Alignment=&RotateVectorByQuaternion(Vector To Rotate=<1 0 0>, Quaternion=$Particles.SkeletalMeshLocation.SampledRotation)
    WriteAlignment=true
    WriteFacing=false
   )
   &ParticleState()
   &ScaleSpriteSize(
    Non-Uniform Curve Scale=<1.2 1.2>
    Non-Uniform Curve Sprite Scale=DataInterface() {propertyValues {
      XCurve {keys [{interpMode=RCIM_Cubic, tangentMode=RCTM_User, time=0, value=0}, {interpMode=RCIM_Cubic, tangentMode=RCTM_User, time=0.05, value=0.6085}]}
      YCurve {keys [{interpMode=RCIM_Cubic, tangentMode=RCTM_User, time=0, value=0}, {interpMode=RCIM_Cubic, tangentMode=RCTM_User, time=0.05, value=0.9144}]}
     }}
    Scale Sprite Size Mode=::NewEnumerator3
   ) {module=ScaleSpriteSize001}
   &ScaleColor(
    Scale Alpha=&FloatFromCurve(
     FloatCurve=DataInterface() {propertyValues {Curve {keys [
         <at 0 0 RCIM_Cubic>
         {arriveTangent=1.0349, interpMode=RCIM_Cubic, leaveTangent=1.0349, time=0.1424, value=1}
         {interpMode=RCIM_Cubic, tangentMode=RCTM_User, time=0.5993, value=0.6202}
         {interpMode=RCIM_Cubic, tangentMode=RCTM_User, time=1, value=0}
        ]}}}
     Scale Curve=0.5
    )
    Scale RGB=&MakeVectorFromLinearColorRGB(Color=&ColorFromCurve(ColorCurve=DataInterface() {propertyValues {
        AlphaCurve {keys [<at 0.0867 1 RCIM_Cubic>,<at 0.5 0.4335>,<at 0.9101 0>]}
        BlueCurve {keys [{arriveTangent=4.0072, interpMode=RCIM_Cubic, time=0, value=0.8361}, <at 0.1526 1>, <at 0.5036 1>, <at 0.6607 1>, <at 1 1 RCIM_Cubic>]}
        GreenCurve {keys [{arriveTangent=-0.4007, interpMode=RCIM_Cubic, time=0, value=1}, <at 0.1526 0>, <at 0.5036 0>, <at 0.6607 0.2>, <at 1 0.4 RCIM_Cubic>]}
        RedCurve {keys [
          {arriveTangent=-11.8213, interpMode=RCIM_Cubic, time=0, value=0}
          <at 0.1526 0.2883>
          <at 0.5036 0.6667>
          <at 0.6607 0.2>
          <at 1 0.1 RCIM_Cubic>
         ]}
       }}))
    ScaleA=true
    ScaleRGB=true
   )
   /SetVariables(
    Particles.GlowColor=&ColorFromCurve(ColorCurve=DataInterface() {propertyValues {
       AlphaCurve {keys [<at 0 0 RCIM_Cubic>, {arriveTangent=0.6, interpMode=RCIM_Cubic, leaveTangent=0.6, time=0.4, value=0.3}, <at 0.5 0.3>, <at 0.75 0>]}
       BlueCurve {keys [<at 0 2 RCIM_Cubic>, {interpMode=RCIM_Cubic, tangentMode=RCTM_User, time=0.2852, value=3}, <at 0.5204 2>, <at 0.7867 0.1833>, <at 1 1>]}
       GreenCurve {keys [<at 0 0 RCIM_Cubic>, {interpMode=RCIM_Cubic, tangentMode=RCTM_User, time=0.2852, value=1.4319}, <at 0.5204 0>, <at 0.7867 0.02>, <at 1 0.05>]}
       RedCurve {keys [
         <at 0 1.484 RCIM_Cubic>
         {interpMode=RCIM_Cubic, tangentMode=RCTM_User, time=0.2852, value=0}
         <at 0.5204 0.8464>
         <at 0.7867 1>
         <at 1 0.6833>
        ]}
      }})
    Particles.GlowSize=&Multiply_Vector2D_ByFloat(Float=55, Vector2D=<1 1>)
   ) {enabled=false, isSetParameters=true}
   &ScaleColor(
    Scale Alpha=0.7
    Scale RGB=&VectorFromFloat(Value=&FNTimeOfDayScalar(Night Scale=25, Morning Scale=1, Day Scale=25, Evening Scale=1))
    ScaleA=true
    ScaleRGB=true
   ) {module=ScaleColor001}
   &SolveForcesAndVelocity(Clamp Velocity=false, Limit Acceleration=false)
  ]
  rendererClasses [
   &NiagaraSpriteRendererProperties
   &NiagaraMeshRendererProperties
   &NiagaraRibbonRendererProperties
  ]
  renderers [
   {rendererClass=&NiagaraSpriteRendererProperties}
   {rendererClass=&NiagaraSpriteRendererProperties, rendererIndex=1}
   {rendererClass=&NiagaraSpriteRendererProperties, rendererIndex=2}
   {rendererClass=&NiagaraMeshRendererProperties, rendererIndex=3}
   {rendererClass=&NiagaraRibbonRendererProperties, rendererIndex=4}
  ]
 }
 hAND_TRAIL_AND_AURA {
  bEnabled=false
  emitterUpdateScript [
   &EmitterState(Life Cycle Mode=::NewEnumerator1, Loop Behavior=::NewEnumerator1, Loop Delay=0.3, Loop Duration=1, UseLoopDelay=false)
   &SpawnBurst_Instantaneous(Spawn Count=2, Spawn Group=1, Use Spawn Probability=false, UseLoopCountLimit=false)
   &SpawnRate(SpawnRate=30, Use Spawn Probability=false)
  ]
  particleSpawnScript [
   &InitializeParticle(
    Color=<rgba 1 1 1 0.5>
    Sprite Rotation Mode=::NewEnumerator1
    Sprite Size Mode=::NewEnumerator3
    Uniform Sprite Size=35
    UsePositionOffset=false
    Write Material Random=true
   )
   &PartitionParticles()
   &SkeletalMeshLocation(
    Bone / Socket Index=$Particles.PartitionParticles.Partition
    Bone Sampling Mode=::NewEnumerator4
    "Inherit Velocity (Scale)"=0.1
    MaskSpawnGroup=false
    Orientation Sampling=::NewEnumerator1
    Sampled Position Offset=&MaskVector(
     Mask Source=1
     Mask Parameter=$Particles.PartitionParticles.Partition
     Mask=0
     Masked Vector=<-4.5799 2.1042 0>
     Passthrough Vector=<-10.5293 -1.044 1.4884>
    )
    Skeletal Mesh=$System.Skeletal Mesh
    Velocity Sampling=::NewEnumerator0
    VelocitySpeedLimit=false
   )
   &InitializeRibbon(
    Color=<rgba 1 1 1 3>
    Lifetime=&FloatFromCurve(FloatCurve=DataInterface() {propertyValues {Curve {keys [<at 0 1>,<at 1 0>]}}})
    Position Mode:ENiagara_PositionInitializationMode=::NewEnumerator0
    Ribbon Facing Mode:ENiagara_UnsetDirectSet=::NewEnumerator1
    Ribbon Facing Vector:Vector3f=&RotateVectorByQuaternion(Vector To Rotate=<1 0 0>, Quaternion=$Particles.SkeletalMeshLocation.SampledRotation)
    Ribbon Width=40
    Write Material Random=true
   )
   &AddVelocity(
    Velocity=<0 0 -15>
    Velocity Speed Scale=&FloatFromCurve(FloatCurve=DataInterface() {propertyValues {Curve {keys [<at 0 1>,<at 0.2 0>,<at 0.5 0.55>,<at 1 0>]}}})
   )
   &PointForce(
    Linear Force Vector:Vector3f=&MaskPosition(
     Mask Parameter:NiagaraInt32=$Particles.PartitionParticles.Partition
     Mask:NiagaraInt32=0
     Masked Position:NiagaraPosition=<1 0 0>
     Passthrough Position:NiagaraPosition=<-1 0 0>
    )
    Module.Use Falloff=true
    Module.UseLinearForce=true
   )
   &ApplyInitialForces()
   /SetVariables(Particles.Aura Lifetime=0.9, Particles.Hand trail Lifetime=0.45) {isSetParameters=true}
   /SetVariables(Particles.Aura PivotOffset:Vector2f=&MaskVector2d(
     Mask Source:NiagaraInt32=0
     Mask Parameter:NiagaraInt32=$Particles.PartitionParticles.Partition
     Mask:NiagaraInt32=1
     Masked Vector2D:Vector2f=<0.5 0.4>
     Passthrough Vector2D:Vector2f=<0.5 0.6>
    )) {enabled=false, isSetParameters=true}
  ]
  particleUpdateScript [
   &ParticleState()
   /SetVariables(
    Particles.Update Aura age:NiagaraFloat=&Divide_Float(A=$Particles.Age, B=$Particles.Initial.Aura Lifetime)
    Particles.Update trail age:NiagaraFloat=&Divide_Float(A=$Particles.Age, B=$Particles.Initial.Hand trail Lifetime)
   ) {isSetParameters=true}
   /SetVariables(
    Particles.Update Aura color:LinearColor=&MakeLinearColorFromVectorAndFloat(
     "Vector (RGB)"=&MakeVectorFromLinearColorRGB(Color=&ColorFromCurve(
       ColorCurve=DataInterface() {propertyValues {
         AlphaCurve {keys [<at 0.0867 1 RCIM_Cubic>,<at 0.5 0.4335>,<at 1 0>]}
         BlueCurve {keys [{arriveTangent=4.0072, interpMode=RCIM_Cubic, time=0, value=0.8361}, <at 0.1526 1>, <at 0.5036 1>, <at 0.6607 1>, <at 1 1 RCIM_Cubic>]}
         GreenCurve {keys [{arriveTangent=-0.4007, interpMode=RCIM_Cubic, time=0, value=1}, <at 0.1526 0>, <at 0.5036 0>, <at 0.6607 0.2>, <at 1 0.4 RCIM_Cubic>]}
         RedCurve {keys [
           {arriveTangent=-11.8213, interpMode=RCIM_Cubic, time=0, value=0}
           <at 0.1526 0.2883>
           <at 0.5036 0.6667>
           <at 0.6607 0.2>
           <at 1 0.1 RCIM_Cubic>
          ]}
        }}
       CurveIndex=$Particles.Update Aura age
      ))
     "Float (Alpha)"=&FloatFromCurve(
      FloatCurve=DataInterface() {propertyValues {Curve {keys [
          <at 0 0 RCIM_Cubic>
          {arriveTangent=1.0349, interpMode=RCIM_Cubic, leaveTangent=1.0349, time=0.1424, value=1}
          {interpMode=RCIM_Cubic, tangentMode=RCTM_User, time=0.5993, value=0.6202}
          {interpMode=RCIM_Cubic, tangentMode=RCTM_User, time=1, value=0}
         ]}}}
      CurveIndex=$Particles.Update Aura age
      Scale Curve=0.5
     )
    )
    Particles.Update flare Color:LinearColor=&MakeLinearColorFromVectorAndFloat(
     "Vector (RGB)"=&MakeVectorFromLinearColorRGB(Color=&ColorFromCurve(
       ColorCurve=DataInterface() {propertyValues {
         AlphaCurve {keys [<at 0.0867 1 RCIM_Cubic>,<at 0.5 0.4335>,<at 0.9101 0>]}
         BlueCurve {keys [{arriveTangent=4.0072, interpMode=RCIM_Cubic, time=0, value=0.8361}, <at 0.1526 1>, <at 0.5036 1>, <at 0.6607 1>, <at 1 1 RCIM_Cubic>]}
         GreenCurve {keys [{arriveTangent=-0.4007, interpMode=RCIM_Cubic, time=0, value=1}, <at 0.1526 0>, <at 0.5036 0>, <at 0.6607 0.2>, <at 1 0.4 RCIM_Cubic>]}
         RedCurve {keys [
           {arriveTangent=-11.8213, interpMode=RCIM_Cubic, time=0, value=0}
           <at 0.1526 0.2883>
           <at 0.5036 0.6667>
           <at 0.6607 0.2>
           <at 1 0.1 RCIM_Cubic>
          ]}
        }}
       CurveIndex=$Particles.Update Aura age
      ))
     "Float (Alpha)"=&FloatFromCurve(
      FloatCurve=DataInterface() {propertyValues {Curve {keys [
          <at 0 0 RCIM_Cubic>
          {arriveTangent=1.0349, interpMode=RCIM_Cubic, leaveTangent=1.0349, time=0.1424, value=1}
          {interpMode=RCIM_Cubic, tangentMode=RCTM_User, time=0.5993, value=0.6202}
          {interpMode=RCIM_Cubic, tangentMode=RCTM_User, time=1, value=0}
         ]}}}
      CurveIndex=$Particles.Update Aura age
     )
    )
    Particles.Update hand trail color:LinearColor=&MakeLinearColorFromVectorAndFloat(
     "Vector (RGB)"=&MakeVectorFromLinearColorRGB(Color=&ColorFromCurve(
       ColorCurve=DataInterface() {propertyValues {
         AlphaCurve {keys [<at 0 0.3116 RCIM_Cubic>,<at 0.5 1>,<at 0.9949 0>]}
         BlueCurve {keys [{arriveTangent=4.0072, interpMode=RCIM_Cubic, time=0, value=0.8361}, <at 0.1526 1>, <at 0.5036 1>, <at 0.6607 1>, <at 1 1 RCIM_Cubic>]}
         GreenCurve {keys [{arriveTangent=-0.4007, interpMode=RCIM_Cubic, time=0, value=1}, <at 0.1526 0>, <at 0.5036 0>, <at 0.6607 0.2>, <at 1 0.4 RCIM_Cubic>]}
         RedCurve {keys [
           {arriveTangent=-11.8213, interpMode=RCIM_Cubic, time=0, value=0}
           <at 0.1526 0.2883>
           <at 0.5036 0.6667>
           <at 0.6607 0.2>
           <at 1 0.1 RCIM_Cubic>
          ]}
        }}
       CurveIndex=$Particles.Update trail age
      ))
     "Float (Alpha)"=&FloatFromCurve(
      FloatCurve=DataInterface() {propertyValues {Curve {keys [
          {arriveTangent=3.1524, interpMode=RCIM_Cubic, leaveTangent=3.1524, tangentMode=RCTM_User, time=0, value=0}
          {arriveTangent=0.3962, interpMode=RCIM_Cubic, leaveTangent=0.3962, tangentMode=RCTM_User, time=0.3, value=1}
          {arriveTangent=-1.7441, interpMode=RCIM_Cubic, leaveTangent=-1.7441, tangentMode=RCTM_User, time=0.47, value=0.95}
          {arriveTangent=-0.4506, interpMode=RCIM_Cubic, leaveTangent=-0.4506, tangentMode=RCTM_User, time=0.87, value=0}
         ]}}}
      CurveIndex=$Particles.Update trail age
     )
    )
   ) {isSetParameters=true}
   /SetVariables(
    Particles.Aura SpriteAlignment:Vector3f=&RotateVectorByQuaternion(Vector To Rotate=<-0.1368 1 0>, Quaternion=$Particles.SkeletalMeshLocation.SampledRotation)
    Particles.Update aura size:Vector2f=&Vector2DFromCurve(
     Vector2Curve=DataInterface() {propertyValues {
       XCurve {keys [{interpMode=RCIM_Cubic, tangentMode=RCTM_User, time=0, value=0}, {interpMode=RCIM_Cubic, tangentMode=RCTM_User, time=0.05, value=0.6085}]}
       YCurve {keys [{interpMode=RCIM_Cubic, tangentMode=RCTM_User, time=0, value=0}, {interpMode=RCIM_Cubic, tangentMode=RCTM_User, time=0.05, value=0.9144}]}
      }}
     ScaleCurve=<1.8 1.8>
    )
   ) {isSetParameters=true}
   /SetVariables(Particles.aura position:NiagaraPosition=&MaskVector(
     Mask Source=1
     Mask Parameter=$Particles.PartitionParticles.Partition
     Mask=0
     Masked Vector=<-42.0089 10 75.0259>
     Passthrough Vector=<41.6578 7.7354 76.9477>
    )) {isSetParameters=true}
   &ScaleColor(
    Scale Alpha=&FloatFromCurve(
     FloatCurve=DataInterface() {propertyValues {Curve {keys [
         {arriveTangent=3.1524, interpMode=RCIM_Cubic, leaveTangent=3.1524, tangentMode=RCTM_User, time=0, value=0}
         {arriveTangent=0.3962, interpMode=RCIM_Cubic, leaveTangent=0.3962, tangentMode=RCTM_User, time=0.3, value=1}
         {arriveTangent=-1.7441, interpMode=RCIM_Cubic, leaveTangent=-1.7441, tangentMode=RCTM_User, time=0.47, value=0.95}
         {arriveTangent=-0.4506, interpMode=RCIM_Cubic, leaveTangent=-0.4506, tangentMode=RCTM_User, time=0.87, value=0}
        ]}}}
     CurveIndex=$Emitter.Age
    )
    Scale RGB=&MakeVectorFromLinearColorRGB(Color=&ColorFromCurve(ColorCurve=DataInterface() {propertyValues {
        AlphaCurve {keys [<at 0 0.3116 RCIM_Cubic>,<at 0.5 1>,<at 0.9949 0>]}
        BlueCurve {keys [{arriveTangent=4.0072, interpMode=RCIM_Cubic, time=0, value=0.8361}, <at 0.1526 1>, <at 0.5036 1>, <at 0.6607 1>, <at 1 1 RCIM_Cubic>]}
        GreenCurve {keys [{arriveTangent=-0.4007, interpMode=RCIM_Cubic, time=0, value=1}, <at 0.1526 0>, <at 0.5036 0>, <at 0.6607 0.2>, <at 1 0.4 RCIM_Cubic>]}
        RedCurve {keys [
          {arriveTangent=-11.8213, interpMode=RCIM_Cubic, time=0, value=0}
          <at 0.1526 0.2883>
          <at 0.5036 0.6667>
          <at 0.6607 0.2>
          <at 1 0.1 RCIM_Cubic>
         ]}
       }}))
    ScaleA=true
    ScaleRGB=true
   ) {enabled=false}
   &ScaleColor(
    Scale Alpha=0.7
    Scale RGB=&VectorFromFloat(Value=&FNTimeOfDayScalar(Night Scale=25, Morning Scale=1, Day Scale=25, Evening Scale=1))
    ScaleA=false
    ScaleRGB=true
   ) {module=ScaleColor001}
   &ScaleRibbonWidth(Ribbon Width Scale=&FloatFromCurve(FloatCurve=DataInterface() {propertyValues {Curve {
        defaultValue=0.5
        keys [{interpMode=RCIM_Cubic, tangentMode=RCTM_User, time=0, value=1}, {interpMode=RCIM_Cubic, tangentMode=RCTM_User, time=1, value=0}]
       }}}))
   &DynamicMaterialParameters(
    Index 0 Param 1=&FloatFromCurve(FloatCurve=DataInterface() {propertyValues {Curve {keys [
         {arriveTangent=0.2015, interpMode=RCIM_Cubic, leaveTangent=0.2015, tangentMode=RCTM_User, time=0, value=1}
         {arriveTangent=-1.5941, interpMode=RCIM_Cubic, leaveTangent=-1.5941, tangentMode=RCTM_User, time=0.3, value=0.5}
         {interpMode=RCIM_Cubic, tangentMode=RCTM_User, time=1, value=0}
        ]}}})
    Index 0 Param 3=-1
    Module.Index0Float0Write=true
    Module.Index0Float1Write=false
    Module.Index0Float2Write=true
    Module.Index0Float3Write=true
   ) {module=DynamicMaterialParameters001}
   &SolveForcesAndVelocity(Clamp Velocity=false, Limit Acceleration=false)
  ]
  rendererClasses [&NiagaraRibbonRendererProperties, &NiagaraSpriteRendererProperties]
  renderers [
   {rendererClass=&NiagaraRibbonRendererProperties}
   {rendererClass=&NiagaraRibbonRendererProperties, rendererIndex=1}
   {rendererClass=&NiagaraRibbonRendererProperties, rendererIndex=2}
   {rendererClass=&NiagaraSpriteRendererProperties, rendererIndex=3}
   {rendererClass=&NiagaraSpriteRendererProperties, rendererIndex=4}
  ]
 }
 ribbon_both_hands {
  emitterUpdateScript [
   &EmitterState(Life Cycle Mode=::NewEnumerator1, Loop Duration=1, UseLoopDelay=false)
   &SpawnBurst_Instantaneous(Use Spawn Probability=false, UseLoopCountLimit=false) {enabled=false}
   &SpawnRate(SpawnRate=60, Use Spawn Probability=false)
  ]
  particleSpawnScript [
   &InitializeParticle(
    Color=<rgba 1.5 1.5 1.5 1.5>
    Lifetime=0.25
    Ribbon Width=20
    Ribbon Width Mode=::NewEnumerator1
    UsePositionOffset=false
    Write Material Random=true
   )
   &PartitionParticles()
   &SkeletalMeshLocation(
    Bone / Socket Index=$Particles.PartitionParticles.Partition
    Bone Sampling Mode=::NewEnumerator4
    "Inherit Velocity (Scale)"=0.1
    MaskSpawnGroup=false
    Orientation Sampling=::NewEnumerator1
    Sampled Position Offset=&MaskVector(
     Mask Source=1
     Mask Parameter=$Particles.PartitionParticles.Partition
     Mask=0
     Masked Vector=<-4.5799 2.1042 0>
     Passthrough Vector=<-10.5293 -1.044 0>
    )
    Skeletal Mesh=$System.Skeletal Mesh
    Velocity Sampling=::NewEnumerator0
    VelocitySpeedLimit=false
   ) {enabled=false, module=SkeletalMeshLocation001}
   &SkeletalMeshLocation(
    Bone / Socket Index=$Particles.PartitionParticles.Partition
    Bone Sampling Mode=::NewEnumerator4
    "Inherit Velocity (Scale)"=0.1
    MaskSpawnGroup=false
    Sampled Position Offset=&MaskVector(
     Mask Source=1
     Mask Parameter=$Particles.PartitionParticles.Partition
     Mask=0
     Masked Vector=<0 -4.8987 0>
     Passthrough Vector=<-8.9766 -0.5 0>
    )
    Skeletal Mesh=$System.Skeletal Mesh
    Velocity Sampling=::NewEnumerator0
    VelocitySpeedLimit=false
   ) {enabled=false}
   &SkeletalMeshLocation(
    Bone / Socket Index=$Particles.PartitionParticles.Partition
    Bone Sampling Mode=::NewEnumerator4
    "Inherit Velocity (Scale)"=0.1
    MaskSpawnGroup=false
    Sampled Position Offset=&MaskVector(Mask Source=1, Mask Parameter=$Particles.PartitionParticles.Partition, Mask=0, Masked Vector=<0 -4.8987 0>, Passthrough Vector=<0 0 0>)
    Skeletal Mesh=$System.Skeletal Mesh
    Velocity Sampling=::NewEnumerator0
    VelocitySpeedLimit=false
   ) {module=SkeletalMeshLocation002}
   &AddVelocity(Velocity=<0 0 191.9938>) {enabled=false}
   &PointForce(
    Linear Force Vector:Vector3f=&MaskPosition(
     Mask Source:NiagaraInt32=0
     Mask Parameter:NiagaraInt32=$Particles.PartitionParticles.Partition
     Mask:NiagaraInt32=0
     Masked Position:NiagaraPosition=<1 0 0>
     Passthrough Position:NiagaraPosition=<-1 0 0>
    )
    Module.Use Falloff=true
    Module.UseLinearForce=true
   )
   &ApplyInitialForces()
  ]
  particleUpdateScript [
   &ParticleState()
   &InheritSourceMovement(Applied Position Scale=<2 2 2>, ApplyForce=false, ApplyPosition=true, ApplyPositionFalloff=false)
   &ScaleRibbonWidth(Ribbon Width Scale=&FloatFromCurve(FloatCurve=DataInterface() {propertyValues {Curve {
        defaultValue=1
        keys [
         {arriveTangent=3.5475, interpMode=RCIM_Cubic, leaveTangent=3.5475, tangentMode=RCTM_User, time=0, value=0.45}
         {arriveTangent=-0.1116, interpMode=RCIM_Cubic, leaveTangent=-0.1116, tangentMode=RCTM_User, time=0.3, value=1}
         {arriveTangent=-0.0405, interpMode=RCIM_Cubic, leaveTangent=-0.0405, tangentMode=RCTM_User, time=1, value=0}
        ]
       }}}))
   &ScaleColor(
    Scale Alpha=&FloatFromCurve(
     FloatCurve=DataInterface() {propertyValues {Curve {keys [
         <at 0 0 RCIM_Cubic>
         {arriveTangent=-3.6678, interpMode=RCIM_Cubic, leaveTangent=-3.6678, tangentMode=RCTM_User, time=0.05, value=0.95}
         <at 0.45 0 RCIM_Cubic>
        ]}}}
     Scale Curve=&Multiply_Float(
      A=&FN_PickFloatDependOnPlatform(TargetPlatformActive=1, TargetPlatformNoActive=0.7)
      B=&FNTimeOfDayScalar(Night Scale=0.6, Morning Scale=1, Evening Scale=1)
     )
    )
    Scale RGB=&MakeVectorFromLinearColorRGB(Color=&ColorFromCurve(ColorCurve=DataInterface() {propertyValues {
        AlphaCurve {keys [<at 0 0 RCIM_Cubic>,<at 0.3319 0.9969>,<at 0.6824 1>,<at 0.9204 0>]}
        BlueCurve {keys [{arriveTangent=4.0072, interpMode=RCIM_Cubic, time=0, value=0.8361}, <at 0.1526 1>, <at 0.5036 1>, <at 0.6607 1>, <at 1 1 RCIM_Cubic>]}
        GreenCurve {keys [{arriveTangent=-0.4007, interpMode=RCIM_Cubic, time=0, value=1}, <at 0.1526 0>, <at 0.5036 0>, <at 0.6607 0.2>, <at 1 0.4 RCIM_Cubic>]}
        RedCurve {keys [
          {arriveTangent=-11.8213, interpMode=RCIM_Cubic, time=0, value=0}
          <at 0.1526 0.2883>
          <at 0.5036 0.6667>
          <at 0.6607 0.2>
          <at 1 0.1 RCIM_Cubic>
         ]}
       }}))
    ScaleA=true
    ScaleRGB=true
   ) {enabled=false}
   &ScaleColor(
    Scale Alpha=&FloatFromCurve(
     FloatCurve=DataInterface() {propertyValues {Curve {keys [
         {arriveTangent=3.1524, interpMode=RCIM_Cubic, leaveTangent=3.1524, tangentMode=RCTM_User, time=0, value=0}
         {arriveTangent=0.3962, interpMode=RCIM_Cubic, leaveTangent=0.3962, tangentMode=RCTM_User, time=0.1, value=0.95}
         {arriveTangent=-0.4506, interpMode=RCIM_Cubic, leaveTangent=-0.4506, tangentMode=RCTM_User, time=0.37, value=0}
        ]}}}
     CurveIndex=$Emitter.Age
    )
    Scale RGB=&MakeVectorFromLinearColorRGB(Color=&ColorFromCurve(ColorCurve=DataInterface() {propertyValues {
        AlphaCurve {keys [<at 0 0.3116 RCIM_Cubic>,<at 0.5 1>,<at 0.9949 0>]}
        BlueCurve {keys [{arriveTangent=4.0072, interpMode=RCIM_Cubic, time=0, value=0.8361}, <at 0.1526 1>, <at 0.5036 1>, <at 0.6607 1>, <at 1 1 RCIM_Cubic>]}
        GreenCurve {keys [{arriveTangent=-0.4007, interpMode=RCIM_Cubic, time=0, value=1}, <at 0.1526 0>, <at 0.5036 0>, <at 0.6607 0.2>, <at 1 0.4 RCIM_Cubic>]}
        RedCurve {keys [
          {arriveTangent=-11.8213, interpMode=RCIM_Cubic, time=0, value=0}
          <at 0.1526 0.2883>
          <at 0.5036 0.6667>
          <at 0.6607 0.2>
          <at 1 0.1 RCIM_Cubic>
         ]}
       }}))
    ScaleA=true
    ScaleRGB=true
   ) {module=ScaleColor001}
   &ScaleColor(
    Scale Alpha=0.7
    Scale RGB=&VectorFromFloat(Value=&FNTimeOfDayScalar(Night Scale=25, Morning Scale=1, Day Scale=25, Evening Scale=1))
    ScaleA=false
    ScaleRGB=true
   ) {module=ScaleColor002}
   &SolveForcesAndVelocity(Clamp Velocity=false, Limit Acceleration=false)
  ]
  rendererClasses [&NiagaraRibbonRendererProperties]
  renderers [
   {rendererClass=&NiagaraRibbonRendererProperties}
   {rendererClass=&NiagaraRibbonRendererProperties, rendererIndex=1}
   {rendererClass=&NiagaraRibbonRendererProperties, rendererIndex=2}
  ]
 }
}
summary {
 systemName=NS_Emote_Polaritywinn_hand_trail_V2
 emitters [
  {
   emitterName=aura_No_flare
   rendererClasses [
    &NiagaraSpriteRendererProperties
    &NiagaraMeshRendererProperties
    &NiagaraRibbonRendererProperties
   ]
  }
  {bEnabled=false, emitterName=New_version, rendererClasses [&NiagaraRibbonRendererProperties]}
  {
   bEnabled=false
   emitterName=hAND_TRAIL_AND_AURA
   rendererClasses [&NiagaraRibbonRendererProperties, &NiagaraSpriteRendererProperties]
  }
  {emitterName=ribbon_both_hands, rendererClasses [&NiagaraRibbonRendererProperties]}
 ]
 userVariables [
  User.Anim_Curve_InitialOpacity {type=NiagaraFloat, defaultValue=1, description=""}
  User.Anim_Curve_SpawnRate {type=NiagaraFloat, defaultValue=0, description=""}
  User.bDrums {type=NiagaraFloat, defaultValue=0, description=""}
  User.Skeletal Mesh {
   type=NiagaraDataInterfaceSkeletalMesh
   defaultValue {
    struct=NiagaraExt_VariableValue_DataInterface
    value {dataInterface=&self:NiagaraDataInterfaceSkeletalMesh_0}
   }
  }
 ]
}
